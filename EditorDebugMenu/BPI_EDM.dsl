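; BPI_EDM is a Blueprint interface for the Editor Debug Menu (EDM), an editor-side debug tool.
; It declares calls to run a console command in a chosen world type, an editor utility task class, or an editor utility object via soft class references.
; Further calls toggle favourites and update collection membership on input data objects, and trigger a save.
; It holds only signatures, so implementers supply the behaviour.
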(fn RunEDMConsoleCommand
  (params (Console%20Command :type String) (Target%20World :type Enum_EDM_WorldType%20Enum :default "NewEnumerator0")))

(fn RunEDMTask
  (params (Task%20Class :type EUT%20EDM%20Base%20Soft%20Class%20Reference)))

(fn RunObject
  (params (EUO :type Editor%20Utility%20Object%20Soft%20Class%20Reference)))

(fn ToggleEDMFavourite
  (params (Add :type Boolean) (Input%20Data%20Obj :type Obj%20EDM%20Input%20Data%20Object%20Reference%20%28by%20ref%29)))

(fn TriggerSaveEDM)

(fn UpdateEDMCollections
  (params (Collections%20To%20Add :type Set%20of%20Strings) (Collections%20To%20Remove :type Set%20of%20Strings) (Data%20Object :type Obj%20EDM%20Input%20Data%20Object%20Reference%20%28by%20ref%29)))
