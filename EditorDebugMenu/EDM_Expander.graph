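# EDM_Expander is an Editor Utility Widget that provides a collapsible section, apparently for the Editor Debug Menu.
# An expandable area, open by default, hosts a title text and a rounded border wrapping a vertical StackBox.
# Add To Widget appends supplied widgets to that StackBox, and a "Set Expanded]" event sets the open state.
# Expansion changes are re-broadcast through the Expansion Changed dispatcher with the new state.
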EditorUtilityWidgetBlueprint /EditorDebugMenu/EDM_Expander
graphs [
 EventGraph {nodes [
   n1 = "On Expansion Changed (EditorUtilityExpandableArea_75)"!() -> (OutputDelegate, then, Area:Expandable Area Object Reference, bIsExpanded:bool) {is=title K2Node_ComponentBoundEvent}
   > Call Expansion Changed!(execute:Exec, self:EDM Expander Object Reference, Expanded:bool=@n1.bIsExpanded) -> (then:Exec) {is=title K2Node_CallDelegate}
   n2 = "Set Expanded]"!() -> (OutputDelegate, then, Is Expanded:bool=~) {is=title K2Node_CustomEvent}
   > SetIsExpanded!(
    execute
    self:Expandable Area Object Reference=Get EditorUtilityExpandableArea_75() -> (EditorUtilityExpandableArea_75:Editor Utility Expandable Area Object Reference)
    IsExpanded:bool=@n2.Is Expanded
   ) -> (then)
   n3 = Add To Widget!() -> (OutputDelegate, then, Content:Widget Object Reference) {is=title K2Node_CustomEvent}
   > AddChildToStackBox!(
    execute
    self:Stack Box Object Reference=Get WidgetContainer() -> (WidgetContainer:Stack Box Object Reference)
    Content:Widget Object Reference=@n3.Content
   ) -> (then, ReturnValue:Stack Box Slot Object Reference)
   Event PreConstruct()
   Event Construct()
   Event Tick()
  ]}
 Expansion Changed {nodes [Expansion Changed!() -> (then, Expanded:bool=~) {is=title K2Node_FunctionEntry}]}
]
variables [Expansion Changed {default=()}]
widgets {
 info {namedSlotCount=2, rootWidgetClass=/Script/Blutility.EditorUtilityExpandableArea, widgetCount=4}
 widgets [
  EditorUtilityExpandableArea_75 {
   widgetClassPath=&EditorUtilityExpandableArea
   bIsVariable=true
   widget=EditorUtilityExpandableArea_75
   properties {
    bIsExpanded=true
    areaPadding=<ltrb 0 0 0 0>
    borderBrush {
     bIsDynamicallyLoaded=false
     drawAs=RoundedBox
     imageType=NoImage
     mirroring=NoMirror
     resourceName=None
     resourceObject=null
     tiling=NoTile
     imageSize=<0 0>
     margin=<ltrb 0 0 0 0>
     outlineSettings {
      bUseBrushTransparency=false
      roundingType=FixedRadius
      width=0
      color {colorUseRule=UseColor_Specified, specifiedColor=<rgba 0 0 0 0>}
      cornerRadii=<2 2 2 2>
     }
     tintColor {colorUseRule=UseColor_Specified, specifiedColor=<rgba 0.0365 0.0365 0.0365 1>}
     uVRegion {bIsValid=false, max=<0 0>, min=<0 0>}
    }
    style {
     rolloutAnimationSeconds=0
     collapsedImage {
      bIsDynamicallyLoaded=false
      drawAs=Image
      imageType=Vector
      mirroring=NoMirror
      resourceName=../../../Engine/Content/Slate/Starship/Common/caret-right.svg
      resourceObject=null
      tiling=NoTile
      imageSize=<16 16>
      margin=<ltrb 0 0 0 0>
      outlineSettings {
       bUseBrushTransparency=false
       roundingType=HalfHeightRadius
       width=0
       color {colorUseRule=UseColor_Specified, specifiedColor=<rgba 0 0 0 0>}
       cornerRadii=<0 0 0 0>
      }
      tintColor {colorUseRule=UseColor_Specified, specifiedColor=<rgba 0.5271 0.5271 0.5271 1>}
      uVRegion {bIsValid=false, max=<0 0>, min=<0 0>}
     }
     expandedImage {
      bIsDynamicallyLoaded=false
      drawAs=Image
      imageType=Vector
      mirroring=NoMirror
      resourceName=../../../Engine/Content/Slate/Starship/Common/caret-down.svg
      resourceObject=null
      tiling=NoTile
      imageSize=<16 16>
      margin=<ltrb 0 0 0 0>
      outlineSettings {
       bUseBrushTransparency=false
       roundingType=HalfHeightRadius
       width=0
       color {colorUseRule=UseColor_Specified, specifiedColor=<rgba 0 0 0 0>}
       cornerRadii=<0 0 0 0>
      }
      tintColor {colorUseRule=UseColor_Specified, specifiedColor=<rgba 0.5271 0.5271 0.5271 1>}
      uVRegion {bIsValid=false, max=<0 0>, min=<0 0>}
     }
    }
   }
  }
  ProjectTitle {
   widgetClassPath=&UEFN_TextBlock
   bIsVariable=true
   namedSlotHost=EditorUtilityExpandableArea_75
   widget=ProjectTitle
   properties {
    text="[Fortnite]"
    colorAndOpacity {colorUseRule=UseColor_Specified, specifiedColor=<rgba 0.3542 0.3542 0.3542 1>}
    font {
     bForceMonospaced=false
     bMaterialIsStencil=false
     fontMaterial=null
     fontObject=&NotoSans
     letterSpacing=0
     monospacedWidth=1
     size=7.5
     skewAmount=0
     typefaceFontName=NotoSans-Bold
     outlineSettings {
      bApplyOutlineToDropShadows=false
      bMiteredCorners=false
      bSeparateFillAlpha=false
      outlineMaterial=null
      outlineSize=0
      outlineColor=<rgba 0 0 0 1>
     }
    }
   }
  }
  Border {
   widgetClassPath=&Border
   bIsVariable=true
   namedSlotHost=EditorUtilityExpandableArea_75
   widget=Border
   properties {background {
     bIsDynamicallyLoaded=false
     drawAs=RoundedBox
     imageType=NoImage
     mirroring=NoMirror
     resourceName=None
     resourceObject=null
     tiling=NoTile
     imageSize=<32 32>
     margin=<ltrb 0 0 0 0>
     outlineSettings {
      bUseBrushTransparency=false
      roundingType=FixedRadius
      width=0
      color {colorUseRule=UseColor_Specified, specifiedColor=<rgba 0 0 0 0>}
      cornerRadii=<5 5 5 5>
     }
     tintColor {colorUseRule=UseColor_Specified, specifiedColor=<rgba 0.0156 0.0156 0.0156 1>}
     uVRegion {bIsValid=false, max=<0 0>, min=<0 0>}
    }}
  }
  WidgetContainer {
   widgetClassPath=&StackBox
   bIsVariable=true
   parent=Border
   widget=WidgetContainer
   properties {orientation=Orient_Vertical}
   slot {class=BorderSlot}
  }
 ]
}
