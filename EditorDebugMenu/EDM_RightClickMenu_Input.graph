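# EDM_RightClickMenu_Input is an Editor Utility Widget, apparently an input element for the EditorDebugMenu right-click menu.
# A dark Border wraps a StackBox holding a text label and an editor string combo box exposed as a variable.
# The label and dropdown options are placeholder values, and the PreConstruct, Construct and Tick events are empty, so the widget is layout only with no logic.
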EditorUtilityWidgetBlueprint /EditorDebugMenu/EDM_RightClickMenu_Input
graphs [EventGraph {nodes [
   Event PreConstruct()
   Event Construct()
   Event Tick()
  ]}]
variables []
widgets {
 info {rootWidgetClass=&Border, widgetCount=4}
 widgets [
  Border_0 {widgetClassPath=&Border, widget=Border_0, properties {brushColor=<rgba 0.0288 0.0312 0.0264 1>}}
  StackBox_0 {widgetClassPath=&StackBox, parent=Border_0, widget=StackBox_0, slot {class=BorderSlot}}
  TextBlock_45 {widgetClassPath=&TextBlock, parent=StackBox_0, widget=TextBlock_45, properties {text=Text Block}, slot {class=StackBoxSlot}}
  EditorUtilityComboBoxString_92 {
   widgetClassPath=/Script/Blutility.EditorUtilityComboBoxString
   bIsVariable=true
   parent=StackBox_0
   widget=EditorUtilityComboBoxString_92
   properties {defaultOptions [1, 2, A really big string], foregroundColor {colorUseRule=UseColor_Specified, specifiedColor=<rgba 0.5271 0.5271 0.5271 1>}}
   slot {class=StackBoxSlot}
  }
 ]
}
