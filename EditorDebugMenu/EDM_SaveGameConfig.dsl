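; Save-game config for the editor debug menu that holds its UI and input state: favourites, search text, category states, project filters and inputs.
; Getter and setter functions wrap sets and maps keyed by string, with project filters defaulting to visible when absent.
; Inputs are stored as object references, with a printed warning when a loaded one is invalid, and input collections as string sets wrapped in structs.
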(fn CheckIsFavourite
  (params (ItemToFind :type String%20%28by%20ref%29))
  (return (Utilities|Set|ContainsItem (Variables|Default|GetFavourites) ItemToFind)))

(fn GetProjectFilterVisibility
  (params (Project :type String%20%28by%20ref%29))
  (bind (_value _returnvalue) (Utilities|Map|Find (Variables|Default|GetProject%20Filters) Project))
  (return (select _returnvalue _value true)))

(fn GetSearch
  (return (Variables|Default|GetSearch)))

(fn GetState
  (params (Category :type String))
  (return (Utilities|Map|Find (Variables|Default|GetStates) Category)))

(fn LoadInput
  (params (Key :type String%20%28by%20ref%29))
  (bind _input_objects (Variables|Default|GetInput%20Objects))
  (bind (_value _returnvalue) (Utilities|Map|Find _input_objects Key))
  (Utilities|IsValid _value
    (:Is%20Valid
      (return _returnvalue _value))
    (:Is%20Not%20Valid
      (Development|PrintString 0 (Utilities|String|Append Key " - not valid"))
      (return false 0))))

(fn LoadInputCollection
  (params (Input%20Unique%20ID :type String%20%28by%20ref%29))
  (bind _collections (Variables|Default|GetCollections))
  (bind (_value _returnvalue) (Utilities|Map|Find _collections Input%20Unique%20ID))
  (if _returnvalue
    (return (Utilities|Struct|BreakStructEDMCollection _value))
    (else
      (return 0))))

(fn ModifyFavourite
  (params (NewItem :type String%20%28by%20ref%29) (New%20State :type Boolean))
  (bind _favourites (Variables|Default|GetFavourites))
  (if New%20State
    (Utilities|Set|Add _favourites NewItem)
    (else
      (Utilities|Set|Remove _favourites NewItem))))

(fn SaveInput
  (params (Key :type String%20%28by%20ref%29) (Input%20Object :type Obj%20EDM%20Input%20Data%20Object%20Reference%20%28by%20ref%29))
  (Utilities|Map|Add (Variables|Default|GetInput%20Objects) Key Input%20Object))

(fn SaveInputCollection
  (params (Input%20Unique%20ID :type String%20%28by%20ref%29) (Collection :type Set%20of%20Strings))
  (Utilities|Map|Add (Variables|Default|GetCollections) Input%20Unique%20ID (Utilities|Struct|MakeStructEDMCollection Collection)))

(fn SaveSearch
  (params (Search :type String))
  (Variables|Default|SetSearch Search))

(fn SaveState
  (params (Category :type String) (State :type Boolean))
  (Utilities|Map|Add (Variables|Default|GetStates) Category State))

(fn SetProjectFilterVisibility
  (params (Project :type String) (Vis%20State :type Boolean))
  (Utilities|Map|Add (Variables|Default|GetProject%20Filters) Project Vis%20State))
