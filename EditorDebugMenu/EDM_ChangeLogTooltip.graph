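# Editor utility widget that shows the change log of the Editor Debug Menu, intended as a tooltip.
# A black border holds a vertical stack box of bold version headers, V1.0.2 then V1.0.1, each followed by regular-weight release notes.
# All text is hardcoded in the text blocks, and the PreConstruct, Construct and Tick events contain no logic.
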EditorUtilityWidgetBlueprint /EditorDebugMenu/EDM_ChangeLogTooltip
graphs [EventGraph {nodes [
   Event PreConstruct()
   Event Construct()
   Event Tick()
  ]}]
variables []
widgets {
 info {rootWidgetClass=&Border, widgetCount=6}
 widgets [
  Border_0 {widgetClassPath=&Border, widget=Border_0, properties {brushColor=<rgba 0 0 0 1>}}
  StackBox_0 {widgetClassPath=&StackBox, parent=Border_0, widget=StackBox_0, properties {orientation=Orient_Vertical}, slot {class=BorderSlot}}
  Txt_ExpandAll_1 {
   widgetClassPath=&UEFN_TextBlock
   parent=StackBox_0
   widget=Txt_ExpandAll_1
   properties {
    text=V1.0.2
    font {
     bForceMonospaced=false
     bMaterialIsStencil=false
     fontMaterial=null
     fontObject=&NotoSans
     letterSpacing=0
     monospacedWidth=1
     size=9
     skewAmount=0
     typefaceFontName=NotoSans-Bold
     outlineSettings {
      bApplyOutlineToDropShadows=false
      bMiteredCorners=false
      bSeparateFillAlpha=false
      outlineMaterial=null
      outlineSize=0
      outlineColor=<rgba 0 0 0 1>
     }
    }
   }
   slot {class=StackBoxSlot, properties {horizontalAlignment=HAlign_Left, verticalAlignment=VAlign_Top, padding=<ltrb 2 2 2 2>}}
  }
  Txt_ExpandAll {
   widgetClassPath=&UEFN_TextBlock
   parent=StackBox_0
   widget=Txt_ExpandAll
   properties {
    text="- Users can now run Editor Utility Objects (EUO) directly from the menu. \r\n- EUO's allow you to communicate with the game world. For exampel, running functions or blueprints. \r\n- There is a new drop down input option to add a EUO."
    font {
     bForceMonospaced=false
     bMaterialIsStencil=false
     fontMaterial=null
     fontObject=&NotoSans
     letterSpacing=0
     monospacedWidth=1
     size=7.5
     skewAmount=0
     typefaceFontName=NotoSans-Regular
     outlineSettings {
      bApplyOutlineToDropShadows=false
      bMiteredCorners=false
      bSeparateFillAlpha=false
      outlineMaterial=null
      outlineSize=0
      outlineColor=<rgba 0 0 0 1>
     }
    }
   }
   slot {class=StackBoxSlot, properties {horizontalAlignment=HAlign_Left, verticalAlignment=VAlign_Top, padding=<ltrb 2 2 2 2>}}
  }
  Txt_ExpandAll_3 {
   widgetClassPath=&UEFN_TextBlock
   parent=StackBox_0
   widget=Txt_ExpandAll_3
   properties {
    text=V1.0.1
    font {
     bForceMonospaced=false
     bMaterialIsStencil=false
     fontMaterial=null
     fontObject=&NotoSans
     letterSpacing=0
     monospacedWidth=1
     size=9
     skewAmount=0
     typefaceFontName=NotoSans-Bold
     outlineSettings {
      bApplyOutlineToDropShadows=false
      bMiteredCorners=false
      bSeparateFillAlpha=false
      outlineMaterial=null
      outlineSize=0
      outlineColor=<rgba 0 0 0 1>
     }
    }
   }
   slot {class=StackBoxSlot, properties {horizontalAlignment=HAlign_Left, verticalAlignment=VAlign_Top, padding=<ltrb 2 2 2 2>}}
  }
  Txt_ExpandAll_2 {
   widgetClassPath=&UEFN_TextBlock
   parent=StackBox_0
   widget=Txt_ExpandAll_2
   properties {
    text="- Added Source Location Data\r\n- Tooltip shows source control location\r\n- Middle Mouse Button takes user to source control location in content browser"
    font {
     bForceMonospaced=false
     bMaterialIsStencil=false
     fontMaterial=null
     fontObject=&NotoSans
     letterSpacing=0
     monospacedWidth=1
     size=7.5
     skewAmount=0
     typefaceFontName=NotoSans-Regular
     outlineSettings {
      bApplyOutlineToDropShadows=false
      bMiteredCorners=false
      bSeparateFillAlpha=false
      outlineMaterial=null
      outlineSize=0
      outlineColor=<rgba 0 0 0 1>
     }
    }
   }
   slot {class=StackBoxSlot, properties {horizontalAlignment=HAlign_Left, verticalAlignment=VAlign_Top, padding=<ltrb 2 2 2 2>}}
  }
 ]
}
