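# Editor Utility Widget that serves as the "What's New" entry of the Editor Debug Menu, showing a docs-icon button beside a V1.0.2 change log label.
# On Construct it creates an EDM_ChangeLogTooltip widget, stores it in the Tooltip variable and sets it as its own tooltip.
# Clicking the button launches a hardcoded Google Docs URL.
# The Tick event is empty.
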EditorUtilityWidgetBlueprint /EditorDebugMenu/EDM_WhatsNew
graphs [EventGraph {nodes [
   "On Clicked (Btn_OpenUrl)"() {is=title K2Node_ComponentBoundEvent}
   > LaunchURL!(execute, URL:String="https://docs.google.com/document/d/1-n4aZ7VlNW6WO4xyi-NLCcCffhm1mh-XG9OyVmUPdgM/edit##heading=h.fiagqi164wyk") -> (then)
   Event Construct()
   > n1 = Create Widget!(
    execute:Exec
    WorldContextObject:Object Reference
    Class:User Widget Class Reference=&/EDM_ChangeLogTooltip.EDM_ChangeLogTooltip_C
    OwningPlayer:Player Controller Object Reference
   ) -> (then:Exec, ReturnValue:EDM Change Log Tooltip Object Reference)
   > n2 = Set Tooltip(Tooltip:EDM Change Log Tooltip Object Reference=@n1) -> (Output_Get:EDM Change Log Tooltip Object Reference)
   > SetToolTip!(execute, self:Widget Object Reference, Widget:Widget Object Reference=@n2) -> (then)
   Event Tick()
  ]}]
variables [Tooltip {type=EDM_ChangeLogTooltip_C, default=null}]
widgets {
 info {rootWidgetClass=&StackBox, widgetCount=3}
 widgets [
  StackBox_28 {widgetClassPath=&StackBox, widget=StackBox_28, properties {visibility=Visible}}
  Btn_OpenUrl {
   widgetClassPath=&EditorUtilityButton
   parent=StackBox_28
   widget=Btn_OpenUrl
   properties {
    bOverride_Cursor=true
    cursor=Hand
    widgetStyle {
     clickedSlateSound {resourceObject=null}
     disabled {
      bIsDynamicallyLoaded=false
      drawAs=RoundedBox
      imageType=NoImage
      mirroring=NoMirror
      resourceName=None
      resourceObject=null
      tiling=NoTile
      imageSize=<0 0>
      margin=<ltrb 0 0 0 0>
      outlineSettings {
       bUseBrushTransparency=false
       roundingType=FixedRadius
       width=1
       color {colorUseRule=UseColor_ColorTable, specifiedColor=<rgba 0 0 0 0>}
       cornerRadii=<4 4 4 4>
      }
      tintColor {colorUseRule=UseColor_Specified, specifiedColor=<rgba 0.0395 0.0395 0.0395 1>}
      uVRegion {bIsValid=false, max=<0 0>, min=<0 0>}
     }
     disabledForeground {colorUseRule=UseColor_ColorTable, specifiedColor=<rgba 0 0 0 0>}
     hovered {
      bIsDynamicallyLoaded=false
      drawAs=Image
      imageType=FullColor
      mirroring=NoMirror
      resourceName=None
      resourceObject=/EditorDebugMenu/Textures/T_Docs
      tiling=NoTile
      imageSize=<25 25>
      margin=<ltrb 0 0 0 0>
      outlineSettings {
       bUseBrushTransparency=false
       roundingType=FixedRadius
       width=1
       color {colorUseRule=UseColor_ColorTable, specifiedColor=<rgba 0 0 0 0>}
       cornerRadii=<4 4 4 4>
      }
      tintColor {colorUseRule=UseColor_Specified, specifiedColor=<rgba 1 1 1 1>}
      uVRegion {bIsValid=false, max=<0 0>, min=<0 0>}
     }
     hoveredForeground {colorUseRule=UseColor_ColorTable, specifiedColor=<rgba 0 0 0 0>}
     hoveredSlateSound {resourceObject=null}
     normal {
      bIsDynamicallyLoaded=false
      drawAs=Image
      imageType=FullColor
      mirroring=NoMirror
      resourceName=None
      resourceObject=&T_Docs
      tiling=NoTile
      imageSize=<25 25>
      margin=<ltrb 0 0 0 0>
      outlineSettings {
       bUseBrushTransparency=false
       roundingType=FixedRadius
       width=1
       color {colorUseRule=UseColor_ColorTable, specifiedColor=<rgba 0 0 0 0>}
       cornerRadii=<4 4 4 4>
      }
      tintColor {colorUseRule=UseColor_Specified, specifiedColor=<rgba 0.8281 0.8281 0.8281 1>}
      uVRegion {bIsValid=false, max=<0 0>, min=<0 0>}
     }
     normalForeground {colorUseRule=UseColor_ColorTable, specifiedColor=<rgba 0 0 0 0>}
     normalPadding=<ltrb 12 1.5 12 1.5>
     pressed {
      bIsDynamicallyLoaded=false
      drawAs=Image
      imageType=FullColor
      mirroring=NoMirror
      resourceName=None
      resourceObject=&T_Docs
      tiling=NoTile
      imageSize=<25 25>
      margin=<ltrb 0 0 0 0>
      outlineSettings {
       bUseBrushTransparency=false
       roundingType=FixedRadius
       width=1
       color {colorUseRule=UseColor_ColorTable, specifiedColor=<rgba 0 0 0 0>}
       cornerRadii=<4 4 4 4>
      }
      tintColor {colorUseRule=UseColor_Specified, specifiedColor=<rgba 1 1 1 1>}
      uVRegion {bIsValid=false, max=<0 0>, min=<0 0>}
     }
     pressedForeground {colorUseRule=UseColor_ColorTable, specifiedColor=<rgba 0 0 0 0>}
     pressedPadding=<ltrb 12 2.5 12 0.5>
     pressedSlateSound {resourceObject=null}
    }
   }
   slot {class=StackBoxSlot, properties {horizontalAlignment=HAlign_Center, verticalAlignment=VAlign_Center}}
  }
  Txt_ExpandAll_1 {
   widgetClassPath=&UEFN_TextBlock
   parent=StackBox_28
   widget=Txt_ExpandAll_1
   properties {
    text="[V1.0.2 Change Log]"
    font {
     bForceMonospaced=false
     bMaterialIsStencil=false
     fontMaterial=null
     fontObject=&NotoSans
     letterSpacing=0
     monospacedWidth=1
     size=7.5
     skewAmount=0
     typefaceFontName=NotoSans-Regular
     outlineSettings {
      bApplyOutlineToDropShadows=false
      bMiteredCorners=false
      bSeparateFillAlpha=false
      outlineMaterial=null
      outlineSize=0
      outlineColor=<rgba 0 0 0 1>
     }
    }
   }
   slot {class=StackBoxSlot, properties {verticalAlignment=VAlign_Center}}
  }
 ]
}
