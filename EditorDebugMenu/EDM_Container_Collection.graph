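# Editor Utility Widget for one named, expandable collection of debug options inside the Editor Debug Menu.
# Add and Remove functions edit its set of data objects, then Refresh Widget rebuilds child widgets using classes the menu picks by debug type.
# Run All triggers every data object in the collection.
# Expansion state persists through the menu's config under a "Collection_" key, and the title follows the menu's text style changes.
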EditorUtilityWidgetBlueprint /EditorDebugMenu/EDM_Container_Collection
graphs [
 Add To Collection {nodes [
   n1 = Add To Collection!() -> (then, NewItem:"Obj EDM Input Data Object Reference (by ref)") {is=title K2Node_FunctionEntry}
   > Add!(
    execute
    TargetSet:Set of Obj EDM Input Data Object References=Get Data Objects() -> (Data Objects:Set of Obj EDM Input Data Object References)
    NewItem:"Obj EDM Input Data Object Reference (by ref)"=@n1
   ) -> (then) {is=title K2Node_CallFunction}
   > Refresh Widget()
  ]}
 EventGraph {nodes [
   "On Clicked (RunAllButton)"() {is=title K2Node_ComponentBoundEvent}
   > n1 = To Array!(execute, A:Set of Obj EDM Input Data Object References=Get Data Objects()) -> (then, Result:Array of Obj EDM Input Data Object References)
   > n2 = For Each Loop(Array:Array of Obj EDM Input Data Object References=@n1) -> (Array Element:Obj EDM Input Data Object Reference)
   > Trigger From Data(self:Obj EDM Input Data Object Reference=@n2)
   n3 = "On Expansion Changed (Expander)"!() -> (OutputDelegate, then, Area:Expandable Area Object Reference, bIsExpanded:bool) {is=title K2Node_ComponentBoundEvent}
   > Branch(Condition=Get ReadyToSave() -> (ReadyToSave:bool))
   > Update Config!(
    execute
    self:Editor Debug Menu Object Reference=Get Debug Menu() -> (Debug Menu:Editor Debug Menu Object Reference)
    Category Name:String=Append(A=Collection_, B=Get Collection Name() -> (Collection Name:String))
    Saved State:bool=@n3.bIsExpanded
   ) -> (then)
   Refresh Widget() {is=title K2Node_CustomEvent}
   > ClearChildren(self=Get CollectionWidgetStackBox() -> (CollectionWidgetStackBox:Stack Box Object Reference))
   > n4 = To Array(A=Get Data Objects())
   > n5 = For Each Loop(Array=@n4)
   n6 = GetData!(self:Obj EDM Input Data Object Reference=@n5) -> (Input Data:Struct EDM Input Data 01 Structure)
   > n7 = Create Widget!(
    execute:Exec
    WorldContextObject:Object Reference
    Class:User Widget Class Reference=Get Widget From Debug Type!(
     self:Editor Debug Menu Object Reference=Get Debug Menu()
     Debug Option:"Enum_EDM_InputType Enum (by ref)"=Break Struct EDM Input Data 01!(Struct_EDM_InputData_01:"Struct EDM Input Data 01 Structure (by ref)"=@n6) -> (
      DebugType_38_#1:Enum_EDM_InputType Enum
      TargetWorld_45_#2:Enum_EDM_WorldType Enum
      EditorUtilityObject_48_#3:Editor Utility Object Soft Class Reference
     ) {is=title K2Node_BreakStruct}
    ) -> (Widget:EDM Input Base Class Reference) {is=title K2Node_CallFunction}
    OwningPlayer:Player Controller Object Reference
    "Input  Data":Obj EDM Input Data Object Reference=@n5
    Debug Menu - Widget:Editor Debug Menu Object Reference=Get Debug Menu()
   ) -> (then:Exec, ReturnValue:EDM Input Base Object Reference)
   > Add(
    TargetArray:Array of EDM Input Base Object References=Get Widgets() -> (Widgets:Array of EDM Input Base Object References)
    NewItem:"EDM Input Base Object Reference (by ref)"=@n7
   ) {is=title K2Node_CallArrayFunction}
   > AddChildToStackBox!(execute, self:Stack Box Object Reference=Get CollectionWidgetStackBox(), Content:Widget Object Reference=@n7) -> (then, ReturnValue:Stack Box Slot Object Reference)
   n8 = Break Struct EDM Input Data 01!(Struct_EDM_InputData_01=Get CurrentData() -> (CurrentData:Struct EDM Input Data 01 Structure)) -> (
    Enabled_40_#4:bool=~
    ConsoleCommand_41_#5:String
    AdditionalConsoleCommands_42_#6:Array of Strings
    DevNotes_16_#7:String
    TargetWorld_45_#2
    EditorUtilityObject_48_#3
   )
   n9 = IsEmpty(InString=@n8.DevNotes_16_#7)
   n10 = Format Text!(Format="LOCTEXT(\"\\r\\n{text}\")", text:String=@n8.DevNotes_16_#7) -> (Result)
   > Set ToolTipText!(
    execute
    ToolTipText:Text=Format Text!(
     Format="LOCTEXT(\"Command: {command}{dev notes}\")"
     command:String=@n8.ConsoleCommand_41_#5
     dev notes:Text=SelectText(B=@n10, bPickA=@n9)
    ) -> (Result)
    self:Widget Object Reference=@n7
   ) -> (then, Output_Get:Text)
   n11 = Debug Option Changed!() -> (OutputDelegate, then, Text Size:int, Text Colour:Slate Color Structure, Button Colour:Slate Color Structure) {is=title K2Node_CustomEvent}
   n12 = Get GroupTitle() -> (GroupTitle:Text Block Object Reference)
   > n13 = Set members in Slate Font Info!(
    execute:Exec
    StructRef:"Slate Font Info Structure (by ref)"=Get Font!(self:Text Object Reference=@n12) -> (Font:Slate Font Info Structure)
    Size:float="To Float (Integer)"(InInt=@n11.Text Size)
   ) -> (then:Exec, StructOut:"Slate Font Info Structure (by ref)") {is=title K2Node_SetFieldsInStruct}
   > SetFont(self=@n12, InFontInfo=@n13)
   > SetColorAndOpacity(self=@n12, InColorAndOpacity=@n11.Text Colour)
   Bind Events() {is=title K2Node_CustomEvent}
   > Get!(execute:Exec) -> (then:Exec, else:Exec, Debug Menu:Editor Debug Menu Object Reference)
   > n14 = Sequence()
   Bind Event to Debug Option Changed(execute=@n14.then_1, self:Editor Debug Menu Object Reference=Get Debug Menu(), Delegate=@n11) {is=title K2Node_AddDelegate}
   Event PreConstruct()
   n15 = GetParent!(self:Widget Object Reference) -> (ReturnValue:Panel Widget Object Reference)
   > n16 = Cast To EUW_EditorDebugMenu!(execute, Object=GetOuterObject!(Object:Object Reference=GetOuterObject(Object=@n15)) -> (ReturnValue:Object Reference)) -> (then, CastFailed, AsEditor Debug Menu:Editor Debug Menu Object Reference) {is=title K2Node_DynamicCast}
   > Set Debug Menu(Debug Menu:Editor Debug Menu Object Reference=@n16) -> (Output_Get:Editor Debug Menu Object Reference)
   Event Construct()
   > Set ReadyToSave(ReadyToSave:bool=true) -> (Output_Get:bool)
   > DelayUntilNextTick!(execute, WorldContextObject:Object Reference) -> (then)
   > Bind Events() {is=title K2Node_CallFunction}
   > "SetText (Text)"(self=Get GroupTitle(), InText="To Text (String)"(InString=Get Collection Name()))
   n17 = Get Collection Name()
   > SetIsExpanded!(
    execute
    self:Expandable Area Object Reference=Get Expander() -> (Expander:Editor Utility Expandable Area Object Reference)
    IsExpanded:bool=Get Config!(self:Editor Debug Menu Object Reference=Get Debug Menu(), Category:String=Append(A=Collection_, B=@n17)) -> (State:bool) {is=title K2Node_CallFunction}
   ) -> (then)
   "On Clicked (DeleteCollectionButton)"() {is=title K2Node_ComponentBoundEvent}
  ]}
 Remove from collection {nodes [
   n1 = Remove from collection!() -> (then, Item:"Obj EDM Input Data Object Reference (by ref)") {is=title K2Node_FunctionEntry}
   n2 = Get Data Objects()
   > Remove!(execute, TargetSet:Set of Obj EDM Input Data Object References=@n2, Item:"Obj EDM Input Data Object Reference (by ref)"=@n1) -> (then, ReturnValue:bool)
   > Refresh Widget() {is=title K2Node_CallFunction}
   > Return Node!(
    execute
    Empty:bool=Is Empty!(TargetSet:Set of Obj EDM Input Data Object References=@n2) -> (ReturnValue:bool) {is=title K2Node_CallFunction}
   )
  ]}
]
variables [
 Collection Name {type=string, default=""}
 CurrentData {
  type=Struct_EDM_InputData_01
  default {
   name=None
   category=""
   console Command=""
   debug Type=Button
   dev Notes=""
   editor Task=null
   editor Utility Object=null
   enabled=true
   target World=Game World
   additional Console Commands []
   button Images []
   slider Data {default=1, end=1, intervals=0.1, start=0}
  }
 }
 Data Objects {type="array<Obj_EDM_InputData_C>", default []}
 Debug Menu {type=EUW_EditorDebugMenu_C, default=null}
 ReadyToSave {type=boolean}
 Should Be Expanded {type=boolean}
 Widgets {type="array<EDM_Input_Base_C>", default []}
]
widgets {
 info {namedSlotCount=2, rootWidgetClass=/Script/Blutility.EditorUtilityExpandableArea, widgetCount=10}
 widgets [
  Expander {
   widgetClassPath=&EditorUtilityExpandableArea
   bIsVariable=true
   widget=Expander
   properties {
    areaPadding=<ltrb 10 1 10 1>
    headerPadding=<ltrb 2 2 4 2>
    style {
     rolloutAnimationSeconds=0
     collapsedImage {
      bIsDynamicallyLoaded=false
      drawAs=Image
      imageType=Vector
      mirroring=NoMirror
      resourceName=../../../Engine/Content/Slate/Starship/Common/caret-right.svg
      resourceObject=null
      tiling=NoTile
      imageSize=<16 16>
      margin=<ltrb 0 0 0 0>
      outlineSettings {
       bUseBrushTransparency=false
       roundingType=HalfHeightRadius
       width=0
       color {colorUseRule=UseColor_Specified, specifiedColor=<rgba 0 0 0 0>}
       cornerRadii=<0 0 0 0>
      }
      tintColor {colorUseRule=UseColor_Specified, specifiedColor=<rgba 0.5271 0.5271 0.5271 1>}
      uVRegion {bIsValid=false, max=<0 0>, min=<0 0>}
     }
     expandedImage {
      bIsDynamicallyLoaded=false
      drawAs=Image
      imageType=Vector
      mirroring=NoMirror
      resourceName=../../../Engine/Content/Slate/Starship/Common/caret-down.svg
      resourceObject=null
      tiling=NoTile
      imageSize=<16 16>
      margin=<ltrb 0 0 0 0>
      outlineSettings {
       bUseBrushTransparency=false
       roundingType=HalfHeightRadius
       width=0
       color {colorUseRule=UseColor_Specified, specifiedColor=<rgba 0 0 0 0>}
       cornerRadii=<0 0 0 0>
      }
      tintColor {colorUseRule=UseColor_Specified, specifiedColor=<rgba 0.5271 0.5271 0.5271 1>}
      uVRegion {bIsValid=false, max=<0 0>, min=<0 0>}
     }
    }
   }
  }
  StackBox_0 {widgetClassPath=&StackBox, namedSlotHost=Expander, widget=StackBox_0}
  GroupTitle {
   widgetClassPath=&UEFN_TextBlock
   bIsVariable=true
   parent=StackBox_0
   widget=GroupTitle
   properties {
    text=Group
    font {
     bForceMonospaced=false
     bMaterialIsStencil=false
     fontMaterial=null
     fontObject=&NotoSans
     letterSpacing=0
     monospacedWidth=1
     size=8
     skewAmount=0
     typefaceFontName=NotoSans-Bold
     outlineSettings {
      bApplyOutlineToDropShadows=false
      bMiteredCorners=false
      bSeparateFillAlpha=false
      outlineMaterial=null
      outlineSize=0
      outlineColor=<rgba 0 0 0 1>
     }
    }
   }
   slot {class=StackBoxSlot, properties {size {sizeRule=Fill, value=1}}}
  }
  StackBox_46 {widgetClassPath=&StackBox, namedSlotHost=Expander, widget=StackBox_46, properties {orientation=Orient_Vertical}}
  Buttons {widgetClassPath=&StackBox, parent=StackBox_46, widget=Buttons, slot {class=StackBoxSlot}}
  RunAllButton {
   widgetClassPath=&EditorUtilityButton
   parent=Buttons
   widget=RunAllButton
   properties {widgetStyle {
     clickedSlateSound {resourceObject=null}
     disabled {
      bIsDynamicallyLoaded=false
      drawAs=RoundedBox
      imageType=NoImage
      mirroring=NoMirror
      resourceName=None
      resourceObject=null
      tiling=NoTile
      imageSize=<0 0>
      margin=<ltrb 0 0 0 0>
      outlineSettings {
       bUseBrushTransparency=false
       roundingType=FixedRadius
       width=1
       color {colorUseRule=UseColor_ColorTable, specifiedColor=<rgba 0 0 0 0>}
       cornerRadii=<4 4 4 4>
      }
      tintColor {colorUseRule=UseColor_Specified, specifiedColor=<rgba 0.0395 0.0395 0.0395 1>}
      uVRegion {bIsValid=false, max=<0 0>, min=<0 0>}
     }
     disabledForeground {colorUseRule=UseColor_ColorTable, specifiedColor=<rgba 0 0 0 0>}
     hovered {
      bIsDynamicallyLoaded=false
      drawAs=RoundedBox
      imageType=NoImage
      mirroring=NoMirror
      resourceName=None
      resourceObject=null
      tiling=NoTile
      imageSize=<32 32>
      margin=<ltrb 0 0 0 0>
      outlineSettings {
       bUseBrushTransparency=false
       roundingType=FixedRadius
       width=1
       color {colorUseRule=UseColor_ColorTable, specifiedColor=<rgba 0 0 0 0>}
       cornerRadii=<4 4 4 4>
      }
      tintColor {colorUseRule=UseColor_Specified, specifiedColor=<rgba 0.0953 0.0953 0.0953 1>}
      uVRegion {bIsValid=false, max=<0 0>, min=<0 0>}
     }
     hoveredForeground {colorUseRule=UseColor_ColorTable, specifiedColor=<rgba 0 0 0 0>}
     hoveredSlateSound {resourceObject=null}
     normal {
      bIsDynamicallyLoaded=false
      drawAs=RoundedBox
      imageType=NoImage
      mirroring=NoMirror
      resourceName=None
      resourceObject=null
      tiling=NoTile
      imageSize=<32 32>
      margin=<ltrb 0 0 0 0>
      outlineSettings {
       bUseBrushTransparency=false
       roundingType=FixedRadius
       width=1
       color {colorUseRule=UseColor_ColorTable, specifiedColor=<rgba 0 0 0 0>}
       cornerRadii=<4 4 4 4>
      }
      tintColor {colorUseRule=UseColor_Specified, specifiedColor=<rgba 0.0295 0.0295 0.0295 1>}
      uVRegion {bIsValid=false, max=<0 0>, min=<0 0>}
     }
     normalForeground {colorUseRule=UseColor_ColorTable, specifiedColor=<rgba 0 0 0 0>}
     normalPadding=<ltrb 12 1.5 12 1.5>
     pressed {
      bIsDynamicallyLoaded=false
      drawAs=RoundedBox
      imageType=NoImage
      mirroring=NoMirror
      resourceName=None
      resourceObject=null
      tiling=NoTile
      imageSize=<32 32>
      margin=<ltrb 0 0 0 0>
      outlineSettings {
       bUseBrushTransparency=false
       roundingType=FixedRadius
       width=1
       color {colorUseRule=UseColor_ColorTable, specifiedColor=<rgba 0 0 0 0>}
       cornerRadii=<4 4 4 4>
      }
      tintColor {colorUseRule=UseColor_Specified, specifiedColor=<rgba 0.0284 0.0284 0.0284 1>}
      uVRegion {bIsValid=false, max=<0 0>, min=<0 0>}
     }
     pressedForeground {colorUseRule=UseColor_ColorTable, specifiedColor=<rgba 0 0 0 0>}
     pressedPadding=<ltrb 12 2.5 12 0.5>
     pressedSlateSound {resourceObject=null}
    }}
   slot {class=StackBoxSlot, properties {horizontalAlignment=HAlign_Left, size {sizeRule=Automatic, value=0.35}}}
  }
  TRASH_Txt_ExpandAll_0 {
   widgetClassPath=&UEFN_TextBlock
   parent=RunAllButton
   widget=TRASH_Txt_ExpandAll_0
   properties {
    text=Run All
    colorAndOpacity {colorUseRule=UseColor_Specified, specifiedColor=<rgba 0.6562 0.6562 0.6562 1>}
    font {
     bForceMonospaced=false
     bMaterialIsStencil=false
     fontMaterial=null
     fontObject=&NotoSans
     letterSpacing=0
     monospacedWidth=1
     size=7.5
     skewAmount=0
     typefaceFontName=NotoSans-Italic
     outlineSettings {
      bApplyOutlineToDropShadows=false
      bMiteredCorners=false
      bSeparateFillAlpha=false
      outlineMaterial=null
      outlineSize=0
      outlineColor=<rgba 0 0 0 1>
     }
    }
   }
   slot {class=ButtonSlot, properties {padding=<ltrb 2 2 2 2>}}
  }
  DeleteCollectionButton {
   widgetClassPath=&EditorUtilityButton
   parent=Buttons
   widget=DeleteCollectionButton
   properties {
    visibility=Collapsed
    widgetStyle {
     clickedSlateSound {resourceObject=null}
     disabled {
      bIsDynamicallyLoaded=false
      drawAs=RoundedBox
      imageType=NoImage
      mirroring=NoMirror
      resourceName=None
      resourceObject=null
      tiling=NoTile
      imageSize=<0 0>
      margin=<ltrb 0 0 0 0>
      outlineSettings {
       bUseBrushTransparency=false
       roundingType=FixedRadius
       width=1
       color {colorUseRule=UseColor_ColorTable, specifiedColor=<rgba 0 0 0 0>}
       cornerRadii=<4 4 4 4>
      }
      tintColor {colorUseRule=UseColor_Specified, specifiedColor=<rgba 0.0395 0.0395 0.0395 1>}
      uVRegion {bIsValid=false, max=<0 0>, min=<0 0>}
     }
     disabledForeground {colorUseRule=UseColor_ColorTable, specifiedColor=<rgba 0 0 0 0>}
     hovered {
      bIsDynamicallyLoaded=false
      drawAs=RoundedBox
      imageType=NoImage
      mirroring=NoMirror
      resourceName=None
      resourceObject=null
      tiling=NoTile
      imageSize=<32 32>
      margin=<ltrb 0 0 0 0>
      outlineSettings {
       bUseBrushTransparency=false
       roundingType=FixedRadius
       width=1
       color {colorUseRule=UseColor_ColorTable, specifiedColor=<rgba 0 0 0 0>}
       cornerRadii=<4 4 4 4>
      }
      tintColor {colorUseRule=UseColor_Specified, specifiedColor=<rgba 0.0953 0.0953 0.0953 1>}
      uVRegion {bIsValid=false, max=<0 0>, min=<0 0>}
     }
     hoveredForeground {colorUseRule=UseColor_ColorTable, specifiedColor=<rgba 0 0 0 0>}
     hoveredSlateSound {resourceObject=null}
     normal {
      bIsDynamicallyLoaded=false
      drawAs=RoundedBox
      imageType=NoImage
      mirroring=NoMirror
      resourceName=None
      resourceObject=null
      tiling=NoTile
      imageSize=<32 32>
      margin=<ltrb 0 0 0 0>
      outlineSettings {
       bUseBrushTransparency=false
       roundingType=FixedRadius
       width=1
       color {colorUseRule=UseColor_ColorTable, specifiedColor=<rgba 0 0 0 0>}
       cornerRadii=<4 4 4 4>
      }
      tintColor {colorUseRule=UseColor_Specified, specifiedColor=<rgba 0.0295 0.0295 0.0295 1>}
      uVRegion {bIsValid=false, max=<0 0>, min=<0 0>}
     }
     normalForeground {colorUseRule=UseColor_ColorTable, specifiedColor=<rgba 0 0 0 0>}
     normalPadding=<ltrb 12 1.5 12 1.5>
     pressed {
      bIsDynamicallyLoaded=false
      drawAs=RoundedBox
      imageType=NoImage
      mirroring=NoMirror
      resourceName=None
      resourceObject=null
      tiling=NoTile
      imageSize=<32 32>
      margin=<ltrb 0 0 0 0>
      outlineSettings {
       bUseBrushTransparency=false
       roundingType=FixedRadius
       width=1
       color {colorUseRule=UseColor_ColorTable, specifiedColor=<rgba 0 0 0 0>}
       cornerRadii=<4 4 4 4>
      }
      tintColor {colorUseRule=UseColor_Specified, specifiedColor=<rgba 0.0284 0.0284 0.0284 1>}
      uVRegion {bIsValid=false, max=<0 0>, min=<0 0>}
     }
     pressedForeground {colorUseRule=UseColor_ColorTable, specifiedColor=<rgba 0 0 0 0>}
     pressedPadding=<ltrb 12 2.5 12 0.5>
     pressedSlateSound {resourceObject=null}
    }
   }
   slot {class=StackBoxSlot, properties {horizontalAlignment=HAlign_Left, size {sizeRule=Automatic, value=0.35}}}
  }
  TRASH_Txt_ExpandAll {
   widgetClassPath=&UEFN_TextBlock
   parent=DeleteCollectionButton
   widget=TRASH_Txt_ExpandAll
   properties {
    text=Run All
    colorAndOpacity {colorUseRule=UseColor_Specified, specifiedColor=<rgba 0.6562 0.6562 0.6562 1>}
    font {
     bForceMonospaced=false
     bMaterialIsStencil=false
     fontMaterial=null
     fontObject=&NotoSans
     letterSpacing=0
     monospacedWidth=1
     size=7.5
     skewAmount=0
     typefaceFontName=NotoSans-Italic
     outlineSettings {
      bApplyOutlineToDropShadows=false
      bMiteredCorners=false
      bSeparateFillAlpha=false
      outlineMaterial=null
      outlineSize=0
      outlineColor=<rgba 0 0 0 1>
     }
    }
   }
   slot {class=ButtonSlot, properties {padding=<ltrb 2 2 2 2>}}
  }
  CollectionWidgetStackBox {
   widgetClassPath=&StackBox
   bIsVariable=true
   parent=StackBox_46
   widget=CollectionWidgetStackBox
   properties {orientation=Orient_Vertical}
   slot {class=StackBoxSlot}
  }
 ]
}
