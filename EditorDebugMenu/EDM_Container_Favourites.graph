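# Editor Utility Widget for the Editor Debug Menu that shows favourited debug options in an expandable "Favourites" group, collapsed when empty.
# PreConstruct finds the owning menu through the parent's outer chain; Construct waits a tick, then binds menu events and populates.
# Population spawns an input widget per favourite, classed by debug type, with tooltips read from CurrentData, which this graph never sets.
# Option-change events restyle the title's font size and colour.
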EditorUtilityWidgetBlueprint /EditorDebugMenu/EDM_Container_Favourites
graphs [EventGraph {nodes [
   n1 = Debug Option Changed!() -> (OutputDelegate, then, Text Size:int, Text Colour:Slate Color Structure, Button Colour:Slate Color Structure) {is=title K2Node_CustomEvent}
   n2 = Get GroupTitle() -> (GroupTitle:Text Block Object Reference)
   > n3 = Set members in Slate Font Info!(
    execute:Exec
    StructRef:"Slate Font Info Structure (by ref)"=Get Font!(self:Text Object Reference=@n2) -> (Font:Slate Font Info Structure)
    Size:float="To Float (Integer)"(InInt=@n1.Text Size)
   ) -> (then:Exec, StructOut:"Slate Font Info Structure (by ref)") {is=title K2Node_SetFieldsInStruct}
   > SetFont(self=@n2, InFontInfo=@n3)
   > SetColorAndOpacity(self=@n2, InColorAndOpacity=@n1.Text Colour)
   Bind Events() {is=title K2Node_CustomEvent}
   > Get!(execute:Exec) -> (then:Exec, else:Exec, Debug Menu:Editor Debug Menu Object Reference)
   > n4 = Sequence()
   > Bind Event to Favourites Changed(self:Editor Debug Menu Object Reference=Get Debug Menu() -> (Debug Menu:Editor Debug Menu Object Reference), Delegate=Create Event()) {is=title K2Node_AddDelegate}
   Bind Event to Debug Option Changed(execute=@n4.then_1, self:Editor Debug Menu Object Reference=Get Debug Menu(), Delegate=@n1) {is=title K2Node_AddDelegate}
   Populate Favourites() {is=title K2Node_CustomEvent}
   > ClearChildren(self=Get GroupStackBox() -> (GroupStackBox:Stack Box Object Reference))
   n5 = Get Favourite Map!(self:Editor Debug Menu Object Reference=Get Debug Menu()) -> (Favourite Map:Set of Obj EDM Input Data Object References)
   > n6 = Branch(Condition=Is Empty!(TargetSet:Set of Obj EDM Input Data Object References=@n5) -> (ReturnValue:bool) {is=title K2Node_CallFunction})
   > SetVisibility(InVisibility=Collapsed)
   SetVisibility(execute=@n6.else, InVisibility=Visible)
   > n7 = To Array!(execute, A:Set of Obj EDM Input Data Object References=@n5) -> (then, Result:Array of Obj EDM Input Data Object References)
   > n8 = For Each Loop(Array:Array of Obj EDM Input Data Object References=@n7) -> (Array Element:Obj EDM Input Data Object Reference)
   n9 = GetData!(self:Obj EDM Input Data Object Reference=@n8) -> (Input Data:Struct EDM Input Data 01 Structure)
   > n10 = Create Widget!(
    execute:Exec
    WorldContextObject:Object Reference
    Class:User Widget Class Reference=Get Widget From Debug Type!(
     self:Editor Debug Menu Object Reference=Get Debug Menu()
     Debug Option:"Enum_EDM_InputType Enum (by ref)"=Break Struct EDM Input Data 01!(Struct_EDM_InputData_01:"Struct EDM Input Data 01 Structure (by ref)"=@n9) -> (
      DebugType_38_#1:Enum_EDM_InputType Enum
      TargetWorld_45_#2:Enum_EDM_WorldType Enum
      EditorUtilityObject_48_#3:Editor Utility Object Soft Class Reference
     ) {is=title K2Node_BreakStruct}
    ) -> (Widget:EDM Input Base Class Reference) {is=title K2Node_CallFunction}
    OwningPlayer:Player Controller Object Reference
    "Input  Data":Obj EDM Input Data Object Reference=@n8
    Debug Menu - Widget:Editor Debug Menu Object Reference=Get Debug Menu()
   ) -> (then:Exec, ReturnValue:EDM Input Base Object Reference)
   > Add(
    TargetArray:Array of EDM Input Base Object References=Get Widgets() -> (Widgets:Array of EDM Input Base Object References)
    NewItem:"EDM Input Base Object Reference (by ref)"=@n10
   ) {is=title K2Node_CallArrayFunction}
   > AddChildToStackBox!(execute, self:Stack Box Object Reference=Get GroupStackBox(), Content:Widget Object Reference=@n10) -> (then, ReturnValue:Stack Box Slot Object Reference)
   n11 = Break Struct EDM Input Data 01!(Struct_EDM_InputData_01=Get CurrentData() -> (CurrentData:Struct EDM Input Data 01 Structure)) -> (
    Enabled_40_#4:bool=~
    ConsoleCommand_41_#5:String
    AdditionalConsoleCommands_42_#6:Array of Strings
    DevNotes_16_#7:String
    TargetWorld_45_#2
    EditorUtilityObject_48_#3
   )
   n12 = IsEmpty(InString=@n11.DevNotes_16_#7)
   n13 = Format Text!(Format="LOCTEXT(\"\\r\\n{text}\")", text:String=@n11.DevNotes_16_#7) -> (Result)
   > Set ToolTipText!(
    execute
    ToolTipText:Text=Format Text!(
     Format="LOCTEXT(\"Command: {command}{dev notes}\")"
     command:String=@n11.ConsoleCommand_41_#5
     dev notes:Text=SelectText(B=@n13, bPickA=@n12)
    ) -> (Result)
    self:Widget Object Reference=@n10
   ) -> (then, Output_Get:Text)
   Event PreConstruct()
   n14 = GetParent!(self:Widget Object Reference) -> (ReturnValue:Panel Widget Object Reference)
   > n15 = Cast To EUW_EditorDebugMenu!(execute, Object=GetOuterObject!(Object:Object Reference=GetOuterObject(Object=@n14)) -> (ReturnValue:Object Reference)) -> (then, CastFailed, AsEditor Debug Menu:Editor Debug Menu Object Reference) {is=title K2Node_DynamicCast}
   > Set Debug Menu(Debug Menu:Editor Debug Menu Object Reference=@n15) -> (Output_Get:Editor Debug Menu Object Reference)
   Event Construct()
   > Set ReadyToSave(ReadyToSave:bool=true) -> (Output_Get:bool)
   > DelayUntilNextTick!(execute, WorldContextObject:Object Reference) -> (then)
   > Bind Events() {is=title K2Node_CallFunction}
   > Populate Favourites() {is=title K2Node_CallFunction}
  ]}]
variables [
 CurrentData {
  type=Struct_EDM_InputData_01
  default {
   name=None
   category=""
   console Command=""
   debug Type=Button
   dev Notes=""
   editor Task=null
   editor Utility Object=null
   enabled=true
   target World=Game World
   additional Console Commands []
   button Images []
   slider Data {default=1, end=1, intervals=0.1, start=0}
  }
 }
 Debug Menu {type=EUW_EditorDebugMenu_C, default=null}
 ReadyToSave {type=boolean}
 Should Be Expanded {type=boolean}
 Widgets {type="array<EDM_Input_Base_C>", default []}
]
widgets {
 info {namedSlotCount=2, rootWidgetClass=/Script/Blutility.EditorUtilityExpandableArea, widgetCount=4}
 widgets [
  Expander {
   widgetClassPath=&EditorUtilityExpandableArea
   bIsVariable=true
   widget=Expander
   properties {
    areaPadding=<ltrb 10 1 10 1>
    headerPadding=<ltrb 2 2 4 2>
    style {
     rolloutAnimationSeconds=0
     collapsedImage {
      bIsDynamicallyLoaded=false
      drawAs=Image
      imageType=Vector
      mirroring=NoMirror
      resourceName=../../../Engine/Content/Slate/Starship/Common/caret-right.svg
      resourceObject=null
      tiling=NoTile
      imageSize=<16 16>
      margin=<ltrb 0 0 0 0>
      outlineSettings {
       bUseBrushTransparency=false
       roundingType=HalfHeightRadius
       width=0
       color {colorUseRule=UseColor_Specified, specifiedColor=<rgba 0 0 0 0>}
       cornerRadii=<0 0 0 0>
      }
      tintColor {colorUseRule=UseColor_Specified, specifiedColor=<rgba 0.5271 0.5271 0.5271 1>}
      uVRegion {bIsValid=false, max=<0 0>, min=<0 0>}
     }
     expandedImage {
      bIsDynamicallyLoaded=false
      drawAs=Image
      imageType=Vector
      mirroring=NoMirror
      resourceName=../../../Engine/Content/Slate/Starship/Common/caret-down.svg
      resourceObject=null
      tiling=NoTile
      imageSize=<16 16>
      margin=<ltrb 0 0 0 0>
      outlineSettings {
       bUseBrushTransparency=false
       roundingType=HalfHeightRadius
       width=0
       color {colorUseRule=UseColor_Specified, specifiedColor=<rgba 0 0 0 0>}
       cornerRadii=<0 0 0 0>
      }
      tintColor {colorUseRule=UseColor_Specified, specifiedColor=<rgba 0.5271 0.5271 0.5271 1>}
      uVRegion {bIsValid=false, max=<0 0>, min=<0 0>}
     }
    }
   }
  }
  HorizontalBox_0 {widgetClassPath=&HorizontalBox, namedSlotHost=Expander, widget=HorizontalBox_0}
  GroupTitle {
   widgetClassPath=&UEFN_TextBlock
   bIsVariable=true
   parent=HorizontalBox_0
   widget=GroupTitle
   properties {
    text=Favourites
    font {
     bForceMonospaced=false
     bMaterialIsStencil=false
     fontMaterial=null
     fontObject=&NotoSans
     letterSpacing=0
     monospacedWidth=1
     size=8
     skewAmount=0
     typefaceFontName=NotoSans-Bold
     outlineSettings {
      bApplyOutlineToDropShadows=false
      bMiteredCorners=false
      bSeparateFillAlpha=false
      outlineMaterial=null
      outlineSize=0
      outlineColor=<rgba 0 0 0 1>
     }
    }
   }
   slot {class=HorizontalBoxSlot}
  }
  GroupStackBox {
   widgetClassPath=&StackBox
   bIsVariable=true
   namedSlotHost=Expander
   widget=GroupStackBox
   properties {orientation=Orient_Vertical}
  }
 ]
}
