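# Editor utility widget forming one collapsible category in the Editor Debug Menu.
# PreConstruct creates an input widget per category data entry, class chosen by debug type, with tooltips showing source, command, task and notes.
# Construct binds menu search, style and expand/collapse events.
# Searches expand the group a tick later, count visible entries and hide empty groups; empty searches restore saved expansion, which is only saved outside searches.
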EditorUtilityWidgetBlueprint /EditorDebugMenu/EDM_Container_Inputs
graphs [
 EventGraph {nodes [
   n1 = "On Expansion Changed (Expander)"!() -> (OutputDelegate, then, Area:Expandable Area Object Reference, bIsExpanded:bool) {is=title K2Node_ComponentBoundEvent}
   n2 = Get Search!(self:Editor Debug Menu Object Reference=Get Debug Menu() -> (Debug Menu:Editor Debug Menu Object Reference)) -> (Search:String)
   > Branch(Condition=AND Boolean(A=Get ReadyToSave() -> (ReadyToSave:bool), B=IsEmpty(InString=@n2)))
   > Update Config!(
    execute
    self:Editor Debug Menu Object Reference=Get Debug Menu()
    Category Name:String=Get Group Name!(self:Obj EDM Category Data Object Reference=Get Debug Data() -> (Debug Data:Obj EDM Category Data Object Reference)) -> (Group Name:String) {is=title K2Node_CallFunction}
    Saved State:bool=@n1.bIsExpanded
   ) -> (then)
   n3 = SearchUpdated!() -> (OutputDelegate, then, Search Term:String) {is=title K2Node_CustomEvent}
   > DelayUntilNextTick!(execute, WorldContextObject:Object Reference) -> (then)
   > n4 = Branch(Condition=IsEmpty(InString=@n3.Search Term))
   n5 = Get Debug Data()
   > n6 = Set Should Be Expanded(Should Be Expanded:bool=Get Config!(self:Editor Debug Menu Object Reference=Get Debug Menu(), Category:String=Get Group Name(self=@n5)) -> (State:bool) {is=title K2Node_CallFunction}) -> (Output_Get:bool)
   Update Expander(execute=[@n6.then, @n7.then])
   n7 = Set Should Be Expanded(execute=@n4.else, Should Be Expanded=true)
   n8 = Debug Option Changed!() -> (OutputDelegate, then, Text Size:int, Text Colour:Slate Color Structure, Button Colour:Slate Color Structure) {is=title K2Node_CustomEvent}
   n9 = Get GroupTitle() -> (GroupTitle:Text Block Object Reference)
   > n10 = Set members in Slate Font Info!(
    execute:Exec
    StructRef:"Slate Font Info Structure (by ref)"=Get Font!(self:Text Object Reference=@n9) -> (Font:Slate Font Info Structure)
    Size:float="To Float (Integer)"(InInt=@n8.Text Size)
   ) -> (then:Exec, StructOut:"Slate Font Info Structure (by ref)") {is=title K2Node_SetFieldsInStruct}
   > SetFont(self=@n9, InFontInfo=@n10)
   > SetColorAndOpacity(self=@n9, InColorAndOpacity=@n8.Text Colour)
   n11 = ViewUpdate!() -> (OutputDelegate, then, Command:String) {is=title K2Node_CustomEvent}
   > n12 = Switch on String!(execute:Exec, Selection:String=@n11.Command) -> (Expand:Exec, Collapse:Exec) {bHasDefaultPin=false, pinNames [Expand,Collapse]}
   n13 = Set Should Be Expanded(execute=@n12.Collapse, Should Be Expanded=false)
   Update Expander(execute=[@n13.then, @n14.then])
   n14 = Set Should Be Expanded(execute=@n12.Expand, Should Be Expanded=true)
   Bind Events() {is=title K2Node_CustomEvent}
   > Get!(execute:Exec) -> (then:Exec, else:Exec, Debug Menu:Editor Debug Menu Object Reference)
   > n15 = Sequence!(execute) -> (then_0, then_1, then_2:Exec, then_3:Exec)
   n16 = Get Debug Menu()
   > Bind Event to Search Updated(self:Editor Debug Menu Object Reference=@n16, Delegate=@n3) {is=title K2Node_AddDelegate}
   > SearchUpdated!(execute, self, Search Term:String=Get Search(self=@n16)) -> (then) {is=title K2Node_CallFunction}
   Bind Event to Debug Option Changed(execute=@n15.then_1, self:Editor Debug Menu Object Reference=Get Debug Menu(), Delegate=@n8) {is=title K2Node_AddDelegate}
   Bind Event to View Update(execute=@n15.then_2, self:Editor Debug Menu Object Reference=Get Debug Menu(), Delegate=@n11) {is=title K2Node_AddDelegate}
   Event PreConstruct()
   > n17 = Get!(execute) -> (then, else, Debug Data:Obj EDM Category Data Object Reference)
   > "SetText (Text)"(self=Get GroupTitle(), InText="To Text (String)"(InString=Get Group Name(self=@n17)))
   n18 = Get Project Name() -> (Project Name:String)
   > "SetText (Text)"(
    self=Get ProjectTitle() -> (ProjectTitle:Text Block Object Reference)
    InText="To Text (String)"(InString=Append!(A="[", B=@n18, C:String="]") -> (ReturnValue))
   )
   > n19 = For Each Loop(Array:Array of Obj EDM Input Data Object References=Get Data!(self:Obj EDM Category Data Object Reference=Get Debug Data()) -> (Group Data:Array of Obj EDM Input Data Object References) {is=title K2Node_CallFunction}) -> (Array Element:Obj EDM Input Data Object Reference)
   > n20 = Set CurrentData(CurrentData:Struct EDM Input Data 01 Structure=Get Input Data!(self:Obj EDM Input Data Object Reference=@n19) -> (Input Data:Struct EDM Input Data 01 Structure)) -> (Output_Get:Struct EDM Input Data 01 Structure)
   > n21 = Create Widget!(
    execute:Exec
    WorldContextObject:Object Reference
    Class:User Widget Class Reference=Get Widget From Debug Type!(
     self:Editor Debug Menu Object Reference=Get Debug Menu()
     Debug Option:"Enum_EDM_InputType Enum (by ref)"=Break Struct EDM Input Data 01!(Struct_EDM_InputData_01:"Struct EDM Input Data 01 Structure (by ref)"=@n20) -> (
      DebugType_38_#1:Enum_EDM_InputType Enum
      TargetWorld_45_#2:Enum_EDM_WorldType Enum
      EditorUtilityObject_48_#3:Editor Utility Object Soft Class Reference
     ) {is=title K2Node_BreakStruct}
    ) -> (Widget:EDM Input Base Class Reference) {is=title K2Node_CallFunction}
    OwningPlayer:Player Controller Object Reference
    "Input  Data":Obj EDM Input Data Object Reference=@n19
    Debug Menu - Widget:Editor Debug Menu Object Reference=Get Debug Menu()
   ) -> (then:Exec, ReturnValue:EDM Input Base Object Reference)
   > n22 = Set CurrentWidget(CurrentWidget:EDM Input Base Object Reference=@n21) -> (Output_Get:EDM Input Base Object Reference)
   > Add(
    TargetArray:Array of EDM Input Base Object References=Get Widgets() -> (Widgets:Array of EDM Input Base Object References)
    NewItem:"EDM Input Base Object Reference (by ref)"=@n22
   ) {is=title K2Node_CallArrayFunction}
   > AddChildToStackBox!(
    execute
    self:Stack Box Object Reference=Get GroupStackBox() -> (GroupStackBox:Stack Box Object Reference)
    Content:Widget Object Reference=@n22
   ) -> (then, ReturnValue:Stack Box Slot Object Reference)
   > Set Tooltip on widget!(
    execute
    self
    Widget:EDM Input Base Object Reference=@n22
    Source Loc:String=Get Source Location!(self:Obj EDM Input Data Object Reference=@n19) -> (Source Location:String)
   ) -> (then) {is=title K2Node_CallFunction}
   Event Construct()
   > Set ReadyToSave(ReadyToSave:bool=true) -> (Output_Get:bool)
   > Bind Events() {is=title K2Node_CallFunction}
  ]}
 Set Tooltip on widget {nodes [
   n1 = Set Tooltip on widget!() -> (then, Widget:EDM Input Base Object Reference, Source Loc:String) {is=title K2Node_FunctionEntry}
   n2 = Break Struct EDM Input Data 01!(Struct_EDM_InputData_01=Get CurrentData() -> (CurrentData:Struct EDM Input Data 01 Structure)) -> (
    Enabled_40_#4:bool=~
    ConsoleCommand_41_#5:String
    AdditionalConsoleCommands_42_#6:Array of Strings
    EditorTask_36_#7:EUT EDM Base Soft Class Reference
    DevNotes_16_#8:String
    TargetWorld_45_#2
    EditorUtilityObject_48_#3
   )
   n3 = BreakTopLevelAssetPath!(TopLevelAssetPath:"Top Level Asset Path Structure (by ref)"=GetSoftClassTopLevelAssetPath!(SoftClassReference:Object Soft Class Reference=@n2.EditorTask_36_#7) -> (ReturnValue:Top Level Asset Path Structure)) -> (PathString:String)
   n4 = IsEmpty(InString=@n3)
   n5 = Format Text!(Format="LOCTEXT(\"\\r\\nEUT: {text}\")", text:String=@n3) -> (Result)
   n6 = IsEmpty(InString=@n2.ConsoleCommand_41_#5)
   n7 = Format Text!(Format="LOCTEXT(\"CMD: {cmd}\")", cmd:String=@n2.ConsoleCommand_41_#5) -> (Result)
   n8 = IsEmpty(InString=@n2.DevNotes_16_#8)
   n9 = Format Text!(Format="LOCTEXT(\"\\r\\nDev Notes: {text}\")", text:String=@n2.DevNotes_16_#8) -> (Result)
   > Set ToolTipText!(
    execute
    ToolTipText:Text=Format Text!(
     Format="LOCTEXT(\"[{Source Loc}]\\r\\n{command}{EUT}{dev notes}\")"
     Source Loc:String=Split(SourceString=@n1.Source Loc, InStr=., SearchDir=FromEnd)
     command:Text=SelectText(B=@n7, bPickA=@n6)
     EUT:Text=SelectText(B=@n5, bPickA=@n4)
     dev notes:Text=SelectText(B=@n9, bPickA=@n8)
    ) -> (Result)
    self:Widget Object Reference=@n1
   ) -> (then, Output_Get:Text)
  ]}
 Update Expander {nodes [
   Update Expander() {is=title K2Node_FunctionEntry}
   n1 = Get Debug Menu()
   > Set Search String Is Empty(Search String Is Empty:bool=IsEmpty(InString=Get Search(self=@n1))) -> (Output_Get:bool)
   > SetIsExpanded!(
    execute
    self:Expandable Area Object Reference=Get Expander() -> (Expander:Editor Utility Expandable Area Object Reference)
    IsExpanded:bool=Get Should Be Expanded() -> (Should Be Expanded:bool)
   ) -> (then)
   > n2 = Sequence!(execute) -> (then_0, then_1)
   n3 = GetAllChildren(self=Get GroupStackBox())
   > Set All Entries(All Entries:int=Length(TargetArray:Array of Widget Object References=@n3)) -> (Output_Get:int=0)
   > n4 = Branch(Condition=Get Search String Is Empty() -> (Search String Is Empty:bool))
   > Set Visible Entries(Visible Entries:int=Get All Entries() -> (All Entries:int=0)) -> (Output_Get:int=0)
   n5 = For Each Loop(Exec=@n4.else, Array:Array of Widget Object References=@n3) -> (Array Element:Widget Object Reference)
   > Branch(Condition="Equal (Enum)"(
     A:ESlateVisibility Enum=GetVisibility!(self:Widget Object Reference=@n5) -> (ReturnValue:ESlateVisibility Enum=Visible)
     B:ESlateVisibility Enum=Visible
    ))
   > Increment Int(Value=Get Visible Entries() -> (Visible Entries:int=0))
   n6 = Get All Entries() -> (All Entries=0)
   n7 = Get Visible Entries() -> (Visible Entries=0)
   Set Text(
    execute=@n2.then_1
    Text=SelectText(
     A=Format Text!(Format) -> (Result)
     B=Format Text!(Format="LOCTEXT(\" - ({vis}/{all})\")", vis:int=@n7, all:int=@n6) -> (Result)
     bPickA=Get Search String Is Empty()
    )
    self=Get GroupNumber() -> (GroupNumber:Text Block Object Reference)
   )
   n8 = OR Boolean(A="integer > integer"(A=Get Visible Entries() -> (Visible Entries=0)), B=Get Search String Is Empty())
   > SetVisibility(InVisibility=Select(Option 0=Collapsed, Option 1=Visible, Index=@n8))
   SetRenderOpacity(self=Get TextHolder() -> (TextHolder:Horizontal Box Object Reference), InOpacity=SelectFloat(A=1, B=0.2, bPickA=@n8))
  ]}
 UpdateNumber {nodes [
   UpdateNumber!() -> (then, Empty:bool=~) {is=title K2Node_FunctionEntry}
   > n1 = Sequence()
   n2 = GetAllChildren(self=Get GroupStackBox())
   > Set All Entries(All Entries=Length(TargetArray=@n2)) -> (Output_Get=0)
   > n3 = Branch(Condition=Get Empty() -> (Empty:bool))
   > Set Visible Entries(Visible Entries=Get All Entries() -> (All Entries=0)) -> (Output_Get=0)
   n4 = For Each Loop(Exec=@n3.else, Array=@n2)
   > Branch(Condition="Equal (Enum)"(A=GetVisibility(self=@n4) -> (ReturnValue=Visible), B=Visible))
   > Increment Int(Value=Get Visible Entries() -> (Visible Entries=0))
   n5 = Get All Entries() -> (All Entries=0)
   n6 = Get Visible Entries() -> (Visible Entries=0)
   Set Text(
    execute=@n1.then_1
    Text=SelectText(
     A=Format Text!(Format) -> (Result)
     B=Format Text!(Format="LOCTEXT(\" - ({vis}/{all})\")", vis:int=@n6, all:int=@n5) -> (Result)
     bPickA=Get Empty()
    )
    self=Get GroupNumber()
   )
  ]}
]
variables [
 CurrentData {
  type=Struct_EDM_InputData_01
  default {
   name=None
   category=""
   console Command=""
   debug Type=Button
   dev Notes=""
   editor Task=null
   editor Utility Object=null
   enabled=true
   target World=Game World
   additional Console Commands []
   button Images []
   slider Data {default=1, end=1, intervals=0.1, start=0}
  }
 }
 CurrentWidget {type=EDM_Input_Base_C, default=null}
 Debug Data {type=Obj_EDM_CategoryData_C, default=null}
 Debug Menu {type=EUW_EditorDebugMenu_C, default=null}
 Project Name {type=string, default=""}
 ReadyToSave {type=boolean}
 Should Be Expanded {type=boolean}
 Widgets {type="array<EDM_Input_Base_C>", default []}
]
widgets {
 info {namedSlotCount=2, rootWidgetClass=/Script/Blutility.EditorUtilityExpandableArea, widgetCount=6}
 widgets [
  Expander {
   widgetClassPath=&EditorUtilityExpandableArea
   bIsVariable=true
   widget=Expander
   properties {
    areaPadding=<ltrb 10 1 10 1>
    headerPadding=<ltrb 2 2 4 2>
    style {
     rolloutAnimationSeconds=0
     collapsedImage {
      bIsDynamicallyLoaded=false
      drawAs=Image
      imageType=Vector
      mirroring=NoMirror
      resourceName=../../../Engine/Content/Slate/Starship/Common/caret-right.svg
      resourceObject=null
      tiling=NoTile
      imageSize=<16 16>
      margin=<ltrb 0 0 0 0>
      outlineSettings {
       bUseBrushTransparency=false
       roundingType=HalfHeightRadius
       width=0
       color {colorUseRule=UseColor_Specified, specifiedColor=<rgba 0 0 0 0>}
       cornerRadii=<0 0 0 0>
      }
      tintColor {colorUseRule=UseColor_Specified, specifiedColor=<rgba 0.5271 0.5271 0.5271 1>}
      uVRegion {bIsValid=false, max=<0 0>, min=<0 0>}
     }
     expandedImage {
      bIsDynamicallyLoaded=false
      drawAs=Image
      imageType=Vector
      mirroring=NoMirror
      resourceName=../../../Engine/Content/Slate/Starship/Common/caret-down.svg
      resourceObject=null
      tiling=NoTile
      imageSize=<16 16>
      margin=<ltrb 0 0 0 0>
      outlineSettings {
       bUseBrushTransparency=false
       roundingType=HalfHeightRadius
       width=0
       color {colorUseRule=UseColor_Specified, specifiedColor=<rgba 0 0 0 0>}
       cornerRadii=<0 0 0 0>
      }
      tintColor {colorUseRule=UseColor_Specified, specifiedColor=<rgba 0.5271 0.5271 0.5271 1>}
      uVRegion {bIsValid=false, max=<0 0>, min=<0 0>}
     }
    }
   }
  }
  TextHolder {widgetClassPath=&HorizontalBox, bIsVariable=true, namedSlotHost=Expander, widget=TextHolder}
  GroupTitle {
   widgetClassPath=&UEFN_TextBlock
   bIsVariable=true
   parent=TextHolder
   widget=GroupTitle
   properties {
    text=Example
    font {
     bForceMonospaced=false
     bMaterialIsStencil=false
     fontMaterial=null
     fontObject=&NotoSans
     letterSpacing=0
     monospacedWidth=1
     size=8
     skewAmount=0
     typefaceFontName=NotoSans-Bold
     outlineSettings {
      bApplyOutlineToDropShadows=false
      bMiteredCorners=false
      bSeparateFillAlpha=false
      outlineMaterial=null
      outlineSize=0
      outlineColor=<rgba 0 0 0 1>
     }
    }
   }
   slot {class=HorizontalBoxSlot}
  }
  GroupNumber {
   widgetClassPath=&UEFN_TextBlock
   bIsVariable=true
   parent=TextHolder
   widget=GroupNumber
   properties {
    text="- (1/10)"
    font {
     bForceMonospaced=false
     bMaterialIsStencil=false
     fontMaterial=null
     fontObject=&NotoSans
     letterSpacing=0
     monospacedWidth=1
     size=8
     skewAmount=0
     typefaceFontName=NotoSans-Regular
     outlineSettings {
      bApplyOutlineToDropShadows=false
      bMiteredCorners=false
      bSeparateFillAlpha=false
      outlineMaterial=null
      outlineSize=0
      outlineColor=<rgba 0 0 0 1>
     }
    }
   }
   slot {class=HorizontalBoxSlot, properties {padding=<ltrb 5 0 0 0>}}
  }
  ProjectTitle {
   widgetClassPath=&UEFN_TextBlock
   bIsVariable=true
   parent=TextHolder
   widget=ProjectTitle
   properties {
    text="[Fortnite]"
    visibility=Collapsed
    colorAndOpacity {colorUseRule=UseColor_Specified, specifiedColor=<rgba 0.2135 0.2135 0.2135 1>}
    font {
     bForceMonospaced=false
     bMaterialIsStencil=false
     fontMaterial=null
     fontObject=&NotoSans
     letterSpacing=0
     monospacedWidth=1
     size=7.5
     skewAmount=0
     typefaceFontName=NotoSans-Italic
     outlineSettings {
      bApplyOutlineToDropShadows=false
      bMiteredCorners=false
      bSeparateFillAlpha=false
      outlineMaterial=null
      outlineSize=0
      outlineColor=<rgba 0 0 0 1>
     }
    }
   }
   slot {
    class=HorizontalBoxSlot
    properties {horizontalAlignment=HAlign_Right, verticalAlignment=VAlign_Center, padding=<ltrb 5 0 5 0>, size {sizeRule=Fill, value=1}}
   }
  }
  GroupStackBox {
   widgetClassPath=&StackBox
   bIsVariable=true
   namedSlotHost=Expander
   widget=GroupStackBox
   properties {orientation=Orient_Vertical}
  }
 ]
}
