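# Editor utility widget for the Editor Debug Menu that shows one project as a collapsible section with a title.
# On construct it creates an EDM_Container_Inputs child for each category in the project data and restores its saved expanded state.
# It binds to menu and project events for title text size, visibility, expand/collapse commands and search, which forces expansion.
# Expansion changes save to config only after construct and while search is empty.
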EditorUtilityWidgetBlueprint /EditorDebugMenu/EDM_Container_Project
graphs [
 EventGraph {nodes [
   n1 = "On Expansion Changed (Expander)"!() -> (OutputDelegate, then, Area:Expandable Area Object Reference, bIsExpanded:bool) {is=title K2Node_ComponentBoundEvent}
   n2 = Get Search!(self:Editor Debug Menu Object Reference=Get Debug Menu() -> (Debug Menu:Editor Debug Menu Object Reference)) -> (Search:String)
   > Branch(Condition=AND Boolean(A=Get ReadyToSave() -> (ReadyToSave:bool), B=IsEmpty(InString=@n2)))
   > Update Config!(
    execute
    self:Editor Debug Menu Object Reference=Get Debug Menu()
    Category Name:String=Append(A=Project_, B=Get Project Name() -> (Project Name:String))
    Saved State:bool=@n1.bIsExpanded
   ) -> (then)
   Event Construct()
   > Set ReadyToSave(ReadyToSave:bool=true) -> (Output_Get:bool)
   > n3 = Sequence!(execute) -> (then_0, then_1, then_2:Exec)
   > n4 = Get!(execute:Exec) -> (then:Exec, else:Exec, Project Data:Obj Project Data Object Reference)
   n5 = Get Project Name()
   > "SetText (Text)"(self=Get ProjectTitle() -> (ProjectTitle:Text Block Object Reference), InText="To Text (String)"(InString=@n5))
   n6 = Get Data!(self:Obj Project Data Object Reference=@n4) -> (Data:Map of Strings to Obj EDM Category Data Object References)
   > n7 = Keys!(execute, TargetMap:Map of Strings to Obj EDM Category Data Object References=@n6) -> (then, Keys:Array of Strings)
   > n8 = For Each Loop(Array=@n7)
   > n9 = Create Widget!(
    execute:Exec
    WorldContextObject:Object Reference
    Class:User Widget Class Reference=&/EDM_Container_Inputs.EDM_Container_Inputs_C
    OwningPlayer:Player Controller Object Reference
    Debug Data:Obj EDM Category Data Object Reference=Find(TargetMap:Map of Strings to Obj EDM Category Data Object References=@n6, Key=@n8) -> (Value:Obj EDM Category Data Object Reference)
    Debug Menu:Editor Debug Menu Object Reference=Get Debug Menu()
    Project Name:String=Get Project Name()
   ) -> (then:Exec, ReturnValue:EDM Container Inputs Object Reference)
   > Add(
    TargetArray:Array of EDM Container Inputs Object References=Get Container Widgets() -> (Container Widgets:Array of EDM Container Inputs Object References)
    NewItem:"EDM Container Inputs Object Reference (by ref)"=@n9
   ) {is=title K2Node_CallArrayFunction}
   > AddChild(self=Get WidgetContainer() -> (WidgetContainer:Stack Box Object Reference), Content=@n9)
   n10 = Get Project Name()
   Update Expander!(
    execute=@n3.then_1
    self
    IsExpanded:bool=Get Config!(self:Editor Debug Menu Object Reference=Get Debug Menu(), Category:String=Append(A=Project_, B=@n10)) -> (State:bool) {is=title K2Node_CallFunction}
   ) -> (then)
   K2Node_Composite_0 = Bindings(execute=@n3.then_2) {is=title K2Node_Composite}
   ToUpper!(SourceString:String=Append!(A="[", B=@n5, C:String="]") -> (ReturnValue)) -> (ReturnValue:String)
  ]}
 Bindings {
  path=EventGraph.K2Node_Composite_0.Bindings
  nodes [
   n1 = Debug Options Updated!() -> (OutputDelegate, then, Text Size:int, Text Colour:Slate Color Structure, Button Colour:Slate Color Structure) {is=title K2Node_CustomEvent}
   n2 = Get ProjectTitle()
   > n3 = Set members in Slate Font Info!(
    execute:Exec
    StructRef:"Slate Font Info Structure (by ref)"=Get Font!(self:Text Object Reference=@n2) -> (Font:Slate Font Info Structure)
    Size:float="To Float (Integer)"(InInt=@n1.Text Size)
   ) -> (then:Exec, StructOut:"Slate Font Info Structure (by ref)") {is=title K2Node_SetFieldsInStruct}
   > SetFont(self=@n2, InFontInfo=@n3)
   n4 = Project Visibility Updated!() -> (OutputDelegate, then, New Visibility:bool=~) {is=title K2Node_CustomEvent}
   > SetVisibility(InVisibility=Select(Option 0=Collapsed, Option 1=Visible, Index=@n4.New Visibility))
   n5 = ViewUpdate!() -> (OutputDelegate, then, Command:String) {is=title K2Node_CustomEvent}
   > n6 = Switch on String!(execute:Exec, Selection:String=@n5.Command) -> (Expand:Exec, Collapse:Exec) {bHasDefaultPin=false, pinNames [Expand,Collapse]}
   Update Expander(execute=@n6.Collapse, IsExpanded=false)
   Update Expander(execute=@n6.Expand, IsExpanded=true)
   n7 = Search Updated!() -> (OutputDelegate, then, Search Term:String) {is=title K2Node_CustomEvent}
   > n8 = Branch(Condition=IsEmpty(InString=@n7.Search Term))
   n9 = Get Project Name()
   > Update Expander(IsExpanded=Get Config(self=Get Debug Menu(), Category=Append!(A=Project_, B=@n9) -> (ReturnValue)))
   Update Expander(execute=@n8.else, IsExpanded=true)
   Inputs()
   > n10 = Get!(execute) -> (then, else, Debug Menu:Editor Debug Menu Object Reference)
   > Bind Event to Debug Option Changed(self:Editor Debug Menu Object Reference=@n10, Delegate=@n1) {is=title K2Node_AddDelegate}
   > Bind Event to View Update(self:Editor Debug Menu Object Reference=@n10, Delegate=@n5) {is=title K2Node_AddDelegate}
   > Bind Event to Search Updated(self:Editor Debug Menu Object Reference=@n10, Delegate=@n7) {is=title K2Node_AddDelegate}
   > Search Updated!(execute, self, Search Term:String=Get Search(self=@n10)) -> (then) {is=title K2Node_CallFunction}
   n11 = Get Project Data() -> (Project Data:Obj Project Data Object Reference)
   > Bind Event to Project Visibility Updated(self:Obj Project Data Object Reference=@n11, Delegate=@n4) {is=title K2Node_AddDelegate}
   > Project Visibility Updated!(execute, self, New Visibility:bool=Get Visible!(self:Obj Project Data Object Reference=@n11) -> (Visible:bool)) -> (then) {is=title K2Node_CallFunction}
   Outputs()
  ]
 }
 Update Expander {nodes [
   n1 = Update Expander!() -> (then, IsExpanded:bool=~) {is=title K2Node_FunctionEntry}
   > SetIsExpanded!(
    execute
    self:Expandable Area Object Reference=Get Expander() -> (Expander:Editor Utility Expandable Area Object Reference)
    IsExpanded:bool=@n1
   ) -> (then)
  ]}
]
variables [
 Container Widgets {type="array<EDM_Container_Inputs_C>", default []}
 Debug Menu {type=EUW_EditorDebugMenu_C, default=null}
 Project Data {type=Obj_ProjectData_C, default=null}
 Project Name {type=string, default=""}
 ReadyToSave {type=boolean}
]
widgets {
 info {namedSlotCount=2, rootWidgetClass=&StackBox, widgetCount=6}
 widgets [
  StackBox_0 {widgetClassPath=&StackBox, widget=StackBox_0, properties {orientation=Orient_Vertical}}
  Expander {
   widgetClassPath=/Script/Blutility.EditorUtilityExpandableArea
   bIsVariable=true
   parent=StackBox_0
   widget=Expander
   properties {
    areaPadding=<ltrb 5 1 1 1>
    headerPadding=<ltrb 2 2 4 2>
    style {
     rolloutAnimationSeconds=0
     collapsedImage {
      bIsDynamicallyLoaded=false
      drawAs=Image
      imageType=Vector
      mirroring=NoMirror
      resourceName=../../../Engine/Content/Slate/Starship/Common/caret-right.svg
      resourceObject=null
      tiling=NoTile
      imageSize=<16 16>
      margin=<ltrb 0 0 0 0>
      outlineSettings {
       bUseBrushTransparency=false
       roundingType=HalfHeightRadius
       width=0
       color {colorUseRule=UseColor_Specified, specifiedColor=<rgba 0 0 0 0>}
       cornerRadii=<0 0 0 0>
      }
      tintColor {colorUseRule=UseColor_Specified, specifiedColor=<rgba 0.5271 0.5271 0.5271 1>}
      uVRegion {bIsValid=false, max=<0 0>, min=<0 0>}
     }
     expandedImage {
      bIsDynamicallyLoaded=false
      drawAs=Image
      imageType=Vector
      mirroring=NoMirror
      resourceName=../../../Engine/Content/Slate/Starship/Common/caret-down.svg
      resourceObject=null
      tiling=NoTile
      imageSize=<16 16>
      margin=<ltrb 0 0 0 0>
      outlineSettings {
       bUseBrushTransparency=false
       roundingType=HalfHeightRadius
       width=0
       color {colorUseRule=UseColor_Specified, specifiedColor=<rgba 0 0 0 0>}
       cornerRadii=<0 0 0 0>
      }
      tintColor {colorUseRule=UseColor_Specified, specifiedColor=<rgba 0.5271 0.5271 0.5271 1>}
      uVRegion {bIsValid=false, max=<0 0>, min=<0 0>}
     }
    }
   }
   slot {class=StackBoxSlot}
  }
  StackBox_52 {widgetClassPath=&StackBox, namedSlotHost=Expander, widget=StackBox_52}
  ProjectTitle {
   widgetClassPath=&UEFN_TextBlock
   bIsVariable=true
   parent=StackBox_52
   widget=ProjectTitle
   properties {
    text="[Fortnite]"
    colorAndOpacity {colorUseRule=UseColor_Specified, specifiedColor=<rgba 0.3542 0.3542 0.3542 1>}
    font {
     bForceMonospaced=false
     bMaterialIsStencil=false
     fontMaterial=null
     fontObject=&NotoSans
     letterSpacing=0
     monospacedWidth=1
     size=7.5
     skewAmount=0
     typefaceFontName=NotoSans-Italic
     outlineSettings {
      bApplyOutlineToDropShadows=false
      bMiteredCorners=false
      bSeparateFillAlpha=false
      outlineMaterial=null
      outlineSize=0
      outlineColor=<rgba 0 0 0 1>
     }
    }
   }
   slot {class=StackBoxSlot, properties {verticalAlignment=VAlign_Center, padding=<ltrb 0 2 2 2>}}
  }
  WidgetContainer {
   widgetClassPath=&StackBox
   bIsVariable=true
   namedSlotHost=Expander
   widget=WidgetContainer
   properties {orientation=Orient_Vertical}
  }
  Image_346 {
   widgetClassPath=&Image
   bIsVariable=true
   parent=StackBox_0
   widget=Image_346
   properties {
    brush {
     bIsDynamicallyLoaded=false
     drawAs=Image
     imageType=NoImage
     mirroring=NoMirror
     resourceName=None
     resourceObject=null
     tiling=NoTile
     imageSize=<32 1>
     margin=<ltrb 0 0 0 0>
     outlineSettings {
      bUseBrushTransparency=false
      roundingType=HalfHeightRadius
      width=0
      color {colorUseRule=UseColor_Specified, specifiedColor=<rgba 0 0 0 0>}
      cornerRadii=<0 0 0 0>
     }
     tintColor {colorUseRule=UseColor_Specified, specifiedColor=<rgba 1 1 1 1>}
     uVRegion {bIsValid=false, max=<0 0>, min=<0 0>}
    }
    colorAndOpacity=<rgba 0.0607 0.0677 0.0625 0.3>
   }
   slot {class=StackBoxSlot, properties {verticalAlignment=VAlign_Center, padding=<ltrb 2 0 2 0>, size {sizeRule=Automatic, value=0.5}}}
  }
 ]
}
