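# EDM_DebugMenuOptions is an editor utility widget that serves as the options panel for an editor debug menu.
# It exposes settings like button and text colours, text size, collection editing and user directories by binding a Details View to itself on Construct, filtered to selected categories.
# Property edits fire an Options Changed dispatcher with the property name and widget reference.
# A Reset Widget To Default function fires a separate Reset Widget dispatcher.
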EditorUtilityWidgetBlueprint /EditorDebugMenu/EDM_DebugMenuOptions
graphs [
 EventGraph {nodes [
   n1 = "On Property Changed (OptionsView)"!() -> (OutputDelegate, then, PropertyName:Name=None) {is=title K2Node_ComponentBoundEvent}
   > Call Options Changed!(
    execute:Exec
    self:EDM Debug Menu Options Object Reference
    Property Name:Name=@n1.PropertyName
    Widget:EDM Debug Menu Options Object Reference=Self-Reference()
   ) -> (then:Exec) {is=title K2Node_CallDelegate}
   Event Construct()
   > SetObject(self=Get OptionsView() -> (OptionsView:Details View Object Reference), NewObject=Self-Reference())
  ]}
 Options Changed {nodes [Options Changed!() -> (then, Property Name:Name, Widget:EDM Debug Menu Options Object Reference) {is=title K2Node_FunctionEntry}]}
 Reset Widget {nodes [Reset Widget() {is=title K2Node_FunctionEntry}]}
 Reset Widget To Default {nodes [
   Reset Widget To Default() {is=title K2Node_FunctionEntry}
   > Call Reset Widget!(execute:Exec, self:EDM Debug Menu Options Object Reference) -> (then:Exec) {is=title K2Node_CallDelegate}
  ]}
]
variables [
 Button Colour {type=SlateColor, default {colorUseRule=UseColor_Specified, specifiedColor=<rgba 0.0395 0.0395 0.0395 1>}}
 Edit Collections {type=boolean}
 Options Changed {default=()}
 Reset Widget {default=()}
 Text Colour {type=SlateColor, default {colorUseRule=UseColor_Specified, specifiedColor=<rgba 1 1 1 1>}}
 Text Size {type=integer, default=8}
 User Directories {type="array<string>", default []}
]
widgets {
 info {rootWidgetClass=&StackBox, widgetCount=2}
 widgets [
  StackBox_0 {widgetClassPath=&StackBox, widget=StackBox_0, properties {orientation=Orient_Vertical}}
  OptionsView {
   widgetClassPath=&DetailsView
   parent=StackBox_0
   widget=OptionsView
   properties {categoriesToShow [Debug Menu Options, User Directories, Debug Menu Options|Accessibility, DebugMenuOptions]}
   slot {class=StackBoxSlot}
  }
 ]
}
