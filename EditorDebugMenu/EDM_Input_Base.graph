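# Base editor utility widget for Editor Debug Menu inputs, with empty widget-sorting and setup hooks.
# On construct it binds to the menu's search and option events, then polls until Input Data is valid before setup.
# Search toggles visibility by matching category, name or console commands; option changes restyle widget arrays' text size and colours.
# Right-click copies the console command; middle-click syncs the Content Browser to the source asset.
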EditorUtilityWidgetBlueprint /EditorDebugMenu/EDM_Input_Base
graphs [
 EventGraph {nodes [
   n1 = Search Updated!() -> (OutputDelegate, then, Search Term:String) {is=title K2Node_CustomEvent}
   > Filter From Search!(execute, self, Search:String=@n1.Search Term) -> (then)
   n2 = DebugOptionChanged!() -> (OutputDelegate, then, Text Size:int, Text Colour:Slate Color Structure, Button Colour:Slate Color Structure) {is=title K2Node_CustomEvent}
   > n3 = Sequence()
   > Do Once()
   > SortWidgetsIntoArrays()
   Set Text Size(execute=@n3.then_1, Text Size:int=@n2.Text Size) -> (Output_Get:int=0)
   > Set Button Colour(Button Colour:Slate Color Structure=@n2.Button Colour) -> (Output_Get:Slate Color Structure)
   > Set Text Colour(Text Colour:Slate Color Structure=@n2.Text Colour) -> (Output_Get:Slate Color Structure)
   > UpdateAccessibilityOptions()
   Event Construct()
   > SortWidgetsIntoArrays()
   > Set Debug Menu - Ready(Debug Menu - Ready:bool=true) -> (Output_Get:bool)
   > n4 = Get!(execute:Exec) -> (then:Exec, else:Exec, Debug Menu - Widget:Editor Debug Menu Object Reference)
   > Bind Event to Search Updated(self:Editor Debug Menu Object Reference=@n4, Delegate=@n1) {is=title K2Node_AddDelegate}
   > Bind Event to Debug Option Changed(self:Editor Debug Menu Object Reference=@n4, Delegate=@n2) {is=title K2Node_AddDelegate}
   > n5 = Search Updated!(execute, self, Search Term:String=Get Search!(self:Editor Debug Menu Object Reference=@n4) -> (Search:String)) -> (then) {is=title K2Node_CallFunction}
   n6 = Get!(execute=[@n5.then, @n7.then]) -> (then, else, "Input  Data":Obj EDM Input Data Object Reference)
   > Setup Widget()
   n7 = Delay!(execute=@n6.else, WorldContextObject:Object Reference, Duration=0.2) -> (then)
   Setup Widget() {is=title K2Node_CustomEvent}
   Event PreConstruct()
  ]}
 Filter From Search {nodes [
   n1 = Filter From Search!() -> (then, Search:String) {is=title K2Node_FunctionEntry}
   > n2 = Branch(Condition=IsEmpty(InString=@n1))
   > SetVisibility(InVisibility=Visible)
   > Return Node()
   n3 = Sequence(execute=@n2.else)
   n4 = Break Struct EDM Input Data 01!(Struct_EDM_InputData_01:"Struct EDM Input Data 01 Structure (by ref)"=GetData!(self:Obj EDM Input Data Object Reference="Get Input  Data"() -> ("Input  Data":Obj EDM Input Data Object Reference)) -> (Input Data:Struct EDM Input Data 01 Structure)) -> (
    Category_8_#1:String
    Name_10_#2:Name
    ConsoleCommand_41_#3:String
    AdditionalConsoleCommands_42_#4:Array of Strings
   ) {is=title K2Node_BreakStruct}
   > Branch(Condition=Contains(SearchIn=@n4, Substring=Get Search!() -> (Search)))
   > SetVisibility(InVisibility=Visible)
   > Return Node()
   n5 = Set Search Array(execute=@n3.then_1, Search Array:Array of Strings=@n4.AdditionalConsoleCommands_42_#4) -> (Output_Get:Array of Strings)
   > Add(TargetArray=@n5, NewItem=@n4.ConsoleCommand_41_#3) {is=title K2Node_CallArrayFunction}
   > Add(TargetArray=@n5, NewItem="To String (Name)"(InName=@n4.Name_10_#2))
   > n6 = For Each Loop with Break!(Exec:Exec, Array:Array of Strings=Get Search Array() -> (Search Array:Array of Strings), Break:Exec=@n7.OutputPin) -> (LoopBody:Exec, Array Element:String, Array Index:int, Completed:Exec)
   SetVisibility(execute=@n6.Completed, InVisibility=Collapsed)
   Branch(execute=@n6.LoopBody, Condition=Contains(SearchIn=@n6, Substring=Get Search()))
   > Set Found(Found:bool=true) -> (Output_Get:bool)
   > SetVisibility(InVisibility=Visible)
   > Return Node()
   n7 = Reroute Node()
   Branch(Condition=Get Found() -> (Found:bool))
  ]}
 OnMouseButtonDown {nodes [
   n1 = OnMouseButtonDown!() -> (then, MyGeometry:Geometry Structure, MouseEvent:"Pointer Event Structure (by ref)") {is=title K2Node_FunctionEntry}
   > n2 = Branch(Condition=Is Mouse Button Down!(Input:"Pointer Event Structure (by ref)"=@n1.MouseEvent, MouseButton:Key Structure=RightMouseButton) -> (ReturnValue:bool))
   n3 = Break Struct EDM Input Data 01!(Struct_EDM_InputData_01=GetData(self="Get Input  Data"())) -> (
    Enabled_40_#5:bool=~
    DebugType_38_#6:Enum_EDM_InputType Enum
    ConsoleCommand_41_#3
    AdditionalConsoleCommands_42_#4
    SliderData_37_#7:Struct EDM Option Slider Data Structure
    EditorTask_36_#8:EUT EDM Base Soft Class Reference
    TargetWorld_45_#9:Enum_EDM_WorldType Enum
    EditorUtilityObject_48_#10:Editor Utility Object Soft Class Reference
   )
   > CopyStringToClipboard!(execute, InString:String=@n3.ConsoleCommand_41_#3) -> (then)
   > PrintString(InString=Append!(A="Copied: \"", B=@n3.ConsoleCommand_41_#3, C:String="\" to the clipboard") -> (ReturnValue))
   > Return Node!(execute, ReturnValue:Event Reply Structure=Handled())
   n4 = Branch(execute=@n2.else, Condition=Is Mouse Button Down(Input=@n1.MouseEvent, MouseButton=MiddleMouseButton))
   n5 = Get Source Location!(self:Obj EDM Input Data Object Reference="Get Input  Data"()) -> (Source Location:String)
   > SyncBrowserToObjects(AssetPaths=Make Array("[0]":String=@n5) -> (Array:Array of Strings))
   > Return Node(ReturnValue=Handled())
   Return Node(execute=@n4.else, ReturnValue=Unhandled())
   PrintText(InText=Format Text!(Format="LOCTEXT(\"Source Location: {loc}\")", loc:String=@n5) -> (Result))
   Get EditorAssetSubsystem() -> (ReturnValue:Editor Asset Subsystem Object Reference) {is=title K2Node_GetEditorSubsystem}
   Get LevelEditorSubsystem()
  ]}
 SortWidgetsIntoArrays {nodes [SortWidgetsIntoArrays() {is=title K2Node_FunctionEntry}]}
 UpdateAccessibilityOptions {nodes [
   UpdateAccessibilityOptions() {is=title K2Node_FunctionEntry}
   > n1 = Sequence!(execute) -> (then_0, then_1, then_2:Exec)
   > n2 = For Each Loop(Array:Array of Text Block Object References=Get Text Widgets() -> (Text Widgets:Array of Text Block Object References)) -> (Array Element:Text Block Object Reference)
   > Is Valid(InputObject=@n2)
   > n3 = Set members in Slate Font Info!(
    execute:Exec
    StructRef:"Slate Font Info Structure (by ref)"=Get Font!(self:Text Object Reference=@n2) -> (Font:Slate Font Info Structure)
    Size:float="To Float (Integer)"(InInt=Get Text Size() -> (Text Size:int=0))
   ) -> (then:Exec, StructOut:"Slate Font Info Structure (by ref)") {is=title K2Node_SetFieldsInStruct}
   > SetFont(self=@n2, InFontInfo=@n3)
   > Set ColorAndOpacity!(execute, ColorAndOpacity:Slate Color Structure=Get Text Colour() -> (Text Colour:Slate Color Structure), self:Text Object Reference=@n2) -> (then, Output_Get:Slate Color Structure)
   n4 = For Each Loop(
    Exec=@n1.then_1
    Array:Array of Editor Utility Button Object References=Get Button Widgets() -> (Button Widgets:Array of Editor Utility Button Object References)
   ) -> (Array Element:Editor Utility Button Object Reference)
   > Is Valid(InputObject=@n4)
   n5 = Get WidgetStyle!(self:Button Object Reference=@n4) -> (WidgetStyle:Button Style Structure)
   > n6 = Set members in Slate Brush!(
    execute:Exec
    StructRef:"Slate Brush Structure (by ref)"=Break Button Style!(ButtonStyle:"Button Style Structure (by ref)"=@n5) -> (Normal:Slate Brush Structure, ClickedSlateSound:Slate Sound Structure) {is=title K2Node_BreakStruct}
    TintColor:Slate Color Structure=Get Button Colour() -> (Button Colour:Slate Color Structure)
   ) -> (then:Exec, StructOut:"Slate Brush Structure (by ref)") {is=title K2Node_SetFieldsInStruct}
   > n7 = Set members in Button Style!(execute:Exec, StructRef:"Button Style Structure (by ref)"=@n5, Normal:Slate Brush Structure=@n6) -> (then:Exec, StructOut:"Button Style Structure (by ref)") {is=title K2Node_SetFieldsInStruct}
   > SetStyle!(execute, self:Button Object Reference=@n4, InStyle:"Button Style Structure (by ref)"=@n7) -> (then)
   n8 = For Each Loop(Exec=@n1.then_2, Array:Array of Border Object References=Get Border Widgets() -> (Border Widgets:Array of Border Object References)) -> (Array Element:Border Object Reference)
   > Is Valid(InputObject=@n8)
   > SetBrushColor(
    self=@n8
    InBrushColor=Get!() -> (Button Colour_SpecifiedColor:Linear Color Structure, Button Colour_ColorUseRule:ESlateColorStylingMode Enum)
   )
  ]}
]
variables [
 Border Widgets {type="array<Border>", default []}
 Button Colour {type=SlateColor, default {colorUseRule=UseColor_Specified, specifiedColor=<rgba 1 0 1 1>}}
 Button Widgets {type="array<EditorUtilityButton>", default []}
 Debug Menu - Ready {type=boolean}
 Debug Menu - Widget {type=EUW_EditorDebugMenu_C, default=null}
 "Input  Data" {type=Obj_EDM_InputData_C, default=null}
 Text Colour {type=SlateColor, default {colorUseRule=UseColor_Specified, specifiedColor=<rgba 1 0 1 1>}}
 Text Size {type=integer}
 Text Widgets {type="array<UEFN_TextBlock_C>", default []}
]
widgets {
 info {rootWidgetClass=&StackBox, widgetCount=1}
 widgets [MainStackBox {widgetClassPath=&StackBox, bIsVariable=true, widget=MainStackBox}]
}
