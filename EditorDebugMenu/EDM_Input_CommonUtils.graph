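# Editor Utility Widget adding a favourite button and collapsible collections text field to an EDM_Input_Base option in the editor debug menu.
# On Construct it finds that parent through GetParent and outer objects, then binds to its input data and menu events.
# The button toggles Favourite and is re-tinted; Edit Collections shows the field.
# Committed text updates the data's collections, which display comma-joined via a member string not cleared between rebuilds.
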EditorUtilityWidgetBlueprint /EditorDebugMenu/EDM_Input_CommonUtils
graphs [EventGraph {nodes [
   "On Clicked (Btn_Favourite)"() {is=title K2Node_ComponentBoundEvent}
   > Toggle Favourite(self:Obj EDM Input Data Object Reference="Get Input  Data"!(self:EDM Input Base Object Reference=Get Option Widget Parent() -> (Option Widget Parent:EDM Input Base Object Reference)) -> ("Input  Data":Obj EDM Input Data Object Reference))
   n1 = "OnTextCommitted (Editable Text) (CollectionText)"!() -> (OutputDelegate, then, Text:"Text (by ref)", CommitMethod:ETextCommit Enum=Default) {is=title K2Node_ComponentBoundEvent}
   > n2 = Switch on ETextCommit!(execute, Selection:ETextCommit Enum=@n1.CommitMethod) -> (Default:Exec, OnEnter:Exec, OnUserMovedFocus:Exec, OnCleared:Exec) {is=title K2Node_SwitchEnum}
   Update Collections from Input!(
    execute=[@n2.OnEnter, @n2.OnUserMovedFocus]
    self:Obj EDM Input Data Object Reference="Get Input  Data"(self=Get Option Widget Parent())
    Collections:String="To String (Text)"(InText=@n1.Text)
   ) -> (then)
   Update Normal Colour() {is=title K2Node_CustomEvent}
   n3 = Get WidgetStyle!(self:Button Object Reference=Get Btn_Favourite() -> (Btn_Favourite:Editor Utility Button Object Reference)) -> (WidgetStyle:Button Style Structure)
   n4 = Get Favourite!(self:Obj EDM Input Data Object Reference="Get Input  Data"(self=Get Option Widget Parent())) -> (Favourite:bool)
   > n5 = Set members in Slate Brush!(
    execute:Exec
    StructRef:"Slate Brush Structure (by ref)"=Break Button Style!(ButtonStyle:"Button Style Structure (by ref)"=@n3) -> (Normal:Slate Brush Structure, ClickedSlateSound:Slate Sound Structure) {is=title K2Node_BreakStruct}
    TintColor:Slate Color Structure=Make Slate Color(SpecifiedColor=Select(Option 0:Linear Color Structure=(R=0.0677,G=0.0677,B=0.0677,A=1), Option 1:Linear Color Structure=(R=1,G=0.8569,B=0.2023,A=1), Index=@n4) -> (ReturnValue:Linear Color Structure))
   ) -> (then:Exec, StructOut:"Slate Brush Structure (by ref)") {is=title K2Node_SetFieldsInStruct}
   > n6 = Set members in Button Style!(execute:Exec, StructRef:"Button Style Structure (by ref)"=@n3, Normal:Slate Brush Structure=@n5) -> (then:Exec, StructOut:"Button Style Structure (by ref)") {is=title K2Node_SetFieldsInStruct}
   > SetStyle!(execute, self:Button Object Reference=Get Btn_Favourite(), InStyle:"Button Style Structure (by ref)"=@n6) -> (then)
   n7 = Edit Collections!() -> (OutputDelegate, then, New Toggle:bool=~) {is=title K2Node_CustomEvent}
   > SetVisibility(
    self=Get CollectionContainer() -> (CollectionContainer:Border Object Reference)
    InVisibility=Select(Option 0:ESlateVisibility Enum=Collapsed, Option 1:ESlateVisibility Enum=Visible, Index=@n7.New Toggle) -> (ReturnValue:ESlateVisibility Enum)
   )
   n8 = Favourite Updated!() -> (OutputDelegate, then, Is Favourite:bool=~) {is=title K2Node_CustomEvent}
   > Update Normal Colour() {is=title K2Node_CallFunction}
   n9 = Update Collections String() {is=title K2Node_CustomEvent}
   n10 = Get In collections!(self:Obj EDM Input Data Object Reference="Get Input  Data"(self=Get Option Widget Parent())) -> (In collections:Set of Strings)
   > n11 = Branch(Condition=Is Empty!(TargetSet:Set of Strings=@n10) -> (ReturnValue:bool) {is=title K2Node_CallFunction})
   > "SetText (Editable Text)"!(
    execute
    self:Editable Text Object Reference=Get CollectionText() -> (CollectionText:Editor Utility Editable Text Object Reference)
    InText:Text
   ) -> (then)
   n12 = To Array!(execute=@n11.else, A:Set of Strings=@n10) -> (then, Result:Array of Strings)
   > n13 = For Each Loop(Array=@n12)
   "SetText (Editable Text)"(execute=@n13.Completed, self=Get CollectionText(), InText="To Text (String)"(InString=Get Collection Text() -> (Collection Text:String)))
   n14 = Get Collection Text()
   Set Collection Text(
    execute=@n13.LoopBody
    Collection Text:String=SelectString(A=@n13, B=Append!(A=@n14, B=", ", C:String=@n13) -> (ReturnValue), bPickA="Equal (Integer)"(A=@n13.Array Index))
   ) -> (Output_Get:String)
   Event Construct()
   n15 = GetParent!(self:Widget Object Reference) -> (ReturnValue:Panel Widget Object Reference)
   > n16 = Cast To EDM_Input_Base!(execute, Object=GetOuterObject!(Object:Object Reference=GetOuterObject(Object=@n15)) -> (ReturnValue:Object Reference)) -> (then, CastFailed, AsEDM Input Base:EDM Input Base Object Reference) {is=title K2Node_DynamicCast}
   > n17 = Set Option Widget Parent(Option Widget Parent:EDM Input Base Object Reference=@n16) -> (Output_Get:EDM Input Base Object Reference)
   n18 = "Get Input  Data"(self=@n17)
   > Bind Event to Favourite Updated(self:Obj EDM Input Data Object Reference=@n18, Delegate=@n8) {is=title K2Node_AddDelegate}
   > Favourite Updated!(execute, self, Is Favourite:bool=Get Favourite(self=@n18)) -> (then) {is=title K2Node_CallFunction}
   n19 = Get Debug Menu - Widget!(self:EDM Input Base Object Reference=@n17) -> (Debug Menu - Widget:Editor Debug Menu Object Reference)
   > Bind Event to Edit Collections Toggled(self:Editor Debug Menu Object Reference=@n19, Delegate=@n7) {is=title K2Node_AddDelegate}
   > Edit Collections!(execute, self, New Toggle:bool=Get Edit Collections!(self:Editor Debug Menu Object Reference=@n19) -> (Edit Collections:bool)) -> (then) {is=title K2Node_CallFunction}
   > Bind Event to Collections Updated(self:Obj EDM Input Data Object Reference="Get Input  Data"(self=Get Option Widget Parent()), Delegate=@n9) {is=title K2Node_AddDelegate}
   > Update Collections String() {is=title K2Node_CallFunction}
  ]}]
variables [
 Collection Text {type=string, default=""}
 Is Favourite {type=boolean}
 Option Widget Parent {type=EDM_Input_Base_C, default=null}
]
widgets {
 info {rootWidgetClass=&StackBox, widgetCount=4}
 widgets [
  StackBox_0 {widgetClassPath=&StackBox, widget=StackBox_0}
  Btn_Favourite {
   widgetClassPath=&EditorUtilityButton
   parent=StackBox_0
   widget=Btn_Favourite
   properties {widgetStyle {
     clickedSlateSound {resourceObject=null}
     disabled {
      bIsDynamicallyLoaded=false
      drawAs=Image
      imageType=NoImage
      mirroring=NoMirror
      resourceName=None
      resourceObject=/EditorDebugMenu/Textures/T_Favourite
      tiling=NoTile
      imageSize=<13 13>
      margin=<ltrb 0 0 0 0>
      outlineSettings {
       bUseBrushTransparency=false
       roundingType=FixedRadius
       width=1
       color {colorUseRule=UseColor_ColorTable, specifiedColor=<rgba 0 0 0 0>}
       cornerRadii=<4 4 4 4>
      }
      tintColor {colorUseRule=UseColor_Specified, specifiedColor=<rgba 0.0395 0.0395 0.0395 1>}
      uVRegion {bIsValid=false, max=<0 0>, min=<0 0>}
     }
     disabledForeground {colorUseRule=UseColor_ColorTable, specifiedColor=<rgba 0 0 0 0>}
     hovered {
      bIsDynamicallyLoaded=false
      drawAs=Image
      imageType=NoImage
      mirroring=NoMirror
      resourceName=None
      resourceObject=&T_Favourite
      tiling=NoTile
      imageSize=<13 13>
      margin=<ltrb 0 0 0 0>
      outlineSettings {
       bUseBrushTransparency=false
       roundingType=FixedRadius
       width=1
       color {colorUseRule=UseColor_ColorTable, specifiedColor=<rgba 0 0 0 0>}
       cornerRadii=<4 4 4 4>
      }
      tintColor {colorUseRule=UseColor_Specified, specifiedColor=<rgba 1 1 1 1>}
      uVRegion {bIsValid=false, max=<0 0>, min=<0 0>}
     }
     hoveredForeground {colorUseRule=UseColor_ColorTable, specifiedColor=<rgba 0 0 0 0>}
     hoveredSlateSound {resourceObject=null}
     normal {
      bIsDynamicallyLoaded=false
      drawAs=Image
      imageType=NoImage
      mirroring=NoMirror
      resourceName=None
      resourceObject=&T_Favourite
      tiling=NoTile
      imageSize=<13 13>
      margin=<ltrb 0 0 0 0>
      outlineSettings {
       bUseBrushTransparency=false
       roundingType=FixedRadius
       width=1
       color {colorUseRule=UseColor_ColorTable, specifiedColor=<rgba 0 0 0 0>}
       cornerRadii=<4 4 4 4>
      }
      tintColor {colorUseRule=UseColor_Specified, specifiedColor=<rgba 0.3594 0.3594 0.3594 1>}
      uVRegion {bIsValid=false, max=<0 0>, min=<0 0>}
     }
     normalForeground {colorUseRule=UseColor_ColorTable, specifiedColor=<rgba 0 0 0 0>}
     normalPadding=<ltrb 12 1.5 12 1.5>
     pressed {
      bIsDynamicallyLoaded=false
      drawAs=Image
      imageType=NoImage
      mirroring=NoMirror
      resourceName=None
      resourceObject=&T_Favourite
      tiling=NoTile
      imageSize=<13 13>
      margin=<ltrb 0 0 0 0>
      outlineSettings {
       bUseBrushTransparency=false
       roundingType=FixedRadius
       width=1
       color {colorUseRule=UseColor_ColorTable, specifiedColor=<rgba 0 0 0 0>}
       cornerRadii=<4 4 4 4>
      }
      tintColor {colorUseRule=UseColor_Specified, specifiedColor=<rgba 1 0.7828 0.3851 1>}
      uVRegion {bIsValid=false, max=<0 0>, min=<0 0>}
     }
     pressedForeground {colorUseRule=UseColor_ColorTable, specifiedColor=<rgba 0 0 0 0>}
     pressedPadding=<ltrb 12 2.5 12 0.5>
     pressedSlateSound {resourceObject=null}
    }}
   slot {class=StackBoxSlot, properties {horizontalAlignment=HAlign_Center, verticalAlignment=VAlign_Center}}
  }
  CollectionContainer {
   widgetClassPath=&Border
   bIsVariable=true
   parent=StackBox_0
   widget=CollectionContainer
   properties {visibility=Collapsed, brushColor=<rgba 0.0104 0.0104 0.0104 1>}
   slot {class=StackBoxSlot, properties {verticalAlignment=VAlign_Center, size {sizeRule=Fill, value=1}}}
  }
  CollectionText {
   widgetClassPath=/Script/Blutility.EditorUtilityEditableText
   bIsVariable=true
   parent=CollectionContainer
   widget=CollectionText
   properties {minimumDesiredWidth=100}
   slot {class=BorderSlot}
  }
 ]
}
