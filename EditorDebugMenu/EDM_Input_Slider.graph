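# Editor utility widget, derived from EDM_Input_Base, that shows a named slider row with value text and reset button for the editor debug menu.
# Setup reads the EDM input data object, derives step size from interval and range, then binds to its slider-updated event.
# Slider changes (once ready), button clicks and resets go to that object, whose event refreshes text and handle.
# UpdateValue also maps 0–1 into start–end; hover animates the label.
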EditorUtilityWidgetBlueprint /EditorDebugMenu/EDM_Input_Slider
animations [Hover {
  binding_count=1
  length_seconds=0.2
  object=&self:Hover
  sequence {
   bindings [Txt_Name {
     id=#1
     parent_id=null
     possessed_class=UEFN_TextBlock_C
     tracks [{
       type=MovieScene2DTransformTrack
       sections [{
         type=MovieScene2DTransformSection
         end_frame=4
         start_frame=0
         channels [
          Translation.X {type=MovieSceneScriptingFloatChannel, keys [[0,0],[4,5]]}
          Translation.Y {type=MovieSceneScriptingFloatChannel, const [0,2]}
          Angle {type=MovieSceneScriptingFloatChannel, const [0,2]}
          Scale.X {type=MovieSceneScriptingFloatChannel, const [1,2]}
          Scale.Y {type=MovieSceneScriptingFloatChannel, const [1,2]}
          Shear.X {type=MovieSceneScriptingFloatChannel, const [0,2]}
          Shear.Y {type=MovieSceneScriptingFloatChannel, const [0,2]}
         ]
        }]
      }]
    }]
   display_rate {denominator=1, numerator=20}
   tick_resolution {denominator=1, numerator=60000}
   tracks []
  }
 }]
graphs [
 EventGraph {nodes [
   n1 = "On Value Changed (Slider)"!() -> (OutputDelegate, then, Value:float) {is=title K2Node_ComponentBoundEvent}
   > Branch(Condition=Get Debug Menu - Ready() -> (Debug Menu - Ready:bool))
   > Update Slider!(
    execute
    self:Obj EDM Input Data Object Reference="Get Input  Data"() -> ("Input  Data":Obj EDM Input Data Object Reference)
    New Value:double=@n1.Value
   ) -> (then)
   "On Clicked (DebugButton)"() {is=title K2Node_ComponentBoundEvent}
   > Update Slider(
    self="Get Input  Data"()
    New Value=GetValue!(self:Slider Object Reference=Get Slider() -> (Slider:Editor Utility Slider Object Reference)) -> (ReturnValue:float)
   )
   "On Hovered (DebugButton)"() {is=title K2Node_ComponentBoundEvent}
   n2 = Get Hover() -> (Hover:Widget Animation Object Reference)
   > PlayAnimationForward(InAnimation=@n2)
   "On Unhovered (DebugButton)"() {is=title K2Node_ComponentBoundEvent}
   > PlayAnimationReverse(InAnimation=@n2)
   "Clicked (EDM_Input_Reset)"() {is=title K2Node_ComponentBoundEvent}
   > Reset Slider(self:Obj EDM Input Data Object Reference="Get Input  Data"())
   n3 = Slider Data Updated!() -> (OutputDelegate, then, Slider Value:double=~, Mapped Value:double=~) {is=title K2Node_CustomEvent}
   > "SetText (Text)"(
    self=Get Txt_Value() -> (Txt_Value:Text Block Object Reference)
    InText="To Text (Float)"(Value=@n3.Mapped Value, MinimumIntegralDigits=1, MaximumIntegralDigits=324, MinimumFractionalDigits=0, MaximumFractionalDigits=3)
   )
   > Set Value!(execute, Value:float=@n3.Slider Value, self:Slider Object Reference=Get Slider()) -> (then, Output_Get:float)
   Event Setup Widget() {is=title K2Node_Event}
   > n4 = Get!(execute:Exec) -> (then:Exec, else:Exec, "Input  Data":Obj EDM Input Data Object Reference)
   n5 = Break Struct EDM Input Data 01!(Struct_EDM_InputData_01:"Struct EDM Input Data 01 Structure (by ref)"=GetData!(self:Obj EDM Input Data Object Reference=@n4) -> (Input Data:Struct EDM Input Data 01 Structure)) -> (
    Name_10_#2:Name
    SliderData_37_#3:Struct EDM Option Slider Data Structure
    TargetWorld_45_#4:Enum_EDM_WorldType Enum
    EditorUtilityObject_48_#5:Editor Utility Object Soft Class Reference
   ) {is=title K2Node_BreakStruct}
   > "SetText (Text)"(self=Get Txt_Name() -> (Txt_Name:Text Block Object Reference), InText="To Text (Name)"!(InName:Name=@n5) -> (ReturnValue:Text))
   n6 = Break Struct EDM Option Slider Data!(Struct_EDM_Option_SliderData:"Struct EDM Option Slider Data Structure (by ref)"=@n5.SliderData_37_#3) -> (
    Start_2_#6:double=~
    End_4_#7:double=~
    Intervals_6_#8:double=~
   ) {is=title K2Node_BreakStruct}
   n7 = float - float(A=@n6.End_4_#7, B=@n6)
   > SetStepSize!(
    execute
    self:Slider Object Reference=Get Slider()
    InValue:float=float / float(A="Absolute (Float)"(A=@n6.Intervals_6_#8), B="Absolute (Float)"(A=@n7))
   ) -> (then)
   > Set Value(Value=Get Slider Mapped Value!(self:Obj EDM Input Data Object Reference=@n4) -> (Slider Mapped Value:double), self=Get Slider())
   n8 = "Get Input  Data"()
   > Bind Event to Slider Updated(self:Obj EDM Input Data Object Reference=@n8, Delegate=@n3) {is=title K2Node_AddDelegate}
   > Slider Data Updated!(
    execute
    self
    Slider Value:double=Get Slider Value!(self:Obj EDM Input Data Object Reference=@n8) -> (Slider Value:double)
    Mapped Value:double=Get Slider Mapped Value(self=@n8)
   ) -> (then) {is=title K2Node_CallFunction}
   Get Slider Value(self=@n4)
  ]}
 SortWidgetsIntoArrays {nodes [
   SortWidgetsIntoArrays() {is=title K2Node_FunctionEntry}
   > Parent: SortWidgetsIntoArrays() {is=title K2Node_CallParentFunction}
   n1 = Get Text Widgets() -> (Text Widgets:Array of Text Block Object References)
   > Add(TargetArray:Array of Text Block Object References=@n1, NewItem:"Text Block Object Reference (by ref)"=Get Txt_Name()) {is=title K2Node_CallArrayFunction}
   > Add(TargetArray=@n1, NewItem=Get Txt_Value())
   > Add(
    TargetArray:Array of Editor Utility Button Object References=Get Button Widgets() -> (Button Widgets:Array of Editor Utility Button Object References)
    NewItem:"Editor Utility Button Object Reference (by ref)"=Get DebugButton() -> (DebugButton:Editor Utility Button Object Reference)
   )
   Get Border Widgets() -> (Border Widgets:Array of Border Object References)
   Get DebugButton()
  ]}
 UpdateValue {nodes [
   UpdateValue() {is=title K2Node_FunctionEntry}
   n1 = GetValue(self=Get Slider())
   > Update Slider(self="Get Input  Data"(), New Value=@n1)
   n2 = "Get Input  Data"()
   n3 = Break Struct EDM Option Slider Data!(Struct_EDM_Option_SliderData=Break Struct EDM Input Data 01!(Struct_EDM_InputData_01=GetData(self=@n2)) -> (
     Enabled_40_#9:bool=~
     ConsoleCommand_41_#10:String
     AdditionalConsoleCommands_42_#11:Array of Strings
     SliderData_37_#3
     TargetWorld_45_#4
     EditorUtilityObject_48_#5
    ).SliderData_37_#3) -> (Start_2_#6=~, End_4_#7=~)
   n4 = Set MappedRange(MappedRange:double=MapRangeClamped(Value=@n1, InRangeA=0, InRangeB=1, OutRangeA=@n3, OutRangeB=@n3.End_4_#7)) -> (Output_Get:double)
   > "SetText (Text)"(
    self=Get Txt_Value()
    InText="To Text (Float)"(Value=@n4, MinimumIntegralDigits=1, MaximumIntegralDigits=324, MinimumFractionalDigits=0, MaximumFractionalDigits=3)
   )
  ]}
]
variables [
 MappedRange {type=number}
 Ready To Update {type=boolean}
 True or False {type=string, default=""}
]
widgets {
 info {parentClass=&/EDM_Input_Base.EDM_Input_Base_C, rootWidgetClass=&StackBox, widgetCount=9}
 widgets [
  StackBox_0 {widgetClassPath=&StackBox, widget=StackBox_0}
  EditorDebugMenu_CommonUtils {
   widgetClassPath=&/EDM_Input_CommonUtils
   bIsVariable=true
   parent=StackBox_0
   widget=EditorDebugMenu_CommonUtils
   slot {class=StackBoxSlot}
  }
  DebugButton {
   widgetClassPath=&EditorUtilityButton
   parent=StackBox_0
   widget=DebugButton
   slot {class=StackBoxSlot, properties {size {sizeRule=Fill, value=1}}}
  }
  ButtonContainer {
   widgetClassPath=&HorizontalBox
   parent=DebugButton
   widget=ButtonContainer
   slot {class=ButtonSlot, properties {horizontalAlignment=HAlign_Fill, padding=<ltrb 0 0 0 0>}}
  }
  Txt_Name {
   widgetClassPath=&UEFN_TextBlock
   bIsVariable=true
   parent=ButtonContainer
   widget=Txt_Name
   properties {
    text=Button Text
    font {
     bForceMonospaced=false
     bMaterialIsStencil=false
     fontMaterial=null
     fontObject=&NotoSans
     letterSpacing=0
     monospacedWidth=1
     size=8
     skewAmount=0
     typefaceFontName=NotoSans-Regular
     outlineSettings {
      bApplyOutlineToDropShadows=false
      bMiteredCorners=false
      bSeparateFillAlpha=false
      outlineMaterial=null
      outlineSize=0
      outlineColor=<rgba 0 0 0 1>
     }
    }
   }
   slot {class=HorizontalBoxSlot, properties {verticalAlignment=VAlign_Center, padding=<ltrb 2 2 2 2>, size {sizeRule=Fill, value=1}}}
  }
  HorizontalBox_0 {
   widgetClassPath=&HorizontalBox
   parent=ButtonContainer
   widget=HorizontalBox_0
   slot {class=HorizontalBoxSlot, properties {padding=<ltrb 8 0 8 0>, size {sizeRule=Fill, value=1}}}
  }
  Slider {
   widgetClassPath=/Script/Blutility.EditorUtilitySlider
   bIsVariable=true
   parent=HorizontalBox_0
   widget=Slider
   properties {
    mouseUsesStep=true
    value=0.5
    sliderBarColor=<rgba 1 0.1022 0.0724 1>
    sliderHandleColor=<rgba 0.0543 0.3793 0.4167 1>
    widgetStyle {
     barThickness=5
     disabledBarImage {
      bIsDynamicallyLoaded=false
      drawAs=RoundedBox
      imageType=NoImage
      mirroring=NoMirror
      resourceName=None
      resourceObject=null
      tiling=NoTile
      imageSize=<0 0>
      margin=<ltrb 0 0 0 0>
      outlineSettings {
       bUseBrushTransparency=false
       roundingType=FixedRadius
       width=1
       color {colorUseRule=UseColor_ColorTable, specifiedColor=<rgba 0 0 0 0>}
       cornerRadii=<2 2 2 2>
      }
      tintColor {colorUseRule=UseColor_Specified, specifiedColor=<rgba 0.0048 0.0048 0.0048 1>}
      uVRegion {bIsValid=false, max=<0 0>, min=<0 0>}
     }
     disabledThumbImage {
      bIsDynamicallyLoaded=false
      drawAs=RoundedBox
      imageType=NoImage
      mirroring=NoMirror
      resourceName=None
      resourceObject=null
      tiling=NoTile
      imageSize=<8 8>
      margin=<ltrb 0 0 0 0>
      outlineSettings {
       bUseBrushTransparency=false
       roundingType=HalfHeightRadius
       width=0
       color {colorUseRule=UseColor_Specified, specifiedColor=<rgba 0 0 0 0>}
       cornerRadii=<0 0 0 0>
      }
      tintColor {colorUseRule=UseColor_Specified, specifiedColor=<rgba 0.5271 0.5271 0.5271 1>}
      uVRegion {bIsValid=false, max=<0 0>, min=<0 0>}
     }
     hoveredBarImage {
      bIsDynamicallyLoaded=false
      drawAs=RoundedBox
      imageType=NoImage
      mirroring=NoMirror
      resourceName=None
      resourceObject=null
      tiling=NoTile
      imageSize=<0 0>
      margin=<ltrb 0 0 0 0>
      outlineSettings {
       bUseBrushTransparency=false
       roundingType=FixedRadius
       width=1
       color {colorUseRule=UseColor_ColorTable, specifiedColor=<rgba 0 0 0 0>}
       cornerRadii=<2 2 2 2>
      }
      tintColor {colorUseRule=UseColor_Specified, specifiedColor=<rgba 0.6198 1 0.9591 1>}
      uVRegion {bIsValid=false, max=<0 0>, min=<0 0>}
     }
     hoveredThumbImage {
      bIsDynamicallyLoaded=false
      drawAs=RoundedBox
      imageType=NoImage
      mirroring=NoMirror
      resourceName=None
      resourceObject=null
      tiling=NoTile
      imageSize=<8 8>
      margin=<ltrb 0 0 0 0>
      outlineSettings {
       bUseBrushTransparency=false
       roundingType=HalfHeightRadius
       width=0
       color {colorUseRule=UseColor_Specified, specifiedColor=<rgba 0 0 0 0>}
       cornerRadii=<0 0 0 0>
      }
      tintColor {colorUseRule=UseColor_Specified, specifiedColor=<rgba 1 1 1 1>}
      uVRegion {bIsValid=false, max=<0 0>, min=<0 0>}
     }
     normalBarImage {
      bIsDynamicallyLoaded=false
      drawAs=RoundedBox
      imageType=NoImage
      mirroring=NoMirror
      resourceName=None
      resourceObject=null
      tiling=NoTile
      imageSize=<0 0>
      margin=<ltrb 0 0 0 0>
      outlineSettings {
       bUseBrushTransparency=false
       roundingType=FixedRadius
       width=1
       color {colorUseRule=UseColor_ColorTable, specifiedColor=<rgba 0 0 0 0>}
       cornerRadii=<2 2 2 2>
      }
      tintColor {colorUseRule=UseColor_Specified, specifiedColor=<rgba 0.0384 0.4323 0.3935 1>}
      uVRegion {bIsValid=false, max=<0 0>, min=<0 0>}
     }
     normalThumbImage {
      bIsDynamicallyLoaded=false
      drawAs=RoundedBox
      imageType=NoImage
      mirroring=NoMirror
      resourceName=None
      resourceObject=null
      tiling=NoTile
      imageSize=<8 8>
      margin=<ltrb 0 0 0 0>
      outlineSettings {
       bUseBrushTransparency=false
       roundingType=HalfHeightRadius
       width=0
       color {colorUseRule=UseColor_Specified, specifiedColor=<rgba 0 0 0 0>}
       cornerRadii=<0 0 0 0>
      }
      tintColor {colorUseRule=UseColor_Specified, specifiedColor=<rgba 0.5271 0.5271 0.5271 1>}
      uVRegion {bIsValid=false, max=<0 0>, min=<0 0>}
     }
    }
   }
   slot {class=HorizontalBoxSlot, properties {size {sizeRule=Fill, value=1}}}
  }
  Txt_Value {
   widgetClassPath=&UEFN_TextBlock
   bIsVariable=true
   parent=HorizontalBox_0
   widget=Txt_Value
   properties {
    text=Button Text
    font {
     bForceMonospaced=false
     bMaterialIsStencil=false
     fontMaterial=null
     fontObject=&NotoSans
     letterSpacing=0
     monospacedWidth=1
     size=7.5
     skewAmount=0
     typefaceFontName=NotoSans-Bold
     outlineSettings {
      bApplyOutlineToDropShadows=false
      bMiteredCorners=false
      bSeparateFillAlpha=false
      outlineMaterial=null
      outlineSize=0
      outlineColor=<rgba 0 0 0 1>
     }
    }
   }
   slot {
    class=HorizontalBoxSlot
    properties {horizontalAlignment=HAlign_Center, verticalAlignment=VAlign_Center, padding=<ltrb 3 0 3 0>, size {sizeRule=Fill, value=0.25}}
   }
  }
  EDM_Input_Reset {
   widgetClassPath=&/EDM_Input_Reset
   bIsVariable=true
   parent=HorizontalBox_0
   widget=EDM_Input_Reset
   slot {class=HorizontalBoxSlot, properties {horizontalAlignment=HAlign_Center, verticalAlignment=VAlign_Center}}
  }
 ]
}
