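# EDM_Input_Button_Image is an Editor Utility Widget that displays a texture as a small image, apparently an input button icon for the editor debug menu.
# A SizeBox with aspect ratio overrides centers a single Image widget.
# On PreConstruct, it sets the Image brush from the Soft Texture variable via SetBrushFromSoftTexture without matching the texture's size, so the chosen texture also previews in the designer.
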EditorUtilityWidgetBlueprint /EditorDebugMenu/EDM_Input_Button_Image
graphs [EventGraph {nodes [
   Event PreConstruct()
   > SetBrushFromSoftTexture!(
    execute
    self:Image Object Reference=Get Image() -> (Image:Image Object Reference)
    SoftTexture:Texture 2D Soft Object Reference=Get Soft Texture() -> (Soft Texture:Texture 2D Soft Object Reference)
    bMatchSize:bool=false
   ) -> (then)
   Event Construct()
   Event Tick()
  ]}]
variables [Soft Texture {type=Texture2D}]
widgets {
 info {rootWidgetClass=&SizeBox, widgetCount=2}
 widgets [
  SizeBox_0 {widgetClassPath=&SizeBox, widget=SizeBox_0, properties {bOverride_MaxAspectRatio=true, bOverride_MinAspectRatio=true}}
  Image {
   widgetClassPath=&Image
   bIsVariable=true
   parent=SizeBox_0
   widget=Image
   properties {brush {
     bIsDynamicallyLoaded=false
     drawAs=Image
     imageType=NoImage
     mirroring=NoMirror
     resourceName=None
     resourceObject=null
     tiling=NoTile
     imageSize=<20 20>
     margin=<ltrb 0 0 0 0>
     outlineSettings {
      bUseBrushTransparency=false
      roundingType=HalfHeightRadius
      width=0
      color {colorUseRule=UseColor_Specified, specifiedColor=<rgba 0 0 0 0>}
      cornerRadii=<0 0 0 0>
     }
     tintColor {colorUseRule=UseColor_Specified, specifiedColor=<rgba 1 1 1 1>}
     uVRegion {bIsValid=false, max=<0 0>, min=<0 0>}
    }}
   slot {class=SizeBoxSlot, properties {horizontalAlignment=HAlign_Center, verticalAlignment=VAlign_Center}}
  }
 ]
}
