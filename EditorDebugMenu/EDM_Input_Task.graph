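# Task button row for the Editor Debug Menu, an editor utility widget derived from EDM_Input_Base that calls Run Task on its input data object when clicked.
# Setup Widget sets the label from the data's name and adds an EDM_Input_Button_Image to ImageStack for each valid soft texture.
# Hover plays a short label slide forward and unhover reverses it.
# SortWidgetsIntoArrays extends the parent to register the text and button in its widget arrays.
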EditorUtilityWidgetBlueprint /EditorDebugMenu/EDM_Input_Task
animations [Hover {
  binding_count=1
  length_seconds=0.2
  object=&self:Hover
  sequence {
   bindings [ButtonText {
     id=#1
     parent_id=null
     possessed_class=UEFN_TextBlock_C
     tracks [{
       type=MovieScene2DTransformTrack
       sections [{
         type=MovieScene2DTransformSection
         end_frame=4
         start_frame=0
         channels [
          Translation.X {type=MovieSceneScriptingFloatChannel, keys [[0,0],[4,5]]}
          Translation.Y {type=MovieSceneScriptingFloatChannel, const [0,2]}
          Angle {type=MovieSceneScriptingFloatChannel, const [0,2]}
          Scale.X {type=MovieSceneScriptingFloatChannel, const [1,2]}
          Scale.Y {type=MovieSceneScriptingFloatChannel, const [1,2]}
          Shear.X {type=MovieSceneScriptingFloatChannel, const [0,2]}
          Shear.Y {type=MovieSceneScriptingFloatChannel, const [0,2]}
         ]
        }]
      }]
    }]
   display_rate {denominator=1, numerator=20}
   tick_resolution {denominator=1, numerator=60000}
   tracks []
  }
 }]
graphs [
 EventGraph {nodes [
   "On Clicked (DebugButton)"() {is=title K2Node_ComponentBoundEvent}
   > Run Task(self:Obj EDM Input Data Object Reference="Get Input  Data"() -> ("Input  Data":Obj EDM Input Data Object Reference))
   "On Hovered (DebugButton)"() {is=title K2Node_ComponentBoundEvent}
   n1 = Get Hover() -> (Hover:Widget Animation Object Reference)
   > PlayAnimationForward(InAnimation=@n1)
   "On Unhovered (DebugButton)"() {is=title K2Node_ComponentBoundEvent}
   > PlayAnimationReverse(InAnimation=@n1)
   Event Setup Widget() {is=title K2Node_Event}
   > n2 = Get!(execute:Exec) -> (then:Exec, else:Exec, "Input  Data":Obj EDM Input Data Object Reference)
   n3 = Break Struct EDM Input Data 01!(Struct_EDM_InputData_01:"Struct EDM Input Data 01 Structure (by ref)"=GetData!(self:Obj EDM Input Data Object Reference=@n2) -> (Input Data:Struct EDM Input Data 01 Structure)) -> (
    Name_10_#2:Name
    ButtonImages_29_#3:Array of Texture 2D Soft Object References
    TargetWorld_45_#4:Enum_EDM_WorldType Enum
    EditorUtilityObject_48_#5:Editor Utility Object Soft Class Reference
   ) {is=title K2Node_BreakStruct}
   > "SetText (Text)"(self=Get ButtonText() -> (ButtonText:Text Block Object Reference), InText="To Text (Name)"!(InName:Name=@n3) -> (ReturnValue:Text))
   > n4 = For Each Loop(Array:Array of Texture 2D Soft Object References=@n3.ButtonImages_29_#3) -> (Array Element:Texture 2D Soft Object Reference)
   > Branch(Condition=IsValidSoftObjectReference(SoftObjectReference=@n4))
   > n5 = Create Widget!(
    execute:Exec
    WorldContextObject:Object Reference
    Class:User Widget Class Reference=&/EDM_Input_Button_Image.EDM_Input_Button_Image_C
    OwningPlayer:Player Controller Object Reference
    Soft Texture:Texture 2D Soft Object Reference=@n4
   ) -> (then:Exec, ReturnValue:EDM Input Button Image Object Reference)
   > AddChildToStackBox!(execute, self:Stack Box Object Reference=Get ImageStack() -> (ImageStack:Stack Box Object Reference), Content:Widget Object Reference=@n5) -> (then, ReturnValue:Stack Box Slot Object Reference)
   "Get Input  Data"()
  ]}
 SortWidgetsIntoArrays {nodes [
   SortWidgetsIntoArrays() {is=title K2Node_FunctionEntry}
   > Parent: SortWidgetsIntoArrays() {is=title K2Node_CallParentFunction}
   > Add(
    TargetArray:Array of Text Block Object References=Get Text Widgets() -> (Text Widgets:Array of Text Block Object References)
    NewItem:"Text Block Object Reference (by ref)"=Get ButtonText()
   ) {is=title K2Node_CallArrayFunction}
   > Add(
    TargetArray:Array of Editor Utility Button Object References=Get Button Widgets() -> (Button Widgets:Array of Editor Utility Button Object References)
    NewItem:"Editor Utility Button Object Reference (by ref)"=Get DebugButton() -> (DebugButton:Editor Utility Button Object Reference)
   )
  ]}
]
variables []
widgets {
 info {parentClass=&/EDM_Input_Base.EDM_Input_Base_C, rootWidgetClass=&StackBox, widgetCount=7}
 widgets [
  StackBox_0 {widgetClassPath=&StackBox, widget=StackBox_0}
  EditorDebugMenu_CommonUtils {
   widgetClassPath=&/EDM_Input_CommonUtils
   bIsVariable=true
   parent=StackBox_0
   widget=EditorDebugMenu_CommonUtils
   slot {class=StackBoxSlot}
  }
  DebugButton {
   widgetClassPath=&EditorUtilityButton
   parent=StackBox_0
   widget=DebugButton
   properties {widgetStyle {
     clickedSlateSound {resourceObject=null}
     disabled {
      bIsDynamicallyLoaded=false
      drawAs=RoundedBox
      imageType=NoImage
      mirroring=NoMirror
      resourceName=None
      resourceObject=null
      tiling=NoTile
      imageSize=<0 0>
      margin=<ltrb 0 0 0 0>
      outlineSettings {
       bUseBrushTransparency=false
       roundingType=FixedRadius
       width=1
       color {colorUseRule=UseColor_ColorTable, specifiedColor=<rgba 0 0 0 0>}
       cornerRadii=<4 4 4 4>
      }
      tintColor {colorUseRule=UseColor_Specified, specifiedColor=<rgba 0.0395 0.0395 0.0395 1>}
      uVRegion {bIsValid=false, max=<0 0>, min=<0 0>}
     }
     disabledForeground {colorUseRule=UseColor_ColorTable, specifiedColor=<rgba 0 0 0 0>}
     hovered {
      bIsDynamicallyLoaded=false
      drawAs=RoundedBox
      imageType=NoImage
      mirroring=NoMirror
      resourceName=None
      resourceObject=null
      tiling=NoTile
      imageSize=<32 32>
      margin=<ltrb 0 0 0 0>
      outlineSettings {
       bUseBrushTransparency=false
       roundingType=FixedRadius
       width=1
       color {colorUseRule=UseColor_ColorTable, specifiedColor=<rgba 0 0 0 0>}
       cornerRadii=<4 4 4 4>
      }
      tintColor {colorUseRule=UseColor_Specified, specifiedColor=<rgba 0.0953 0.0953 0.0953 1>}
      uVRegion {bIsValid=false, max=<0 0>, min=<0 0>}
     }
     hoveredForeground {colorUseRule=UseColor_ColorTable, specifiedColor=<rgba 0 0 0 0>}
     hoveredSlateSound {resourceObject=null}
     normal {
      bIsDynamicallyLoaded=false
      drawAs=RoundedBox
      imageType=NoImage
      mirroring=NoMirror
      resourceName=None
      resourceObject=null
      tiling=NoTile
      imageSize=<32 32>
      margin=<ltrb 0 0 0 0>
      outlineSettings {
       bUseBrushTransparency=false
       roundingType=FixedRadius
       width=1
       color {colorUseRule=UseColor_ColorTable, specifiedColor=<rgba 0 0 0 0>}
       cornerRadii=<4 4 4 4>
      }
      tintColor {colorUseRule=UseColor_Specified, specifiedColor=<rgba 0.0395 0.0395 0.0395 1>}
      uVRegion {bIsValid=false, max=<0 0>, min=<0 0>}
     }
     normalForeground {colorUseRule=UseColor_ColorTable, specifiedColor=<rgba 0 0 0 0>}
     normalPadding=<ltrb 5 1.5 12 1.5>
     pressed {
      bIsDynamicallyLoaded=false
      drawAs=RoundedBox
      imageType=NoImage
      mirroring=NoMirror
      resourceName=None
      resourceObject=null
      tiling=NoTile
      imageSize=<32 32>
      margin=<ltrb 0 0 0 0>
      outlineSettings {
       bUseBrushTransparency=false
       roundingType=FixedRadius
       width=1
       color {colorUseRule=UseColor_ColorTable, specifiedColor=<rgba 0 0 0 0>}
       cornerRadii=<4 4 4 4>
      }
      tintColor {colorUseRule=UseColor_Specified, specifiedColor=<rgba 0.0284 0.0284 0.0284 1>}
      uVRegion {bIsValid=false, max=<0 0>, min=<0 0>}
     }
     pressedForeground {colorUseRule=UseColor_ColorTable, specifiedColor=<rgba 0 0 0 0>}
     pressedPadding=<ltrb 12 2.5 12 0.5>
     pressedSlateSound {resourceObject=null}
    }}
   slot {class=StackBoxSlot, properties {size {sizeRule=Fill, value=1}}}
  }
  Container {
   widgetClassPath=&HorizontalBox
   bIsVariable=true
   parent=DebugButton
   widget=Container
   slot {class=ButtonSlot, properties {horizontalAlignment=HAlign_Fill, padding=<ltrb 0 0 0 0>}}
  }
  Image_45 {
   widgetClassPath=&Image
   bIsVariable=true
   parent=Container
   widget=Image_45
   properties {brush {
     bIsDynamicallyLoaded=false
     drawAs=Image
     imageType=NoImage
     mirroring=NoMirror
     resourceName=None
     resourceObject=/EditorDebugMenu/Textures/T_Task
     tiling=NoTile
     imageSize=<12 12>
     margin=<ltrb 0 0 0 0>
     outlineSettings {
      bUseBrushTransparency=false
      roundingType=HalfHeightRadius
      width=0
      color {colorUseRule=UseColor_Specified, specifiedColor=<rgba 0 0 0 0>}
      cornerRadii=<0 0 0 0>
     }
     tintColor {colorUseRule=UseColor_Specified, specifiedColor=<rgba 1 1 1 0.65>}
     uVRegion {bIsValid=false, max=<0 0>, min=<0 0>}
    }}
   slot {class=HorizontalBoxSlot, properties {horizontalAlignment=HAlign_Center, verticalAlignment=VAlign_Center}}
  }
  ButtonText {
   widgetClassPath=&UEFN_TextBlock
   bIsVariable=true
   parent=Container
   widget=ButtonText
   properties {
    text=Button Text
    font {
     bForceMonospaced=false
     bMaterialIsStencil=false
     fontMaterial=null
     fontObject=&NotoSans
     letterSpacing=0
     monospacedWidth=1
     size=8.25
     skewAmount=0
     typefaceFontName=NotoSans-Regular
     outlineSettings {
      bApplyOutlineToDropShadows=false
      bMiteredCorners=false
      bSeparateFillAlpha=false
      outlineMaterial=null
      outlineSize=0
      outlineColor=<rgba 0 0 0 1>
     }
    }
   }
   slot {class=HorizontalBoxSlot, properties {padding=<ltrb 6 2 2 2>, size {sizeRule=Fill, value=1}}}
  }
  ImageStack {
   widgetClassPath=&StackBox
   bIsVariable=true
   parent=Container
   widget=ImageStack
   slot {class=HorizontalBoxSlot, properties {padding=<ltrb 5 0 5 0>}}
  }
 ]
}
