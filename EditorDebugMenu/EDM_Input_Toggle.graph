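# EDM_Input_Toggle is an editor utility widget that provides a labeled checkbox input for the Editor Debug Menu, extending EDM_Input_Base.
# On setup it registers its label, sets the text from the input data's name, binds to that object's Boolean Updated event and syncs the current value.
# Checkbox changes pass a menu-ready check before updating the data object, whose update event sets the checked state.
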EditorUtilityWidgetBlueprint /EditorDebugMenu/EDM_Input_Toggle
graphs [EventGraph {nodes [
   n1 = "On Check State Changed (CheckBox)"!() -> (OutputDelegate, then, bIsChecked:bool) {is=title K2Node_ComponentBoundEvent}
   > Branch(Condition=Get Debug Menu - Ready() -> (Debug Menu - Ready:bool))
   > Update Boolean!(
    execute
    self:Obj EDM Input Data Object Reference="Get Input  Data"() -> ("Input  Data":Obj EDM Input Data Object Reference)
    Booolean Value:bool=@n1.bIsChecked
   ) -> (then)
   n2 = Boolean Updated!() -> (OutputDelegate, then, Value:bool=~) {is=title K2Node_CustomEvent}
   > SetCheckedState!(
    execute
    self:Check Box Object Reference=Get CheckBox() -> (CheckBox:Editor Utility Check Box Object Reference)
    InCheckedState:ECheckBoxState Enum=Select(Option 0:ECheckBoxState Enum=Unchecked, Option 1:ECheckBoxState Enum=Checked, Index=@n2.Value) -> (ReturnValue:ECheckBoxState Enum)
   ) -> (then)
   Event Setup Widget() {is=title K2Node_Event}
   > n3 = Get!(execute:Exec) -> (then:Exec, else:Exec, "Input  Data":Obj EDM Input Data Object Reference)
   > Add(
    TargetArray:Array of Text Block Object References=Get Text Widgets() -> (Text Widgets:Array of Text Block Object References)
    NewItem:"Text Block Object Reference (by ref)"=Get ButtonText() -> (ButtonText:Text Block Object Reference)
   ) {is=title K2Node_CallArrayFunction}
   n4 = GetData!(self:Obj EDM Input Data Object Reference=@n3) -> (Input Data:Struct EDM Input Data 01 Structure)
   > "SetText (Text)"(
    self=Get ButtonText()
    InText="To Text (Name)"!(InName:Name=Break Struct EDM Input Data 01!(Struct_EDM_InputData_01:"Struct EDM Input Data 01 Structure (by ref)"=@n4) -> (
      Name_10_#1:Name
      TargetWorld_45_#2:Enum_EDM_WorldType Enum
      EditorUtilityObject_48_#3:Editor Utility Object Soft Class Reference
     ) {is=title K2Node_BreakStruct}) -> (ReturnValue:Text)
   )
   n5 = "Get Input  Data"()
   > Bind Event to Boolean Updated(self:Obj EDM Input Data Object Reference=@n5, Delegate=@n2) {is=title K2Node_AddDelegate}
   > Boolean Updated!(execute, self, Value:bool=Get Booolean Value!(self:Obj EDM Input Data Object Reference=@n5) -> (Booolean Value:bool)) -> (then) {is=title K2Node_CallFunction}
   "Get Input  Data"()
  ]}]
variables [
 NewVar {type=boolean}
 True or False {type=string, default=""}
]
widgets {
 info {parentClass=&/EDM_Input_Base.EDM_Input_Base_C, rootWidgetClass=&StackBox, widgetCount=5}
 widgets [
  StackBox_0 {widgetClassPath=&StackBox, widget=StackBox_0}
  EditorDebugMenu_CommonUtils {
   widgetClassPath=&/EDM_Input_CommonUtils
   bIsVariable=true
   parent=StackBox_0
   widget=EditorDebugMenu_CommonUtils
   slot {class=StackBoxSlot}
  }
  DebugButton {
   widgetClassPath=&HorizontalBox
   parent=StackBox_0
   widget=DebugButton
   slot {class=StackBoxSlot, properties {size {sizeRule=Fill, value=1}}}
  }
  CheckBox {
   widgetClassPath=&EditorUtilityCheckBox
   bIsVariable=true
   parent=DebugButton
   widget=CheckBox
   slot {class=HorizontalBoxSlot, properties {horizontalAlignment=HAlign_Left, verticalAlignment=VAlign_Center, padding=<ltrb 5 0 5 0>}}
  }
  ButtonText {
   widgetClassPath=&UEFN_TextBlock
   bIsVariable=true
   parent=DebugButton
   widget=ButtonText
   properties {
    text=Button Text
    font {
     bForceMonospaced=false
     bMaterialIsStencil=false
     fontMaterial=null
     fontObject=&NotoSans
     letterSpacing=0
     monospacedWidth=1
     size=8
     skewAmount=0
     typefaceFontName=NotoSans-Regular
     outlineSettings {
      bApplyOutlineToDropShadows=false
      bMiteredCorners=false
      bSeparateFillAlpha=false
      outlineMaterial=null
      outlineSize=0
      outlineColor=<rgba 0 0 0 1>
     }
    }
   }
   slot {class=HorizontalBoxSlot, properties {verticalAlignment=VAlign_Center, size {sizeRule=Fill, value=1}}}
  }
 ]
}
