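# Editor utility list-entry button that represents one project in the Editor Debug Menu's project filter and toggles that project's visibility on click.
# Setting the list item stores its Obj_ProjectData, binds to its visibility-updated event and applies the current state; entry release unbinds.
# Visibility changes tint the button green or red.
# Getter functions supply a matching background color and "project is Visible/Hidden" text.
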EditorUtilityWidgetBlueprint /EditorDebugMenu/EDM_ListEntry_ProjectFilter
graphs [
 EventGraph {nodes [
   "On Clicked (Button)"() {is=title K2Node_ComponentBoundEvent}
   > Toggle Project Visibility(self:Obj Project Data Object Reference=Get Project Data() -> (Project Data:Obj Project Data Object Reference))
   n1 = VisUpdated!() -> (OutputDelegate, then, New Visibility:bool=~) {is=title K2Node_CustomEvent}
   > SetBackgroundColor!(
    execute
    self:Button Object Reference=Get Button() -> (Button:Editor Utility Button Object Reference)
    InBackgroundColor:Linear Color Structure=SelectColor(A=(R=0.1888,G=0.7865,B=0.1599,A=1), B=(R=0.4167,G=0.1053,B=0.0694,A=1), bPickA=@n1.New Visibility)
   ) -> (then)
   Event On Entry Released() {is=title K2Node_Event}
   > Unbind Event from Project Visibility Updated(self:Obj Project Data Object Reference=Get Project Data(), Delegate=@n1) {is=title K2Node_RemoveDelegate}
   n2 = Event OnListItemObjectSet()
   > n3 = Cast To Obj_ProjectData!(execute, Object=@n2.ListItemObject) -> (then, CastFailed, AsObj Project Data:Obj Project Data Object Reference) {is=title K2Node_DynamicCast}
   > n4 = Set Project Data(Project Data:Obj Project Data Object Reference=@n3) -> (Output_Get:Obj Project Data Object Reference)
   > Bind Event to Project Visibility Updated(self:Obj Project Data Object Reference=@n4, Delegate=@n1) {is=title K2Node_AddDelegate}
   > VisUpdated!(execute, self, New Visibility:bool=Get Visible!(self:Obj Project Data Object Reference=@n4) -> (Visible:bool)) -> (then) {is=title K2Node_CallFunction}
   Event PreConstruct()
  ]}
 Get_BG_BrushColor {nodes [
   Get_BG_BrushColor() {is=title K2Node_FunctionEntry}
   n1 = Get Project Data()
   > Return Node!(
    execute
    ReturnValue:Linear Color Structure=SelectColor(A=(R=0.066,G=0.1667,B=0.109,A=1), B=(R=0.0417,G=0.0213,B=0.0184,A=1), bPickA=Get Visible(self=@n1))
   )
  ]}
 Get_Txt_ProjectName_Text {nodes [
   Get_Txt_ProjectName_Text() {is=title K2Node_FunctionEntry}
   n1 = Get Project Data()
   n2 = Get Visible(self=@n1)
   > Return Node(ReturnValue:Text=Format Text!(
     Format="LOCTEXT(\"{project} is {vis}\")"
     project:String=Get Project Name!(self:Obj Project Data Object Reference=@n1) -> (Project Name:String)
     vis:String=SelectString(A=Visible, B=Hidden, bPickA=@n2)
    ) -> (Result))
  ]}
 VisToggled {nodes [VisToggled() {is=title K2Node_FunctionEntry}]}
]
variables [
 Debug Menu {type=EUW_EditorDebugMenu_C, default=null}
 Project Data {type=Obj_ProjectData_C, default=null}
 VisToggled {default=()}
]
widgets {
 info {rootWidgetClass=&EditorUtilityButton, widgetCount=2}
 widgets [
  Button {widgetClassPath=&EditorUtilityButton, widget=Button, slot=~}
  ButtonText {
   widgetClassPath=&UEFN_TextBlock
   bIsVariable=true
   parent=Button
   widget=ButtonText
   properties {
    text=Button Text
    font {
     bForceMonospaced=false
     bMaterialIsStencil=false
     fontMaterial=null
     fontObject=&NotoSans
     letterSpacing=0
     monospacedWidth=1
     size=8.25
     skewAmount=0
     typefaceFontName=NotoSans-Regular
     outlineSettings {
      bApplyOutlineToDropShadows=false
      bMiteredCorners=false
      bSeparateFillAlpha=false
      outlineMaterial=null
      outlineSize=0
      outlineColor=<rgba 0 0 0 1>
     }
    }
   }
   slot {class=ButtonSlot, properties {horizontalAlignment=HAlign_Fill, padding=<ltrb 6 2 2 2>}}
  }
 ]
}
