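# Editor utility widget for the editor debug menu that offers a "List Item Override" pick through category and value dropdowns.
# Selecting a category clears the value dropdown and runs its generator, which refills it through SetList.
# Sources are data table cheat values, hard-coded hero ranks, an AMS folder scan, and an Asset Registry query for prefix-stripped, sorted vehicle bodies.
# Get Selected List Item returns the chosen value unless the category is None.
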EditorUtilityWidgetBlueprint /EditorDebugMenu/EDM_ListGenerator
graphs [
 ClearList {nodes [
   ClearList() {is=title K2Node_FunctionEntry}
   > ClearOptions(self:"Combo Box (String) Object Reference"=Get Combo_ListValues() -> (Combo_ListValues:Editor Utility Combo Box String Object Reference))
  ]}
 EventGraph {nodes [
   n1 = "On Selection Changed (Combo_ListSelector)"!() -> (OutputDelegate, then, SelectedItem:String, SelectionType:ESelectInfo Enum) {is=title K2Node_ComponentBoundEvent}
   > n2 = Switch on String!(execute:Exec, Selection:String=@n1.SelectedItem) -> (
    None:Exec
    AMS:Exec
    NPCs:Exec
    HeroRanks:Exec
    Mars_HeroClass:Exec
    Mars_Enemy:Exec
    Mars_Creature:Exec
    LFN Vehicle Trims:Exec
    LFN Vehicle Bodies:Exec
   ) {
    bHasDefaultPin=false
    pinNames [None, AMS, NPCs, HeroRanks, Mars_HeroClass, Mars_Enemy, Mars_Creature, LFN Vehicle Trims, LFN Vehicle Bodies]
   }
   GenList_AMS(execute=@n2.AMS)
   GenListHeroRanks(execute=@n2.HeroRanks)
   GenList_LFNVehicleBodies(execute=@n2.LFN Vehicle Bodies)
   GenList_LFNVehicleTrims(execute=@n2.LFN Vehicle Trims)
   GenList_MarsCreature(execute=@n2.Mars_Creature)
   GenList_MarsEnemy(execute=@n2.Mars_Enemy)
   GenList_MarsHeroClass(execute=@n2.Mars_HeroClass)
   GenList_NPCs(execute=@n2.NPCs)
   ClearList(execute=@n2.None) {is=title K2Node_CallFunction}
   Event PreConstruct()
   Event Construct()
   Event Tick()
  ]}
 GenListHeroRanks {nodes [
   GenListHeroRanks() {is=title K2Node_FunctionEntry}
   > ClearList()
   > SetList!(
    execute
    self
    List Strings:Array of Strings=Make Array!("[0]":String=S+, "[1]":String=S, "[2]":String=A, "[3]":String=B, "[4]":String=C) -> (Array:Array of Strings)
   ) -> (then)
  ]}
 GenList_AMS {nodes [
   GenList_AMS() {is=title K2Node_FunctionEntry}
   > ClearList()
   > n1 = ListAssets!(execute, DirectoryPath:String=/FigureCosmetics/AMS, bRecursive:bool=false, bIncludeFolder:bool=false) -> (then, ReturnValue:Array of Strings)
   > n2 = For Each Loop(Array=@n1)
   SetList(execute=@n2.Completed, List Strings=Get AMS Name List() -> (AMS Name List:Array of Strings))
   Branch(execute=@n2.LoopBody, Condition=Contains(SearchIn=@n2, Substring=AMS_))
   > Add(
    TargetArray=Get AMS Name List()
    NewItem=Split(SourceString=Split(SourceString=@n2, InStr=/, SearchDir=FromEnd).RightS, InStr=., SearchDir=FromStart)
   ) {is=title K2Node_CallArrayFunction}
  ]}
 GenList_LFNVehicleBodies {nodes [
   GenList_LFNVehicleBodies() {is=title K2Node_FunctionEntry}
   > ClearList()
   > n1 = GetAssets!(
    execute
    self:Asset Registry Interface=GetAssetRegistry()
    Filter_PackageNames:Array of Names
    Filter_PackagePaths:Array of Names=Make Array!("[0]":Name=/JunoVehicleCosmetics/Bodies/) -> (Array:Array of Names)
    Filter_SoftObjectPaths:Array of Soft Object Path Structures
    Filter_ClassPaths:Array of Top Level Asset Path Structures
    Filter_RecursiveClassPathsExclusionSet:Set of Top Level Asset Path Structures
    Filter_ClassNames:Array of Names=Make Array("[0]"=/Script/CustomizableObject)
    Filter_RecursiveClassesExclusionSet:Set of Names
    Filter_bRecursivePaths:bool=true
    Filter_bRecursiveClasses:bool=false
    Filter_bIncludeOnlyOnDiskAssets:bool=false
    bSkipARFilteredAssets:bool=true
   ) -> (then, OutAssetData:Array of Asset Data Structures, ReturnValue:bool)
   > n2 = For Each Loop!(Exec, Array:Array of Asset Data Structures=@n1) -> (
    LoopBody
    Array Element_PackageName:Name
    Array Element_PackagePath:Name
    Array Element_AssetName:Name
    Array Element_AssetClassPath:Top Level Asset Path Structure
    Array Index
    Completed
   )
   SortStringArray!(
    execute=@n2.Completed
    TargetArray:Array of Strings=Get Body_NameList() -> (Body_NameList:Array of Strings)
    bStableSort:bool=false
    SortOrder:EArraySortOrder Enum=Ascending
   ) -> (then)
   > SetList(List Strings=Get Body_NameList())
   n3 = "To String (Name)"(InName=@n2.Array Element_AssetName)
   n4 = MakeLiteralString(Value=CO_LFNBody_)
   Branch(execute=@n2.LoopBody, Condition=StartsWith(SourceString=@n3, InPrefix=@n4))
   n5 = Split(SourceString=@n3, InStr=@n4, SearchDir=FromStart)
   > n6 = Branch(Condition=IsEmpty(InString=@n5.RightS))
   Add Unique(execute=@n6.else, TargetArray=Get Body_NameList(), NewItem=@n5.RightS)
  ]}
 GenList_LFNVehicleTrims {nodes [
   GenList_LFNVehicleTrims() {is=title K2Node_FunctionEntry}
   > ClearList()
   > n1 = GetDataTableRowNames(Table=/EditorDebugMenu/Data/Sprout/List_LFNVehicleCosmetics_Trims)
   > n2 = For Each Loop!(Exec, Array:Array of Names=@n1) -> (LoopBody, Array Element:Name, Array Index, Completed)
   SetList(execute=@n2.Completed, List Strings=Get Color_NameList() -> (Color_NameList:Array of Strings))
   n3 = Get Data Table Row List_LFNVehicleCosmetics_Trims!(
    execute:Exec=@n2.LoopBody
    DataTable:Data Table Object Reference=&List_LFNVehicleCosmetics_Trims
    RowName:Name=@n2
   ) -> (then:Exec, RowNotFound:Exec, ReturnValue:Struct EDM List Data Structure) {is=title K2Node_GetDataTableRow}
   > Add Unique(
    TargetArray=Get Color_NameList()
    NewItem=Break Struct EDM List Data!(Struct_EDM_ListData:"Struct EDM List Data Structure (by ref)"=@n3) -> (ListCommonName_7_#1:String, ListCheatValue_5_#2:String) {is=title K2Node_BreakStruct}.ListCheatValue_5_#2
   )
  ]}
 GenList_MarsCreature {nodes [
   GenList_MarsCreature() {is=title K2Node_FunctionEntry}
   > ClearList()
   > n1 = GetDataTableRowNames(Table=/EditorDebugMenu/Data/Mars/List_Mars_Creature)
   > n2 = For Each Loop(Array=@n1)
   SetList(execute=@n2.Completed, List Strings=Get Creature_NameList() -> (Creature_NameList:Array of Strings))
   n3 = Get Data Table Row List_Mars_Creature!(execute:Exec=@n2.LoopBody, DataTable:Data Table Object Reference=&List_Mars_Creature, RowName:Name=@n2) -> (then:Exec, RowNotFound:Exec, ReturnValue:Struct EDM List Data Structure) {is=title K2Node_GetDataTableRow}
   > Add Unique(
    TargetArray=Get Creature_NameList()
    NewItem=Break Struct EDM List Data(Struct_EDM_ListData=@n3).ListCheatValue_5_#2
   )
  ]}
 GenList_MarsEnemy {nodes [
   GenList_MarsEnemy() {is=title K2Node_FunctionEntry}
   > ClearList()
   > n1 = GetDataTableRowNames(Table=/EditorDebugMenu/Data/Mars/List_Mars_Enemy)
   > n2 = For Each Loop(Array=@n1)
   SetList(execute=@n2.Completed, List Strings=Get Enemy_NameList() -> (Enemy_NameList:Array of Strings))
   n3 = Get Data Table Row List_Mars_Enemy!(execute:Exec=@n2.LoopBody, DataTable:Data Table Object Reference=&List_Mars_Enemy, RowName:Name=@n2) -> (then:Exec, RowNotFound:Exec, ReturnValue:Struct EDM List Data Structure) {is=title K2Node_GetDataTableRow}
   > Add Unique(
    TargetArray=Get Enemy_NameList()
    NewItem=Break Struct EDM List Data(Struct_EDM_ListData=@n3).ListCheatValue_5_#2
   )
  ]}
 GenList_MarsHeroClass {nodes [
   GenList_MarsHeroClass() {is=title K2Node_FunctionEntry}
   > ClearList()
   > n1 = GetDataTableRowNames(Table=/EditorDebugMenu/Data/Mars/List_Mars_HeroClass)
   > n2 = For Each Loop(Array=@n1)
   SetList(execute=@n2.Completed, List Strings=Get Hero_NameList() -> (Hero_NameList:Array of Strings))
   n3 = Get Data Table Row List_Mars_HeroClass!(execute:Exec=@n2.LoopBody, DataTable:Data Table Object Reference=&List_Mars_HeroClass, RowName:Name=@n2) -> (then:Exec, RowNotFound:Exec, ReturnValue:Struct EDM List Data Structure) {is=title K2Node_GetDataTableRow}
   > Add Unique(
    TargetArray=Get Hero_NameList()
    NewItem=Break Struct EDM List Data(Struct_EDM_ListData=@n3).ListCheatValue_5_#2
   )
  ]}
 GenList_NPCs {nodes [
   GenList_NPCs() {is=title K2Node_FunctionEntry}
   > ClearList()
   > n1 = GetDataTableRowNames(Table=/EditorDebugMenu/Data/Fortnite/DT_EDM_NPCList)
   > n2 = For Each Loop(Array=@n1)
   SetList(execute=@n2.Completed, List Strings=Get NPC_NameList() -> (NPC_NameList:Array of Strings))
   n3 = Get Data Table Row DT_EDM_NPCList!(execute:Exec=@n2.LoopBody, DataTable:Data Table Object Reference=&DT_EDM_NPCList, RowName:Name=@n2) -> (then:Exec, RowNotFound:Exec, ReturnValue:FEDM NPCList Structure) {is=title K2Node_GetDataTableRow}
   > Add Unique(
    TargetArray=Get NPC_NameList()
    NewItem=Break FEDM NPCList!(FEDM_NPCList:"FEDM NPCList Structure (by ref)"=@n3) -> (NPCCommonName_5_#3:String, NPCCheatName_2_#4:String) {is=title K2Node_BreakStruct}.NPCCheatName_2_#4
   )
  ]}
 Get Selected List Item {nodes [
   Get Selected List Item() {is=title K2Node_FunctionEntry}
   n1 = Get Combo_ListSelector() -> (Combo_ListSelector:Editor Utility Combo Box String Object Reference)
   > n2 = Branch(Condition="Not Equal, Case Insenstive (String)"!(A:String=GetSelectedOption!(self:"Combo Box (String) Object Reference"=@n1) -> (ReturnValue:String), B:String=None) -> (ReturnValue:bool))
   > Return Node!(execute, List Item:String=GetSelectedOption(self=Get Combo_ListValues()))
   Return Node(execute=@n2.else)
  ]}
 SetList {nodes [
   n1 = SetList!() -> (then, List Strings:Array of Strings) {is=title K2Node_FunctionEntry}
   > n2 = For Each Loop(Array:Array of Strings=@n1) -> (Array Element:String)
   > AddOption!(execute, self:"Combo Box (String) Object Reference"=Get Combo_ListValues(), Option:String=@n2) -> (then)
  ]}
]
variables []
widgets {
 info {rootWidgetClass=&StackBox, widgetCount=5}
 widgets [
  StackBox_1 {widgetClassPath=&StackBox, widget=StackBox_1, properties {orientation=Orient_Vertical}}
  Txt_ExpandAll_1 {
   widgetClassPath=&UEFN_TextBlock
   parent=StackBox_1
   widget=Txt_ExpandAll_1
   properties {
    text=List Item Override:
    colorAndOpacity {colorUseRule=UseColor_Specified, specifiedColor=<rgba 0.2969 0.2969 0.2969 1>}
    font {
     bForceMonospaced=false
     bMaterialIsStencil=false
     fontMaterial=null
     fontObject=&NotoSans
     letterSpacing=0
     monospacedWidth=1
     size=8.25
     skewAmount=0
     typefaceFontName=NotoSans-Regular
     outlineSettings {
      bApplyOutlineToDropShadows=false
      bMiteredCorners=false
      bSeparateFillAlpha=false
      outlineMaterial=null
      outlineSize=0
      outlineColor=<rgba 0 0 0 1>
     }
    }
   }
   slot {class=StackBoxSlot, properties {padding=<ltrb 5 0 0 0>}}
  }
  StackBox_0 {
   widgetClassPath=&StackBox
   parent=StackBox_1
   widget=StackBox_0
   slot {class=StackBoxSlot, properties {verticalAlignment=VAlign_Top, padding=<ltrb 10 0 0 0>}}
  }
  Combo_ListSelector {
   widgetClassPath=/Script/Blutility.EditorUtilityComboBoxString
   bIsVariable=true
   parent=StackBox_0
   widget=Combo_ListSelector
   properties {
    selectedOption=None
    defaultOptions [None, AMS, NPCs, HeroRanks, Mars_Creature, Mars_Enemy, Mars_HeroClass, LFN Vehicle Trims, LFN Vehicle Bodies]
    font {
     bForceMonospaced=false
     bMaterialIsStencil=false
     fontMaterial=null
     fontObject=&NotoSans
     letterSpacing=0
     monospacedWidth=1
     size=7.5
     skewAmount=0
     typefaceFontName=NotoSans-Regular
     outlineSettings {
      bApplyOutlineToDropShadows=false
      bMiteredCorners=false
      bSeparateFillAlpha=false
      outlineMaterial=null
      outlineSize=0
      outlineColor=<rgba 0 0 0 1>
     }
    }
    foregroundColor {colorUseRule=UseColor_Specified, specifiedColor=<rgba 1 1 1 1>}
    widgetStyle {
     comboButtonStyle {
      downArrowAlign=VAlign_Center
      buttonStyle {
       clickedSlateSound {resourceObject=null}
       disabled {
        bIsDynamicallyLoaded=false
        drawAs=RoundedBox
        imageType=NoImage
        mirroring=NoMirror
        resourceName=None
        resourceObject=null
        tiling=NoTile
        imageSize=<0 0>
        margin=<ltrb 0 0 0 0>
        outlineSettings {
         bUseBrushTransparency=false
         roundingType=FixedRadius
         width=1
         color {colorUseRule=UseColor_ColorTable, specifiedColor=<rgba 0 0 0 0>}
         cornerRadii=<4 4 4 4>
        }
        tintColor {colorUseRule=UseColor_Specified, specifiedColor=<rgba 0.0395 0.0395 0.0395 1>}
        uVRegion {bIsValid=false, max=<0 0>, min=<0 0>}
       }
       disabledForeground {colorUseRule=UseColor_ColorTable, specifiedColor=<rgba 0 0 0 0>}
       hovered {
        bIsDynamicallyLoaded=false
        drawAs=RoundedBox
        imageType=NoImage
        mirroring=NoMirror
        resourceName=None
        resourceObject=null
        tiling=NoTile
        imageSize=<0 0>
        margin=<ltrb 0 0 0 0>
        outlineSettings {
         bUseBrushTransparency=false
         roundingType=FixedRadius
         width=1
         color {colorUseRule=UseColor_ColorTable, specifiedColor=<rgba 0 0 0 0>}
         cornerRadii=<4 4 4 4>
        }
        tintColor {colorUseRule=UseColor_Specified, specifiedColor=<rgba 0.0048 0.0048 0.0048 1>}
        uVRegion {bIsValid=false, max=<0 0>, min=<0 0>}
       }
       hoveredForeground {colorUseRule=UseColor_ColorTable, specifiedColor=<rgba 0 0 0 0>}
       hoveredSlateSound {resourceObject=null}
       normal {
        bIsDynamicallyLoaded=false
        drawAs=RoundedBox
        imageType=NoImage
        mirroring=NoMirror
        resourceName=None
        resourceObject=null
        tiling=NoTile
        imageSize=<0 0>
        margin=<ltrb 0 0 0 0>
        outlineSettings {
         bUseBrushTransparency=false
         roundingType=FixedRadius
         width=1
         color {colorUseRule=UseColor_ColorTable, specifiedColor=<rgba 0 0 0 0>}
         cornerRadii=<4 4 4 4>
        }
        tintColor {colorUseRule=UseColor_Specified, specifiedColor=<rgba 0.0048 0.0048 0.0048 1>}
        uVRegion {bIsValid=false, max=<0 0>, min=<0 0>}
       }
       normalForeground {colorUseRule=UseColor_ColorTable, specifiedColor=<rgba 0 0 0 0>}
       normalPadding=<ltrb 8 1 8 1>
       pressed {
        bIsDynamicallyLoaded=false
        drawAs=RoundedBox
        imageType=NoImage
        mirroring=NoMirror
        resourceName=None
        resourceObject=null
        tiling=NoTile
        imageSize=<0 0>
        margin=<ltrb 0 0 0 0>
        outlineSettings {
         bUseBrushTransparency=false
         roundingType=FixedRadius
         width=1
         color {colorUseRule=UseColor_ColorTable, specifiedColor=<rgba 0 0 0 0>}
         cornerRadii=<4 4 4 4>
        }
        tintColor {colorUseRule=UseColor_Specified, specifiedColor=<rgba 0.0048 0.0048 0.0048 1>}
        uVRegion {bIsValid=false, max=<0 0>, min=<0 0>}
       }
       pressedForeground {colorUseRule=UseColor_ColorTable, specifiedColor=<rgba 0 0 0 0>}
       pressedPadding=<ltrb 8 2 8 0>
       pressedSlateSound {resourceObject=null}
      }
      contentPadding=<ltrb 0 0 0 0>
      downArrowImage {
       bIsDynamicallyLoaded=false
       drawAs=Image
       imageType=Vector
       mirroring=NoMirror
       resourceName=../../../Engine/Content/Slate/Starship/CoreWidgets/ComboBox/wide-chevron-down.svg
       resourceObject=null
       tiling=NoTile
       imageSize=<16 16>
       margin=<ltrb 0 0 0 0>
       outlineSettings {
        bUseBrushTransparency=false
        roundingType=HalfHeightRadius
        width=0
        color {colorUseRule=UseColor_Specified, specifiedColor=<rgba 0 0 0 0>}
        cornerRadii=<0 0 0 0>
       }
       tintColor {colorUseRule=UseColor_Specified, specifiedColor=<rgba 1 1 1 1>}
       uVRegion {bIsValid=false, max=<0 0>, min=<0 0>}
      }
      downArrowPadding=<ltrb 2 2 2 2>
      menuBorderBrush {
       bIsDynamicallyLoaded=false
       drawAs=RoundedBox
       imageType=NoImage
       mirroring=NoMirror
       resourceName=None
       resourceObject=null
       tiling=NoTile
       imageSize=<0 0>
       margin=<ltrb 0 0 0 0>
       outlineSettings {
        bUseBrushTransparency=false
        roundingType=FixedRadius
        width=1
        color {colorUseRule=UseColor_ColorTable, specifiedColor=<rgba 0 0 0 0>}
        cornerRadii=<0 0 0 0>
       }
       tintColor {colorUseRule=UseColor_Specified, specifiedColor=<rgba 0.0395 0.0395 0.0395 1>}
       uVRegion {bIsValid=false, max=<0 0>, min=<0 0>}
      }
      menuBorderPadding=<ltrb 0 0 0 0>
      shadowColorAndOpacity=<rgba 0 0 0 1>
      shadowOffset=<0 0>
     }
     contentPadding=<ltrb 0 0 0 0>
     menuRowPadding=<ltrb 8 3 8 3>
     pressedSlateSound {resourceObject=null}
     selectionChangeSlateSound {resourceObject=null}
    }
   }
   slot {class=StackBoxSlot, properties {padding=<ltrb 0 0 5 0>}}
  }
  Combo_ListValues {
   widgetClassPath=&EditorUtilityComboBoxString
   bIsVariable=true
   parent=StackBox_0
   widget=Combo_ListValues
   properties {
    defaultOptions [None,AMS,NPCs,HeroRanks]
    font {
     bForceMonospaced=false
     bMaterialIsStencil=false
     fontMaterial=null
     fontObject=&NotoSans
     letterSpacing=0
     monospacedWidth=1
     size=7.5
     skewAmount=0
     typefaceFontName=NotoSans-Regular
     outlineSettings {
      bApplyOutlineToDropShadows=false
      bMiteredCorners=false
      bSeparateFillAlpha=false
      outlineMaterial=null
      outlineSize=0
      outlineColor=<rgba 0 0 0 1>
     }
    }
    foregroundColor {colorUseRule=UseColor_Specified, specifiedColor=<rgba 1 1 1 1>}
   }
   slot {class=StackBoxSlot, properties {size {sizeRule=Fill, value=1}}}
  }
 ]
}
