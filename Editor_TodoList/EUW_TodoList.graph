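# Editor Utility Widget that provides a persistent to-do list panel inside the Unreal Editor.
# Construct loads or creates a SaveGame in the TodoList slot; add, modify, remove and clear actions save it and rebuild the list.
# Each task becomes an EUW_TodoList_Entry, with priority tasks inserted on top and edits matched by GUID.
# Collapsed Jira controls open an issue-search URL and string-parse pasted XML for [PLAY titles, skipping duplicates.
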EditorUtilityWidgetBlueprint /Editor_TodoList/EUW_TodoList
graphs [
 Add Entry {nodes [
   Add Entry() {is=title K2Node_FunctionEntry}
   > Branch(Condition=Get bLoaded() -> (bLoaded:bool))
   n1 = Get TextBox_EntryInput() -> (TextBox_EntryInput:Editor Utility Editable Text Box Object Reference)
   > n2 = Branch(Condition=TextIsEmpty(InText=Get Text(self=@n1)))
   n3 = Get!(execute:Exec=@n2.else) -> (then:Exec, else:Exec, Save File:Save File Todo List Object Reference)
   n4 = Get TextBox_EntryInput()
   > Add(
    TargetArray:Array of Struct Todo List Data Structures=Get Todo List!(self:Save File Todo List Object Reference=@n3) -> (Todo List:Array of Struct Todo List Data Structures)
    NewItem:"Struct Todo List Data Structure (by ref)"=Make Struct Todo List Data!(
     Completed_1_#1:bool=False
     Message_4_#2:Text=Get Text(self=@n4)
     Tooltip_19_#3:Text
     GUID_7_#4:Guid Structure=NewGuid!() -> (ReturnValue:Guid Structure)
     PriorityTask_9_#5:bool=False
     JiraTask_15_#6:bool=False
     JiraAddress_12_#7:String
    ) -> (Struct_TodoListData:Struct Todo List Data Structure) {is=title K2Node_MakeStruct}
   ) {is=title K2Node_CallArrayFunction}
   > "SetText (Text Box)"!(execute, self:Text Box Object Reference=Get TextBox_EntryInput(), InText:Text) -> (then)
   > Save()
   > Populate List()
  ]}
 Check is task already in list {nodes [
   Check is task already in list!() -> (then, Task:String) {is=title K2Node_FunctionEntry}
   > n1 = Get()
   > n2 = For Each Loop(Array:Array of Struct Todo List Data Structures=Get Todo List(self=@n1)) -> (Array Element:Struct Todo List Data Structure)
   Return Node!(execute=@n2.Completed, Already Present:bool=false)
   n3 = Break Struct Todo List Data!(Struct_TodoListData:"Struct Todo List Data Structure (by ref)"=@n2) -> (
    Message_4_#2:Text
    Tooltip_19_#3:Text
    JiraTask_15_#6:bool=~
    JiraAddress_12_#7:String
   ) {is=title K2Node_BreakStruct}
   > Branch(Condition="Equal, Case Insensitive (String)"(A="To String (Text)"(InText=@n3), B=Get Task() -> (Task:String)))
   > Return Node(Already Present=true)
  ]}
 Clear List {nodes [
   n1 = Clear List!() -> (then, All:bool=~) {is=title K2Node_FunctionEntry}
   > Branch(Condition=Get bLoaded())
   > n2 = Sequence()
   > n3 = Branch(Condition=@n1)
   n4 = Get Todo List(self=Get Save File() -> (Save File:Save File Todo List Object Reference))
   > Clear(TargetArray:Array of Struct Todo List Data Structures=@n4)
   n5 = Reverse for Each Loop!(Exec:Exec=@n3.else, Array:Array of Struct Todo List Data Structures=@n4) -> (LoopBody:Exec, ArrayIndex:int, ArrayElement:Struct Todo List Data Structure, Completed:Exec) {is=title K2Node_MacroInstance}
   > Branch(Condition=Break Struct Todo List Data!(Struct_TodoListData=@n5.ArrayElement) -> (
     Completed_1_#1:bool=~
     Message_4_#2
     Tooltip_19_#3
     GUID_7_#4:Guid Structure
     PriorityTask_9_#5:bool=~
     JiraTask_15_#6=~
     JiraAddress_12_#7
    ))
   > Remove Index!(execute, TargetArray:Array of Struct Todo List Data Structures=Get Todo List(self=Get Save File()), IndexToRemove:int=@n5) -> (then)
   Save(execute=@n2.then_1)
   > Populate List()
  ]}
 EventGraph {nodes [
   n1 = "On Text Committed (TextBox_EntryInput)"!() -> (OutputDelegate, then, Text:"Text (by ref)", CommitMethod:ETextCommit Enum=Default) {is=title K2Node_ComponentBoundEvent}
   > n2 = Switch on ETextCommit!(execute, Selection:ETextCommit Enum=@n1.CommitMethod) -> (Default:Exec, OnEnter:Exec, OnUserMovedFocus:Exec, OnCleared:Exec) {is=title K2Node_SwitchEnum}
   Add Entry(execute=[@n3.then, @n2.OnEnter]) {is=title K2Node_CallFunction}
   n3 = "On Pressed (Btn_AddEntry)"() {is=title K2Node_ComponentBoundEvent}
   "On Clicked (Btn_Clear_All)"() {is=title K2Node_ComponentBoundEvent}
   > Clear List!(execute, self, All:bool=true) -> (then) {is=title K2Node_CallFunction}
   "On Clicked (Btn_Clear_Completed)"() {is=title K2Node_ComponentBoundEvent}
   > Clear List(All=false)
   "OnTextChanged (Editable Text) (JIRABox)"!() -> (OutputDelegate, then, Text:"Text (by ref)") {is=title K2Node_ComponentBoundEvent}
   > Parse Jira()
   "On Clicked (Btn_OpenJira)"() {is=title K2Node_ComponentBoundEvent}
   > LaunchURL!(
    execute
    URL:String="https://jira.it.epicgames.com/sr/jira.issueviews:searchrequest-rss/temp/SearchRequest.xml?jqlQuery=assignee+%3D+currentUser%28%29+AND+resolution+%3D+Unresolved+order+by+updated+DESC&&tempMax=1000"
   ) -> (then)
   Event Construct()
   > Delay!(execute, WorldContextObject:Object Reference, Duration=0.2) -> (then)
   > n4 = DoesSaveGameExist!(execute, SlotName:String=TodoList, UserIndex:int=0) -> (then, ReturnValue:bool)
   > n5 = Branch(Condition=@n4)
   > n6 = LoadGameFromSlot!(execute, SlotName:String=TodoList, UserIndex:int=0) -> (then, ReturnValue:Save Game Object Reference)
   > n7 = Cast To SaveFile_TodoList!(execute, Object=@n6) -> (then, CastFailed, AsSave File Todo List:Save File Todo List Object Reference) {is=title K2Node_DynamicCast}
   > Set Save File(Save File:Save File Todo List Object Reference=@n7) -> (Output_Get:Save File Todo List Object Reference)
   > n8 = Populate List()
   Set bLoaded(execute=[@n10.then, @n8.then], bLoaded:bool=true) -> (Output_Get:bool)
   n9 = CreateSaveGameObject!(execute=@n5.else, SaveGameClass:Save Game Class Reference=&/SaveFile_TodoList.SaveFile_TodoList_C) -> (then, ReturnValue:Save File Todo List Object Reference)
   > Set Save File(Save File=@n9)
   > Save()
   > n10 = Populate List()
  ]}
 Find Tasks {nodes [
   Find Tasks() {is=title K2Node_FunctionEntry}
   n1 = "To String (Text)"(InText=Get Text(self:Editable Text Object Reference=Get JIRABox() -> (JIRABox:Editor Utility Editable Text Object Reference)))
   > n2 = Branch(Condition=IsEmpty(InString=@n1))
   Set FullText(execute=@n2.else, FullText:String=@n1) -> (Output_Get:String)
   n3 = Get FullText() -> (FullText:String)
   > Set FullText(FullText=RightChop!(SourceString:String=@n3, Count:int=FindSubstring(SearchIn=@n3, Substring="<title>[PLAY", StartPosition=0)) -> (ReturnValue:String))
   > n4 = Set All Blocks(All Blocks:Array of Strings=ParseIntoArray(SourceString=Get FullText(), Delimiter="<title>[PLAY")) -> (Output_Get:Array of Strings)
   > n5 = For Each Loop(Array:Array of Strings=@n4) -> (Array Element:String)
   Return Node!(execute=@n5.Completed, Tasks:Array of Strings=Get Found Tasks() -> (Found Tasks:Array of Strings))
   > Set Current Block(Current Block:String=@n5) -> (Output_Get:String)
   n6 = Get Current Block() -> (Current Block:String)
   > n7 = Set Found Task(Found Task:String=GetSubstring!(SourceString:String=Get Current Block(), StartIndex:int=0, Length:int=FindSubstring(SearchIn=@n6, Substring="</title>")) -> (ReturnValue:String)) -> (Output_Get:String)
   > ReplaceInline!(execute, SourceString:"String (by ref)"=@n7, SearchText:String="&&apos;", ReplacementText:String, SearchCase:ESearchCase Enum) -> (then, ReturnValue:int=0)
   > Add(TargetArray:Array of Strings=Get Found Tasks(), NewItem:"String (by ref)"=Append(A="[PLAY", B=@n7))
   GetSubstring(SourceString=Get FullText(), StartIndex=0, Length=1)
   n8 = Get Current Block()
   n9 = FindSubstring(SearchIn=@n8, Substring="<title>[PLAY", StartPosition=0)
   int + int(A=@n9, B=7)
   int - int(A=FindSubstring(SearchIn=@n8, Substring="</title>", StartPosition=@n9), B=@n9)
   Get Found Tasks()
  ]}
 Get_Btn_AddEntry_bIsEnabled {nodes [
   Get_Btn_AddEntry_bIsEnabled() {is=title K2Node_FunctionEntry}
   n1 = Get Text(self:Text Box Object Reference=Get TextBox_EntryInput())
   > Return Node!(execute, ReturnValue:bool=NOT Boolean(A=TextIsEmpty(InText=@n1)))
  ]}
 Modify Entry {nodes [
   n1 = Modify Entry!() -> (then, Updated Data:Struct Todo List Data Structure) {is=title K2Node_FunctionEntry}
   > n2 = For Each Loop(Array:Array of Struct Todo List Data Structures=Get Todo List(self=Get Save File())) -> (Array Element:Struct Todo List Data Structure)
   n3 = Break Struct Todo List Data(Struct_TodoListData=@n1) -> (
    Completed_1_#1=~
    PriorityTask_9_#5=~
    JiraTask_15_#6=~
   )
   > Branch(Condition="Equal (Guid)"(
     A:"Guid Structure (by ref)"=Break Struct Todo List Data(Struct_TodoListData=@n2) -> (
      Completed_1_#1=~
      PriorityTask_9_#5=~
      JiraTask_15_#6=~
     ).GUID_7_#4
     B:"Guid Structure (by ref)"=@n3.GUID_7_#4
    ) -> (ReturnValue:bool=~) {is=title K2Node_PromotableOperator})
   n4 = Get Save File()
   > Set members in Struct Todo List Data!(
    execute:Exec
    StructRef:"Struct Todo List Data Structure (by ref)"="Get (a ref)"(Array:Array of Struct Todo List Data Structures=Get Todo List(self=@n4), Dimension 1=@n2.Array Index) -> (Output:"Struct Todo List Data Structure (by ref)")
    Completed_1_#1:bool=@n3
    Message_4_#2:Text=@n3.Message_4_#2
    PriorityTask_9_#5:bool=@n3.PriorityTask_9_#5
   ) -> (then:Exec, StructOut:"Struct Todo List Data Structure (by ref)") {is=title K2Node_SetFieldsInStruct}
   > Save()
   > Populate List()
   > Return Node!(execute)
  ]}
 Parse Jira {nodes [
   Parse Jira() {is=title K2Node_FunctionEntry}
   > Branch(Condition=Get bLoaded())
   n1 = "To String (Text)"(InText=Get Text(self:Editable Text Object Reference=Get JIRABox()))
   > n2 = Branch(Condition="integer > integer"(A=Len(S=@n1), B=1))
   > n3 = Sequence()
   > n4 = Find Tasks!(execute, self) -> (then, Tasks:Array of Strings) {is=title K2Node_CallFunction}
   > n5 = For Each Loop(Array:Array of Strings=@n4) -> (Array Element:String)
   > n6 = Branch(Condition=Check is task already in list!(Task:String=@n5) -> (Already Present:bool) {is=title K2Node_CallFunction})
   n7 = Get(execute=@n6.else)
   > Add(
    TargetArray:Array of Struct Todo List Data Structures=Get Todo List(self=@n7)
    NewItem:"Struct Todo List Data Structure (by ref)"=Make Struct Todo List Data(
     Completed_1_#1=False
     Message_4_#2="To Text (String)"(InString=@n5)
     GUID_7_#4=NewGuid()
     PriorityTask_9_#5=False
     JiraTask_15_#6=true
    )
   )
   Set Text(execute=@n3.then_1, self:Editable Text Object Reference=Get JIRABox())
   > Save()
   > Populate List()
   Set Text(execute=@n2.else, self=Get JIRABox())
  ]}
 Populate List {nodes [
   Populate List() {is=title K2Node_FunctionEntry}
   > ClearChildren(self=Get TodoScrollList() -> (TodoScrollList:Common Hierarchical Scroll Box Object Reference))
   > n1 = Get()
   > n2 = For Each Loop(Array:Array of Struct Todo List Data Structures=Get Todo List(self=@n1)) -> (Array Element:Struct Todo List Data Structure)
   n3 = For Each Loop(
    Exec=@n2.Completed
    Array:Array of EUW Todo List Entry Object References=Get Temp Array() -> (Temp Array:Array of EUW Todo List Entry Object References)
   ) -> (Array Element:EUW Todo List Entry Object Reference)
   > AddChild(self=Get TodoScrollList(), Content=@n3)
   Set Todo Item(execute=@n2.LoopBody, Todo Item:Struct Todo List Data Structure=@n2) -> (Output_Get:Struct Todo List Data Structure)
   > n4 = Create Widget!(
    execute:Exec
    WorldContextObject:Object Reference
    Class:User Widget Class Reference=&/EUW_TodoList_Entry.EUW_TodoList_Entry_C
    OwningPlayer:Player Controller Object Reference
    Todo Data:Struct Todo List Data Structure=Get Todo Item() -> (Todo Item:Struct Todo List Data Structure)
    Parent:To-do List Object Reference=Self-Reference()
   ) -> (then:Exec, ReturnValue:EUW Todo List Entry Object Reference)
   > n5 = Branch(Condition=Break Struct Todo List Data!(Struct_TodoListData=Get Todo Item()) -> (PriorityTask_9_#5=~))
   n6 = Get Temp Array()
   > Insert!(
    execute
    TargetArray:Array of EUW Todo List Entry Object References=@n6
    NewItem:"EUW Todo List Entry Object Reference (by ref)"=@n4
    Index:int=0
   ) -> (then) {is=title K2Node_CallArrayFunction}
   Add(
    execute=@n5.else
    TargetArray:Array of EUW Todo List Entry Object References=@n6
    NewItem:"EUW Todo List Entry Object Reference (by ref)"=@n4
   )
  ]}
 Remove Entry {nodes [
   n1 = Remove Entry!() -> (then, GUID:Guid Structure) {is=title K2Node_FunctionEntry}
   > n2 = For Each Loop(Array:Array of Struct Todo List Data Structures=Get Todo List(self=Get Save File())) -> (Array Element:Struct Todo List Data Structure)
   > Branch(Condition="Equal (Guid)"(
     A=Break Struct Todo List Data!(Struct_TodoListData=@n2) -> (
      Completed_1_#1:bool=~
      Message_4_#2:Text
      Tooltip_19_#3:Text
      GUID_7_#4:Guid Structure
      PriorityTask_9_#5=~
      JiraTask_15_#6:bool=~
      JiraAddress_12_#7:String
     ).GUID_7_#4
     B=@n1
    ) -> (ReturnValue=~))
   > Remove Index(TargetArray=Get Todo List(self=Get Save File()), IndexToRemove=@n2.Array Index)
   > Save()
   > Populate List() {is=title K2Node_CallFunction}
   > Return Node()
  ]}
 Save {nodes [
   Save() {is=title K2Node_FunctionEntry}
   > SaveGameToSlot!(execute, SaveGameObject:Save Game Object Reference=Get Save File(), SlotName:String=TodoList, UserIndex:int=0) -> (then, ReturnValue:bool)
  ]}
]
variables [
 bLoaded {type=boolean}
 Save File {type=SaveFile_TodoList_C, default=null}
]
widgets {
 info {rootWidgetClass=&StackBox, widgetCount=12}
 widgets [
  StackBox_0 {widgetClassPath=&StackBox, widget=StackBox_0, properties {orientation=Orient_Vertical}}
  AddEntryBox {
   widgetClassPath=&StackBox
   parent=StackBox_0
   widget=AddEntryBox
   slot {class=StackBoxSlot, properties {padding=<ltrb 10 10 10 10>}}
  }
  TextBox_EntryInput {
   widgetClassPath=&EditorUtilityEditableTextBox
   bIsVariable=true
   parent=AddEntryBox
   widget=TextBox_EntryInput
   properties {
    clearKeyboardFocusOnCommit=false
    hintText=Add entry
    revertTextOnEscape=true
    selectAllTextWhenFocused=true
    virtualKeyboardOptions {bEnableAutocorrect=true}
    widgetStyle {
     backgroundColor {colorUseRule=UseColor_Specified, specifiedColor=<rgba 1 1 1 1>}
     backgroundImageFocused {
      bIsDynamicallyLoaded=false
      drawAs=RoundedBox
      imageType=NoImage
      mirroring=NoMirror
      resourceName=None
      resourceObject=null
      tiling=NoTile
      imageSize=<0 0>
      margin=<ltrb 0 0 0 0>
      outlineSettings {
       bUseBrushTransparency=false
       roundingType=FixedRadius
       width=1
       color {colorUseRule=UseColor_ColorTable, specifiedColor=<rgba 0 0 0 0>}
       cornerRadii=<4 4 4 4>
      }
      tintColor {colorUseRule=UseColor_Specified, specifiedColor=<rgba 0.0048 0.0048 0.0048 1>}
      uVRegion {bIsValid=false, max=<0 0>, min=<0 0>}
     }
     backgroundImageHovered {
      bIsDynamicallyLoaded=false
      drawAs=RoundedBox
      imageType=NoImage
      mirroring=NoMirror
      resourceName=None
      resourceObject=null
      tiling=NoTile
      imageSize=<0 0>
      margin=<ltrb 0 0 0 0>
      outlineSettings {
       bUseBrushTransparency=false
       roundingType=FixedRadius
       width=1
       color {colorUseRule=UseColor_ColorTable, specifiedColor=<rgba 0 0 0 0>}
       cornerRadii=<4 4 4 4>
      }
      tintColor {colorUseRule=UseColor_Specified, specifiedColor=<rgba 0.0048 0.0048 0.0048 1>}
      uVRegion {bIsValid=false, max=<0 0>, min=<0 0>}
     }
     backgroundImageNormal {
      bIsDynamicallyLoaded=false
      drawAs=RoundedBox
      imageType=NoImage
      mirroring=NoMirror
      resourceName=None
      resourceObject=null
      tiling=NoTile
      imageSize=<0 0>
      margin=<ltrb 0 0 0 0>
      outlineSettings {
       bUseBrushTransparency=false
       roundingType=FixedRadius
       width=1
       color {colorUseRule=UseColor_ColorTable, specifiedColor=<rgba 0 0 0 0>}
       cornerRadii=<4 4 4 4>
      }
      tintColor {colorUseRule=UseColor_Specified, specifiedColor=<rgba 0.0048 0.0048 0.0048 1>}
      uVRegion {bIsValid=false, max=<0 0>, min=<0 0>}
     }
     backgroundImageReadOnly {
      bIsDynamicallyLoaded=false
      drawAs=RoundedBox
      imageType=NoImage
      mirroring=NoMirror
      resourceName=None
      resourceObject=null
      tiling=NoTile
      imageSize=<0 0>
      margin=<ltrb 0 0 0 0>
      outlineSettings {
       bUseBrushTransparency=false
       roundingType=FixedRadius
       width=1
       color {colorUseRule=UseColor_ColorTable, specifiedColor=<rgba 0 0 0 0>}
       cornerRadii=<4 4 4 4>
      }
      tintColor {colorUseRule=UseColor_Specified, specifiedColor=<rgba 0.0284 0.0284 0.0284 1>}
      uVRegion {bIsValid=false, max=<0 0>, min=<0 0>}
     }
     focusedForegroundColor {colorUseRule=UseColor_Specified, specifiedColor=<rgba 1 1 1 1>}
     foregroundColor {colorUseRule=UseColor_Specified, specifiedColor=<rgba 0.5271 0.5271 0.5271 1>}
     hScrollBarPadding=<ltrb 0 0 0 0>
     padding=<ltrb 8 3 8 4>
     readOnlyForegroundColor {colorUseRule=UseColor_Foreground, specifiedColor=<rgba 1 0 1 1>}
     scrollBarStyle {
      thickness=8
      draggedThumbImage {
       bIsDynamicallyLoaded=false
       drawAs=RoundedBox
       imageType=NoImage
       mirroring=NoMirror
       resourceName=None
       resourceObject=null
       tiling=NoTile
       imageSize=<0 0>
       margin=<ltrb 0 0 0 0>
       outlineSettings {
        bUseBrushTransparency=false
        roundingType=FixedRadius
        width=0
        color {colorUseRule=UseColor_Specified, specifiedColor=<rgba 0 0 0 0>}
        cornerRadii=<4 4 4 4>
       }
       tintColor {colorUseRule=UseColor_Specified, specifiedColor=<rgba 0.2159 0.2159 0.2159 1>}
       uVRegion {bIsValid=false, max=<0 0>, min=<0 0>}
      }
      horizontalBackgroundImage {
       bIsDynamicallyLoaded=false
       drawAs=NoDrawType
       imageType=NoImage
       mirroring=NoMirror
       resourceName=None
       resourceObject=null
       tiling=NoTile
       imageSize=<0 0>
       margin=<ltrb 0 0 0 0>
       outlineSettings {
        bUseBrushTransparency=false
        roundingType=HalfHeightRadius
        width=0
        color {colorUseRule=UseColor_Specified, specifiedColor=<rgba 0 0 0 0>}
        cornerRadii=<0 0 0 0>
       }
       tintColor {colorUseRule=UseColor_Specified, specifiedColor=<rgba 1 1 1 1>}
       uVRegion {bIsValid=false, max=<0 0>, min=<0 0>}
      }
      horizontalBottomSlotImage {
       bIsDynamicallyLoaded=false
       drawAs=NoDrawType
       imageType=NoImage
       mirroring=NoMirror
       resourceName=None
       resourceObject=null
       tiling=NoTile
       imageSize=<0 0>
       margin=<ltrb 0 0 0 0>
       outlineSettings {
        bUseBrushTransparency=false
        roundingType=HalfHeightRadius
        width=0
        color {colorUseRule=UseColor_Specified, specifiedColor=<rgba 0 0 0 0>}
        cornerRadii=<0 0 0 0>
       }
       tintColor {colorUseRule=UseColor_Specified, specifiedColor=<rgba 1 1 1 1>}
       uVRegion {bIsValid=false, max=<0 0>, min=<0 0>}
      }
      horizontalTopSlotImage {
       bIsDynamicallyLoaded=false
       drawAs=NoDrawType
       imageType=NoImage
       mirroring=NoMirror
       resourceName=None
       resourceObject=null
       tiling=NoTile
       imageSize=<0 0>
       margin=<ltrb 0 0 0 0>
       outlineSettings {
        bUseBrushTransparency=false
        roundingType=HalfHeightRadius
        width=0
        color {colorUseRule=UseColor_Specified, specifiedColor=<rgba 0 0 0 0>}
        cornerRadii=<0 0 0 0>
       }
       tintColor {colorUseRule=UseColor_Specified, specifiedColor=<rgba 1 1 1 1>}
       uVRegion {bIsValid=false, max=<0 0>, min=<0 0>}
      }
      hoveredThumbImage {
       bIsDynamicallyLoaded=false
       drawAs=RoundedBox
       imageType=NoImage
       mirroring=NoMirror
       resourceName=None
       resourceObject=null
       tiling=NoTile
       imageSize=<0 0>
       margin=<ltrb 0 0 0 0>
       outlineSettings {
        bUseBrushTransparency=false
        roundingType=FixedRadius
        width=0
        color {colorUseRule=UseColor_Specified, specifiedColor=<rgba 0 0 0 0>}
        cornerRadii=<4 4 4 4>
       }
       tintColor {colorUseRule=UseColor_Specified, specifiedColor=<rgba 0.2159 0.2159 0.2159 1>}
       uVRegion {bIsValid=false, max=<0 0>, min=<0 0>}
      }
      normalThumbImage {
       bIsDynamicallyLoaded=false
       drawAs=RoundedBox
       imageType=NoImage
       mirroring=NoMirror
       resourceName=None
       resourceObject=null
       tiling=NoTile
       imageSize=<0 0>
       margin=<ltrb 0 0 0 0>
       outlineSettings {
        bUseBrushTransparency=false
        roundingType=FixedRadius
        width=0
        color {colorUseRule=UseColor_Specified, specifiedColor=<rgba 0 0 0 0>}
        cornerRadii=<4 4 4 4>
       }
       tintColor {colorUseRule=UseColor_Specified, specifiedColor=<rgba 0.0953 0.0953 0.0953 1>}
       uVRegion {bIsValid=false, max=<0 0>, min=<0 0>}
      }
      verticalBackgroundImage {
       bIsDynamicallyLoaded=false
       drawAs=NoDrawType
       imageType=NoImage
       mirroring=NoMirror
       resourceName=None
       resourceObject=null
       tiling=NoTile
       imageSize=<0 0>
       margin=<ltrb 0 0 0 0>
       outlineSettings {
        bUseBrushTransparency=false
        roundingType=HalfHeightRadius
        width=0
        color {colorUseRule=UseColor_Specified, specifiedColor=<rgba 0 0 0 0>}
        cornerRadii=<0 0 0 0>
       }
       tintColor {colorUseRule=UseColor_Specified, specifiedColor=<rgba 1 1 1 1>}
       uVRegion {bIsValid=false, max=<0 0>, min=<0 0>}
      }
      verticalBottomSlotImage {
       bIsDynamicallyLoaded=false
       drawAs=NoDrawType
       imageType=NoImage
       mirroring=NoMirror
       resourceName=None
       resourceObject=null
       tiling=NoTile
       imageSize=<0 0>
       margin=<ltrb 0 0 0 0>
       outlineSettings {
        bUseBrushTransparency=false
        roundingType=HalfHeightRadius
        width=0
        color {colorUseRule=UseColor_Specified, specifiedColor=<rgba 0 0 0 0>}
        cornerRadii=<0 0 0 0>
       }
       tintColor {colorUseRule=UseColor_Specified, specifiedColor=<rgba 1 1 1 1>}
       uVRegion {bIsValid=false, max=<0 0>, min=<0 0>}
      }
      verticalTopSlotImage {
       bIsDynamicallyLoaded=false
       drawAs=NoDrawType
       imageType=NoImage
       mirroring=NoMirror
       resourceName=None
       resourceObject=null
       tiling=NoTile
       imageSize=<0 0>
       margin=<ltrb 0 0 0 0>
       outlineSettings {
        bUseBrushTransparency=false
        roundingType=HalfHeightRadius
        width=0
        color {colorUseRule=UseColor_Specified, specifiedColor=<rgba 0 0 0 0>}
        cornerRadii=<0 0 0 0>
       }
       tintColor {colorUseRule=UseColor_Specified, specifiedColor=<rgba 1 1 1 1>}
       uVRegion {bIsValid=false, max=<0 0>, min=<0 0>}
      }
     }
     textStyle {
      overflowPolicy=Clip
      transformPolicy=None
      colorAndOpacity {colorUseRule=UseColor_Foreground, specifiedColor=<rgba 1 0 1 1>}
      font {
       bForceMonospaced=false
       bMaterialIsStencil=false
       fontMaterial=null
       fontObject=null
       letterSpacing=0
       monospacedWidth=1
       size=10
       skewAmount=0
       typefaceFontName=Regular
       outlineSettings {
        bApplyOutlineToDropShadows=false
        bMiteredCorners=false
        bSeparateFillAlpha=false
        outlineMaterial=null
        outlineSize=0
        outlineColor=<rgba 0 0 0 1>
       }
      }
      highlightColor {colorUseRule=UseColor_ColorTable, specifiedColor=<rgba 0 0 0 0>}
      highlightShape {
       bIsDynamicallyLoaded=false
       drawAs=Image
       imageType=NoImage
       mirroring=NoMirror
       resourceName=None
       resourceObject=null
       tiling=NoTile
       imageSize=<0 0>
       margin=<ltrb 0 0 0 0>
       outlineSettings {
        bUseBrushTransparency=false
        roundingType=HalfHeightRadius
        width=0
        color {colorUseRule=UseColor_Specified, specifiedColor=<rgba 0 0 0 0>}
        cornerRadii=<0 0 0 0>
       }
       tintColor {colorUseRule=UseColor_ColorTable, specifiedColor=<rgba 0 0 0 0>}
       uVRegion {bIsValid=false, max=<0 0>, min=<0 0>}
      }
      selectedBackgroundColor {colorUseRule=UseColor_ColorTable, specifiedColor=<rgba 0 0 0 0>}
      shadowColorAndOpacity=<rgba 0 0 0 1>
      shadowOffset=<0 0>
      strikeBrush {
       bIsDynamicallyLoaded=false
       drawAs=Image
       imageType=NoImage
       mirroring=NoMirror
       resourceName=None
       resourceObject=null
       tiling=NoTile
       imageSize=<32 32>
       margin=<ltrb 0 0 0 0>
       outlineSettings {
        bUseBrushTransparency=false
        roundingType=HalfHeightRadius
        width=0
        color {colorUseRule=UseColor_Specified, specifiedColor=<rgba 0 0 0 0>}
        cornerRadii=<0 0 0 0>
       }
       tintColor {colorUseRule=UseColor_Specified, specifiedColor=<rgba 1 1 1 1>}
       uVRegion {bIsValid=false, max=<0 0>, min=<0 0>}
      }
      underlineBrush {
       bIsDynamicallyLoaded=false
       drawAs=Image
       imageType=NoImage
       mirroring=NoMirror
       resourceName=None
       resourceObject=null
       tiling=NoTile
       imageSize=<32 32>
       margin=<ltrb 0 0 0 0>
       outlineSettings {
        bUseBrushTransparency=false
        roundingType=HalfHeightRadius
        width=0
        color {colorUseRule=UseColor_Specified, specifiedColor=<rgba 0 0 0 0>}
        cornerRadii=<0 0 0 0>
       }
       tintColor {colorUseRule=UseColor_Specified, specifiedColor=<rgba 1 1 1 1>}
       uVRegion {bIsValid=false, max=<0 0>, min=<0 0>}
      }
     }
     vScrollBarPadding=<ltrb 0 0 0 0>
    }
   }
   slot {class=StackBoxSlot, properties {size {sizeRule=Fill, value=1}}}
  }
  Btn_AddEntry {
   widgetClassPath=&EditorUtilityButton
   parent=AddEntryBox
   widget=Btn_AddEntry
   properties {
    bOverride_Cursor=true
    cursor=Hand
    widgetStyle {
     clickedSlateSound {resourceObject=null}
     disabled {
      bIsDynamicallyLoaded=false
      drawAs=Image
      imageType=FullColor
      mirroring=NoMirror
      resourceName=None
      resourceObject=/Game/UI/Foundation/Textures/Icons/Replay/T-Icon-Replay-Plus-L
      tiling=NoTile
      imageSize=<23 23>
      margin=<ltrb 0 0 0 0>
      outlineSettings {
       bUseBrushTransparency=false
       roundingType=FixedRadius
       width=1
       color {colorUseRule=UseColor_ColorTable, specifiedColor=<rgba 0 0 0 0>}
       cornerRadii=<4 4 4 4>
      }
      tintColor {colorUseRule=UseColor_Specified, specifiedColor=<rgba 0.1562 0.1562 0.1562 1>}
      uVRegion {bIsValid=false, max=<0 0>, min=<0 0>}
     }
     disabledForeground {colorUseRule=UseColor_ColorTable, specifiedColor=<rgba 0 0 0 0>}
     hovered {
      bIsDynamicallyLoaded=false
      drawAs=Image
      imageType=FullColor
      mirroring=NoMirror
      resourceName=None
      resourceObject=&T-Icon-Replay-Plus-L
      tiling=NoTile
      imageSize=<23 23>
      margin=<ltrb 0 0 0 0>
      outlineSettings {
       bUseBrushTransparency=false
       roundingType=FixedRadius
       width=1
       color {colorUseRule=UseColor_ColorTable, specifiedColor=<rgba 0 0 0 0>}
       cornerRadii=<4 4 4 4>
      }
      tintColor {colorUseRule=UseColor_Specified, specifiedColor=<rgba 0.8229 0.8229 0.8229 1>}
      uVRegion {bIsValid=false, max=<0 0>, min=<0 0>}
     }
     hoveredForeground {colorUseRule=UseColor_ColorTable, specifiedColor=<rgba 0 0 0 0>}
     hoveredSlateSound {resourceObject=null}
     normal {
      bIsDynamicallyLoaded=false
      drawAs=Image
      imageType=FullColor
      mirroring=NoMirror
      resourceName=None
      resourceObject=&T-Icon-Replay-Plus-L
      tiling=NoTile
      imageSize=<23 23>
      margin=<ltrb 0 0 0 0>
      outlineSettings {
       bUseBrushTransparency=false
       roundingType=FixedRadius
       width=1
       color {colorUseRule=UseColor_ColorTable, specifiedColor=<rgba 0 0 0 0>}
       cornerRadii=<4 4 4 4>
      }
      tintColor {colorUseRule=UseColor_Specified, specifiedColor=<rgba 0.4323 0.4323 0.4323 1>}
      uVRegion {bIsValid=false, max=<0 0>, min=<0 0>}
     }
     normalForeground {colorUseRule=UseColor_ColorTable, specifiedColor=<rgba 0 0 0 0>}
     normalPadding=<ltrb 12 1.5 12 1.5>
     pressed {
      bIsDynamicallyLoaded=false
      drawAs=Image
      imageType=FullColor
      mirroring=NoMirror
      resourceName=None
      resourceObject=&T-Icon-Replay-Plus-L
      tiling=NoTile
      imageSize=<23 23>
      margin=<ltrb 0 0 0 0>
      outlineSettings {
       bUseBrushTransparency=false
       roundingType=FixedRadius
       width=1
       color {colorUseRule=UseColor_ColorTable, specifiedColor=<rgba 0 0 0 0>}
       cornerRadii=<4 4 4 4>
      }
      tintColor {colorUseRule=UseColor_Specified, specifiedColor=<rgba 1 1 1 1>}
      uVRegion {bIsValid=false, max=<0 0>, min=<0 0>}
     }
     pressedForeground {colorUseRule=UseColor_ColorTable, specifiedColor=<rgba 0 0 0 0>}
     pressedPadding=<ltrb 12 2.5 12 0.5>
     pressedSlateSound {resourceObject=null}
    }
   }
   slot {class=StackBoxSlot, properties {padding=<ltrb 5 0 20 0>}}
  }
  Btn_OpenJira {
   widgetClassPath=&EditorUtilityButton
   parent=AddEntryBox
   widget=Btn_OpenJira
   properties {
    bOverride_Cursor=true
    cursor=Hand
    visibility=Collapsed
    widgetStyle {
     clickedSlateSound {resourceObject=null}
     disabled {
      bIsDynamicallyLoaded=false
      drawAs=RoundedBox
      imageType=NoImage
      mirroring=NoMirror
      resourceName=None
      resourceObject=null
      tiling=NoTile
      imageSize=<0 0>
      margin=<ltrb 0 0 0 0>
      outlineSettings {
       bUseBrushTransparency=false
       roundingType=FixedRadius
       width=1
       color {colorUseRule=UseColor_ColorTable, specifiedColor=<rgba 0 0 0 0>}
       cornerRadii=<4 4 4 4>
      }
      tintColor {colorUseRule=UseColor_Specified, specifiedColor=<rgba 0.0395 0.0395 0.0395 1>}
      uVRegion {bIsValid=false, max=<0 0>, min=<0 0>}
     }
     disabledForeground {colorUseRule=UseColor_ColorTable, specifiedColor=<rgba 0 0 0 0>}
     hovered {
      bIsDynamicallyLoaded=false
      drawAs=Image
      imageType=FullColor
      mirroring=NoMirror
      resourceName=None
      resourceObject=/Editor_TodoList/Textures/T_Todo_Jira
      tiling=NoTile
      imageSize=<18.4131 25>
      margin=<ltrb 0 0 0 0>
      outlineSettings {
       bUseBrushTransparency=false
       roundingType=FixedRadius
       width=1
       color {colorUseRule=UseColor_ColorTable, specifiedColor=<rgba 0 0 0 0>}
       cornerRadii=<4 4 4 4>
      }
      tintColor {colorUseRule=UseColor_Specified, specifiedColor=<rgba 1 1 1 1>}
      uVRegion {bIsValid=false, max=<0 0>, min=<0 0>}
     }
     hoveredForeground {colorUseRule=UseColor_ColorTable, specifiedColor=<rgba 0 0 0 0>}
     hoveredSlateSound {resourceObject=null}
     normal {
      bIsDynamicallyLoaded=false
      drawAs=Image
      imageType=FullColor
      mirroring=NoMirror
      resourceName=None
      resourceObject=&T_Todo_Jira
      tiling=NoTile
      imageSize=<18.4131 25>
      margin=<ltrb 0 0 0 0>
      outlineSettings {
       bUseBrushTransparency=false
       roundingType=FixedRadius
       width=1
       color {colorUseRule=UseColor_ColorTable, specifiedColor=<rgba 0 0 0 0>}
       cornerRadii=<4 4 4 4>
      }
      tintColor {colorUseRule=UseColor_Specified, specifiedColor=<rgba 1 1 1 1>}
      uVRegion {bIsValid=false, max=<0 0>, min=<0 0>}
     }
     normalForeground {colorUseRule=UseColor_ColorTable, specifiedColor=<rgba 0 0 0 0>}
     normalPadding=<ltrb 12 1.5 12 1.5>
     pressed {
      bIsDynamicallyLoaded=false
      drawAs=Image
      imageType=FullColor
      mirroring=NoMirror
      resourceName=None
      resourceObject=&T_Todo_Jira
      tiling=NoTile
      imageSize=<18.4131 25>
      margin=<ltrb 0 0 0 0>
      outlineSettings {
       bUseBrushTransparency=false
       roundingType=FixedRadius
       width=1
       color {colorUseRule=UseColor_ColorTable, specifiedColor=<rgba 0 0 0 0>}
       cornerRadii=<4 4 4 4>
      }
      tintColor {colorUseRule=UseColor_Specified, specifiedColor=<rgba 1 1 1 1>}
      uVRegion {bIsValid=false, max=<0 0>, min=<0 0>}
     }
     pressedForeground {colorUseRule=UseColor_ColorTable, specifiedColor=<rgba 0 0 0 0>}
     pressedPadding=<ltrb 12 2.5 12 0.5>
     pressedSlateSound {resourceObject=null}
    }
   }
   slot {class=StackBoxSlot, properties {horizontalAlignment=HAlign_Center, verticalAlignment=VAlign_Center}}
  }
  JIRABox {
   widgetClassPath=/Script/Blutility.EditorUtilityEditableText
   bIsVariable=true
   parent=AddEntryBox
   widget=JIRABox
   properties {
    hintText=PASTE JIRA HERE
    visibility=Collapsed
    widgetStyle {
     backgroundImageComposing {
      bIsDynamicallyLoaded=false
      drawAs=Border
      imageType=FullColor
      mirroring=NoMirror
      resourceName=../../../Engine/Content/Slate/Old/HyperlinkDotted.png
      resourceObject=null
      tiling=Both
      imageSize=<0 0>
      margin=<ltrb 0 0 0 0.1875>
      outlineSettings {
       bUseBrushTransparency=false
       roundingType=HalfHeightRadius
       width=0
       color {colorUseRule=UseColor_Specified, specifiedColor=<rgba 0 0 0 0>}
       cornerRadii=<0 0 0 0>
      }
      tintColor {colorUseRule=UseColor_Specified, specifiedColor=<rgba 1 1 1 1>}
      uVRegion {bIsValid=false, max=<0 0>, min=<0 0>}
     }
     backgroundImageSelected {
      bIsDynamicallyLoaded=false
      drawAs=Image
      imageType=NoImage
      mirroring=NoMirror
      resourceName=None
      resourceObject=null
      tiling=NoTile
      imageSize=<0 0>
      margin=<ltrb 0 0 0 0>
      outlineSettings {
       bUseBrushTransparency=false
       roundingType=HalfHeightRadius
       width=0
       color {colorUseRule=UseColor_Specified, specifiedColor=<rgba 0 0 0 0>}
       cornerRadii=<0 0 0 0>
      }
      tintColor {colorUseRule=UseColor_Specified, specifiedColor=<rgba 0 0.162 0.7454 1>}
      uVRegion {bIsValid=false, max=<0 0>, min=<0 0>}
     }
     caretImage {
      bIsDynamicallyLoaded=false
      drawAs=Image
      imageType=NoImage
      mirroring=NoMirror
      resourceName=None
      resourceObject=null
      tiling=NoTile
      imageSize=<0 0>
      margin=<ltrb 0 0 0 0>
      outlineSettings {
       bUseBrushTransparency=false
       roundingType=HalfHeightRadius
       width=0
       color {colorUseRule=UseColor_Specified, specifiedColor=<rgba 0 0 0 0>}
       cornerRadii=<0 0 0 0>
      }
      tintColor {colorUseRule=UseColor_Specified, specifiedColor=<rgba 1 1 1 1>}
      uVRegion {bIsValid=false, max=<0 0>, min=<0 0>}
     }
     colorAndOpacity {colorUseRule=UseColor_Foreground, specifiedColor=<rgba 1 0 1 1>}
     font {
      bForceMonospaced=false
      bMaterialIsStencil=false
      fontMaterial=null
      fontObject=&NotoSans
      letterSpacing=0
      monospacedWidth=1
      size=9
      skewAmount=0
      typefaceFontName=NotoSans-Regular
      outlineSettings {
       bApplyOutlineToDropShadows=false
       bMiteredCorners=false
       bSeparateFillAlpha=false
       outlineMaterial=null
       outlineSize=0
       outlineColor=<rgba 0 0 0 1>
      }
     }
     hintColorAndOpacity {colorUseRule=UseColor_Foreground, specifiedColor=<rgba 1 0 1 1>}
     selectedBackgroundColor {colorUseRule=UseColor_Foreground, specifiedColor=<rgba 1 0 1 1>}
    }
   }
   slot {class=StackBoxSlot, properties {verticalAlignment=VAlign_Center}}
  }
  TodoScrollList {
   widgetClassPath=/Script/CommonUI.CommonHierarchicalScrollBox
   bIsVariable=true
   parent=StackBox_0
   widget=TodoScrollList
   properties {
    consumeMouseWheel=Always
    scrollbarThickness=<3 9>
    widgetBarStyle {
     thickness=8
     draggedThumbImage {
      bIsDynamicallyLoaded=false
      drawAs=RoundedBox
      imageType=NoImage
      mirroring=NoMirror
      resourceName=None
      resourceObject=null
      tiling=NoTile
      imageSize=<0 0>
      margin=<ltrb 0 0 0 0>
      outlineSettings {
       bUseBrushTransparency=false
       roundingType=FixedRadius
       width=0
       color {colorUseRule=UseColor_Specified, specifiedColor=<rgba 0 0 0 0>}
       cornerRadii=<4 4 4 4>
      }
      tintColor {colorUseRule=UseColor_Specified, specifiedColor=<rgba 0.2159 0.2159 0.2159 1>}
      uVRegion {bIsValid=false, max=<0 0>, min=<0 0>}
     }
     horizontalBackgroundImage {
      bIsDynamicallyLoaded=false
      drawAs=NoDrawType
      imageType=NoImage
      mirroring=NoMirror
      resourceName=None
      resourceObject=null
      tiling=NoTile
      imageSize=<0 0>
      margin=<ltrb 0 0 0 0>
      outlineSettings {
       bUseBrushTransparency=false
       roundingType=HalfHeightRadius
       width=0
       color {colorUseRule=UseColor_Specified, specifiedColor=<rgba 0 0 0 0>}
       cornerRadii=<0 0 0 0>
      }
      tintColor {colorUseRule=UseColor_Specified, specifiedColor=<rgba 1 1 1 1>}
      uVRegion {bIsValid=false, max=<0 0>, min=<0 0>}
     }
     horizontalBottomSlotImage {
      bIsDynamicallyLoaded=false
      drawAs=NoDrawType
      imageType=NoImage
      mirroring=NoMirror
      resourceName=None
      resourceObject=null
      tiling=NoTile
      imageSize=<0 0>
      margin=<ltrb 0 0 0 0>
      outlineSettings {
       bUseBrushTransparency=false
       roundingType=HalfHeightRadius
       width=0
       color {colorUseRule=UseColor_Specified, specifiedColor=<rgba 0 0 0 0>}
       cornerRadii=<0 0 0 0>
      }
      tintColor {colorUseRule=UseColor_Specified, specifiedColor=<rgba 1 1 1 1>}
      uVRegion {bIsValid=false, max=<0 0>, min=<0 0>}
     }
     horizontalTopSlotImage {
      bIsDynamicallyLoaded=false
      drawAs=NoDrawType
      imageType=NoImage
      mirroring=NoMirror
      resourceName=None
      resourceObject=null
      tiling=NoTile
      imageSize=<0 0>
      margin=<ltrb 0 0 0 0>
      outlineSettings {
       bUseBrushTransparency=false
       roundingType=HalfHeightRadius
       width=0
       color {colorUseRule=UseColor_Specified, specifiedColor=<rgba 0 0 0 0>}
       cornerRadii=<0 0 0 0>
      }
      tintColor {colorUseRule=UseColor_Specified, specifiedColor=<rgba 1 1 1 1>}
      uVRegion {bIsValid=false, max=<0 0>, min=<0 0>}
     }
     hoveredThumbImage {
      bIsDynamicallyLoaded=false
      drawAs=RoundedBox
      imageType=NoImage
      mirroring=NoMirror
      resourceName=None
      resourceObject=null
      tiling=NoTile
      imageSize=<0 0>
      margin=<ltrb 0 0 0 0>
      outlineSettings {
       bUseBrushTransparency=false
       roundingType=FixedRadius
       width=0
       color {colorUseRule=UseColor_Specified, specifiedColor=<rgba 0 0 0 0>}
       cornerRadii=<4 4 4 4>
      }
      tintColor {colorUseRule=UseColor_Specified, specifiedColor=<rgba 0.2159 0.2159 0.2159 1>}
      uVRegion {bIsValid=false, max=<0 0>, min=<0 0>}
     }
     normalThumbImage {
      bIsDynamicallyLoaded=false
      drawAs=RoundedBox
      imageType=NoImage
      mirroring=NoMirror
      resourceName=None
      resourceObject=null
      tiling=NoTile
      imageSize=<0 0>
      margin=<ltrb 0 0 0 0>
      outlineSettings {
       bUseBrushTransparency=false
       roundingType=FixedRadius
       width=0
       color {colorUseRule=UseColor_Specified, specifiedColor=<rgba 0 0 0 0>}
       cornerRadii=<4 4 4 4>
      }
      tintColor {colorUseRule=UseColor_Specified, specifiedColor=<rgba 0.0953 0.0953 0.0953 1>}
      uVRegion {bIsValid=false, max=<0 0>, min=<0 0>}
     }
     verticalBackgroundImage {
      bIsDynamicallyLoaded=false
      drawAs=NoDrawType
      imageType=NoImage
      mirroring=NoMirror
      resourceName=None
      resourceObject=null
      tiling=NoTile
      imageSize=<0 0>
      margin=<ltrb 0 0 0 0>
      outlineSettings {
       bUseBrushTransparency=false
       roundingType=HalfHeightRadius
       width=0
       color {colorUseRule=UseColor_Specified, specifiedColor=<rgba 0 0 0 0>}
       cornerRadii=<0 0 0 0>
      }
      tintColor {colorUseRule=UseColor_Specified, specifiedColor=<rgba 1 1 1 1>}
      uVRegion {bIsValid=false, max=<0 0>, min=<0 0>}
     }
     verticalBottomSlotImage {
      bIsDynamicallyLoaded=false
      drawAs=NoDrawType
      imageType=NoImage
      mirroring=NoMirror
      resourceName=None
      resourceObject=null
      tiling=NoTile
      imageSize=<0 0>
      margin=<ltrb 0 0 0 0>
      outlineSettings {
       bUseBrushTransparency=false
       roundingType=HalfHeightRadius
       width=0
       color {colorUseRule=UseColor_Specified, specifiedColor=<rgba 0 0 0 0>}
       cornerRadii=<0 0 0 0>
      }
      tintColor {colorUseRule=UseColor_Specified, specifiedColor=<rgba 1 1 1 1>}
      uVRegion {bIsValid=false, max=<0 0>, min=<0 0>}
     }
     verticalTopSlotImage {
      bIsDynamicallyLoaded=false
      drawAs=NoDrawType
      imageType=NoImage
      mirroring=NoMirror
      resourceName=None
      resourceObject=null
      tiling=NoTile
      imageSize=<0 0>
      margin=<ltrb 0 0 0 0>
      outlineSettings {
       bUseBrushTransparency=false
       roundingType=HalfHeightRadius
       width=0
       color {colorUseRule=UseColor_Specified, specifiedColor=<rgba 0 0 0 0>}
       cornerRadii=<0 0 0 0>
      }
      tintColor {colorUseRule=UseColor_Specified, specifiedColor=<rgba 1 1 1 1>}
      uVRegion {bIsValid=false, max=<0 0>, min=<0 0>}
     }
    }
    widgetStyle {
     barThickness=8
     bottomShadowBrush {
      bIsDynamicallyLoaded=false
      drawAs=Box
      imageType=FullColor
      mirroring=NoMirror
      resourceName=../../../Engine/Content/Slate/Common/ScrollBoxShadowBottom.png
      resourceObject=null
      tiling=NoTile
      imageSize=<16 8>
      margin=<ltrb 0.5 0 0.5 1>
      outlineSettings {
       bUseBrushTransparency=false
       roundingType=HalfHeightRadius
       width=0
       color {colorUseRule=UseColor_Specified, specifiedColor=<rgba 0 0 0 0>}
       cornerRadii=<0 0 0 0>
      }
      tintColor {colorUseRule=UseColor_Specified, specifiedColor=<rgba 1 1 1 1>}
      uVRegion {bIsValid=false, max=<0 0>, min=<0 0>}
     }
     horizontalScrolledContentPadding=<ltrb 0 0 1 0>
     leftShadowBrush {
      bIsDynamicallyLoaded=false
      drawAs=Box
      imageType=FullColor
      mirroring=NoMirror
      resourceName=../../../Engine/Content/Slate/Common/ScrollBoxShadowLeft.png
      resourceObject=null
      tiling=NoTile
      imageSize=<8 16>
      margin=<ltrb 1 0.5 0 0.5>
      outlineSettings {
       bUseBrushTransparency=false
       roundingType=HalfHeightRadius
       width=0
       color {colorUseRule=UseColor_Specified, specifiedColor=<rgba 0 0 0 0>}
       cornerRadii=<0 0 0 0>
      }
      tintColor {colorUseRule=UseColor_Specified, specifiedColor=<rgba 1 1 1 1>}
      uVRegion {bIsValid=false, max=<0 0>, min=<0 0>}
     }
     rightShadowBrush {
      bIsDynamicallyLoaded=false
      drawAs=Box
      imageType=FullColor
      mirroring=NoMirror
      resourceName=../../../Engine/Content/Slate/Common/ScrollBoxShadowRight.png
      resourceObject=null
      tiling=NoTile
      imageSize=<8 16>
      margin=<ltrb 0 0.5 1 0.5>
      outlineSettings {
       bUseBrushTransparency=false
       roundingType=HalfHeightRadius
       width=0
       color {colorUseRule=UseColor_Specified, specifiedColor=<rgba 0 0 0 0>}
       cornerRadii=<0 0 0 0>
      }
      tintColor {colorUseRule=UseColor_Specified, specifiedColor=<rgba 1 1 1 1>}
      uVRegion {bIsValid=false, max=<0 0>, min=<0 0>}
     }
     topShadowBrush {
      bIsDynamicallyLoaded=false
      drawAs=Box
      imageType=FullColor
      mirroring=NoMirror
      resourceName=../../../Engine/Content/Slate/Common/ScrollBoxShadowTop.png
      resourceObject=null
      tiling=NoTile
      imageSize=<16 8>
      margin=<ltrb 0.5 1 0.5 0>
      outlineSettings {
       bUseBrushTransparency=false
       roundingType=HalfHeightRadius
       width=0
       color {colorUseRule=UseColor_Specified, specifiedColor=<rgba 0 0 0 0>}
       cornerRadii=<0 0 0 0>
      }
      tintColor {colorUseRule=UseColor_Specified, specifiedColor=<rgba 1 1 1 1>}
      uVRegion {bIsValid=false, max=<0 0>, min=<0 0>}
     }
     verticalScrolledContentPadding=<ltrb 0 0 0 1>
    }
   }
   slot {class=StackBoxSlot, properties {verticalAlignment=VAlign_Top, padding=<ltrb 10 0 10 10>, size {sizeRule=Fill, value=1}}}
  }
  OptionsBox {
   widgetClassPath=&WrapBox
   bIsVariable=true
   parent=StackBox_0
   widget=OptionsBox
   properties {innerSlotPadding=<10 10>}
   slot {class=StackBoxSlot, properties {horizontalAlignment=HAlign_Right, padding=<ltrb 20 0 20 20>}}
  }
  Btn_Clear_Completed {widgetClassPath=&EditorUtilityButton, parent=OptionsBox, widget=Btn_Clear_Completed, slot {class=WrapBoxSlot}}
  TextBlock {
   widgetClassPath=&TextBlock
   parent=Btn_Clear_Completed
   widget=TextBlock
   properties {
    text=Clear Completed
    colorAndOpacity {colorUseRule=UseColor_Specified, specifiedColor=<rgba 0.5521 0.5521 0.5521 1>}
    font {
     bForceMonospaced=false
     bMaterialIsStencil=false
     fontMaterial=null
     fontObject=&NotoSans
     letterSpacing=0
     monospacedWidth=1
     size=8.25
     skewAmount=0
     typefaceFontName=NotoSans-Regular
     outlineSettings {
      bApplyOutlineToDropShadows=false
      bMiteredCorners=false
      bSeparateFillAlpha=false
      outlineMaterial=null
      outlineSize=0
      outlineColor=<rgba 0 0 0 1>
     }
    }
   }
   slot {class=ButtonSlot}
  }
  Btn_Clear_All {widgetClassPath=&EditorUtilityButton, parent=OptionsBox, widget=Btn_Clear_All, slot {class=WrapBoxSlot}}
  TextBlock_1 {
   widgetClassPath=&TextBlock
   parent=Btn_Clear_All
   widget=TextBlock_1
   properties {
    text=Clear All
    colorAndOpacity {colorUseRule=UseColor_Specified, specifiedColor=<rgba 0.5521 0.5521 0.5521 1>}
    font {
     bForceMonospaced=false
     bMaterialIsStencil=false
     fontMaterial=null
     fontObject=&NotoSans
     letterSpacing=0
     monospacedWidth=1
     size=8.25
     skewAmount=0
     typefaceFontName=NotoSans-Regular
     outlineSettings {
      bApplyOutlineToDropShadows=false
      bMiteredCorners=false
      bSeparateFillAlpha=false
      outlineMaterial=null
      outlineSize=0
      outlineColor=<rgba 0 0 0 1>
     }
    }
   }
   slot {class=ButtonSlot}
  }
 ]
}
