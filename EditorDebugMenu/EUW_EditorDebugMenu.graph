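# Editor Utility Widget for a searchable debug menu that runs console commands, editor tasks and utility objects.
# Rebuilds by gathering Data Registry items and DataTables from default and user folders into per-project data objects, then creating their container widgets.
# Dispatchers push search, styling, favourites and expand/collapse changes; collections are managed as separate containers.
# A SaveGame slot persists options, filters and slider states; console commands can target game, editor or PIE worlds.
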EditorUtilityWidgetBlueprint /EditorDebugMenu/EUW_EditorDebugMenu
graphs [
 Check Load is Valid {nodes [
   Check Load is Valid() {is=title K2Node_FunctionEntry}
   > n1 = Branch(Condition=Is Empty!(TargetMap:Map of Strings to Obj Project Data Object References=Get Project Data() -> (Project Data:Map of Strings to Obj Project Data Object References)) -> (ReturnValue:bool) {is=title K2Node_CallFunction})
   > Set Timer by Event(Delegate=Create Event(), Time=0.2, bLooping=false)
   Toggle Loading Screen!(execute=@n1.else, self, Toggle:bool=false) -> (then)
  ]}
 Check for saved config {nodes [
   Check for saved config() {is=title K2Node_FunctionEntry}
   > n1 = DoesSaveGameExist!(execute, SlotName:String=Get Save Game Slot Name() -> (Save Game Slot Name:String), UserIndex:int=0) -> (then, ReturnValue:bool)
   > n2 = Branch(Condition=@n1)
   > n3 = LoadGameFromSlot!(execute, SlotName:String=Get Save Game Slot Name(), UserIndex:int=0) -> (then, ReturnValue:Save Game Object Reference)
   > n4 = Cast To EDM_SaveGameConfig!(execute, Object=@n3) -> (then, CastFailed, AsEDM Save Game Config:EDM Save Game Config Object Reference) {is=title K2Node_DynamicCast}
   > Set Save game config(Save game config:EDM Save Game Config Object Reference=@n4) -> (Output_Get:EDM Save Game Config Object Reference)
   > Load Debug Menu Options()
   n5 = CreateSaveGameObject!(execute=@n2.else, SaveGameClass:Save Game Class Reference=&/EDM_SaveGameConfig.EDM_SaveGameConfig_C) -> (then, ReturnValue:EDM Save Game Config Object Reference)
   > Set Save game config(Save game config=@n5)
  ]}
 Clear Data {nodes [
   Clear Data() {is=title K2Node_FunctionEntry}
   > Clear!(
    execute
    TargetSet:Set of Obj EDM Input Data Object References=Get Favourite Map() -> (Favourite Map:Set of Obj EDM Input Data Object References)
   ) -> (then) {is=title K2Node_CallFunction}
   > Clear!(execute, TargetMap:Map of Strings to Obj Project Data Object References=Get Project Data()) -> (then)
   > Clear(TargetMap:Map of Strings to EDM Container Collection Object References=Get Collections Map() -> (Collections Map:Map of Strings to EDM Container Collection Object References))
   > ClearChildren(self=Get CollectionsContainer() -> (CollectionsContainer:Stack Box Object Reference))
   Get ProjectFilter_ListView() -> (ProjectFilter_ListView:Editor Utility List View Object Reference)
  ]}
 Create Widgets {nodes [
   Create Widgets() {is=title K2Node_FunctionEntry}
   > ClearChildren(self=Get ProjectsContainer() -> (ProjectsContainer:Stack Box Object Reference))
   > Clear(TargetArray:Array of EDM Container Project Object References=Get Container Widgets() -> (Container Widgets:Array of EDM Container Project Object References)) {is=title K2Node_CallArrayFunction}
   > ClearListItems(self=Get ProjectFilter_ListView())
   n1 = Get Project Data()
   > n2 = Keys!(execute, TargetMap:Map of Strings to Obj Project Data Object References=@n1) -> (then, Keys:Array of Strings)
   > n3 = For Each Loop(Array=@n2)
   n4 = Find(TargetMap:Map of Strings to Obj Project Data Object References=@n1, Key=@n3) -> (Value:Obj Project Data Object Reference)
   n5 = Get ProjectFilter_ListView()
   > AddItem!(execute, self:List View Object Reference=@n5, Item:Object Reference=@n4) -> (then)
   > n6 = Create Widget!(
    execute:Exec
    WorldContextObject:Object Reference
    Class:User Widget Class Reference=&/EDM_Container_Project.EDM_Container_Project_C
    OwningPlayer:Player Controller Object Reference
    Project Data:Obj Project Data Object Reference=@n4
    Project Name:String=@n3
    Debug Menu:Editor Debug Menu Object Reference=Self-Reference()
   ) -> (then:Exec, ReturnValue:EDM Container Project Object Reference)
   > Add(
    TargetArray:Array of EDM Container Project Object References=Get Container Widgets()
    NewItem:"EDM Container Project Object Reference (by ref)"=@n6
   ) {is=title K2Node_CallArrayFunction}
   > AddChild(self=Get ProjectsContainer(), Content=@n6)
   Bind Event to On Item Clicked(self:List View Object Reference=@n5, Delegate=Create Event()) {is=title K2Node_AddDelegate}
  ]}
 Debug Option Changed {nodes [Debug Option Changed!() -> (then, Text Size:int, Text Colour:Slate Color Structure, Button Colour:Slate Color Structure) {is=title K2Node_FunctionEntry}]}
 Edit Collections Toggled {nodes [Edit Collections Toggled!() -> (then, New Toggle:bool=~) {is=title K2Node_FunctionEntry}]}
 EventGraph {nodes [
   "On Clicked (RefreshButton)"() {is=title K2Node_ComponentBoundEvent}
   > n1 = Toggle Loading Screen(Toggle=true)
   n2 = Branch(execute=[@n12.then, @n1.then, @n3.then], Condition=IsLoadingAssets!(self:Asset Registry Interface=GetAssetRegistry()) -> (ReturnValue:bool))
   > n3 = Delay!(execute, WorldContextObject:Object Reference, Duration=0.5) -> (then)
   Rebuild Widget(execute=@n2.else)
   n4 = "On Text Changed (SearchBox)"!() -> (OutputDelegate, then, Text:"Text (by ref)") {is=title K2Node_ComponentBoundEvent}
   > n5 = Set Search(Search:String="To String (Text)"(InText=@n4.Text)) -> (Output_Get:String)
   Call Search Updated!(execute:Exec=[@n9.then, @n5.then], self:Editor Debug Menu Object Reference, Search Term:String=Get Search() -> (Search:String)) -> (then:Exec) {is=title K2Node_CallDelegate}
   > Branch(Condition=Get Ready To Save() -> (Ready To Save:bool))
   > Set Search!(
    execute
    Search=Get Search()
    self:EDM Save Game Config Object Reference=Get Save game config() -> (Save game config:EDM Save Game Config Object Reference)
   ) -> (then, Output_Get)
   > Save Config()
   n6 = "Options Changed (EDM_DebugMenuOptions)"!() -> (OutputDelegate, then, Property Name:Name=None, Widget:EDM Debug Menu Options Object Reference) {is=title K2Node_ComponentBoundEvent}
   > n7 = Set EDM Options Widget(EDM Options Widget:EDM Debug Menu Options Object Reference=@n6.Widget) -> (Output_Get:EDM Debug Menu Options Object Reference)
   > n8 = Switch on Name!(execute:Exec, Selection:Name=@n6.Property Name) -> (Default:Exec, Edit Collections:Exec) {is=title K2Node_SwitchName, pinNames [Edit Collections]}
   > Update Customisation Options()
   Call Edit Collections Toggled!(
    execute:Exec=@n8.Edit Collections
    self:Editor Debug Menu Object Reference
    New Toggle:bool=Get Edit Collections!(self:EDM Debug Menu Options Object Reference=@n7) -> (Edit Collections:bool)
   ) -> (then:Exec) {is=title K2Node_CallDelegate}
   "On Clicked (ClearSearch)"() {is=title K2Node_ComponentBoundEvent}
   > "SetText (Text Box)"!(execute, self:Text Box Object Reference=Get SearchBox() -> (SearchBox:Editor Utility Editable Text Box Object Reference), InText:Text) -> (then)
   > n9 = Set Search!(execute, Search) -> (then, Output_Get)
   "On Clicked (ExpandButton)"() {is=title K2Node_ComponentBoundEvent}
   > Call View Update!(execute:Exec, self:Editor Debug Menu Object Reference, Command:String=Expand) -> (then:Exec) {is=title K2Node_CallDelegate}
   "On Clicked (CollapseAllButton)"() {is=title K2Node_ComponentBoundEvent}
   > Call View Update(Command=Collapse)
   "On Pressed (SlackButton)"() {is=title K2Node_ComponentBoundEvent}
   > ExecutePythonCommand(PythonCommand="import webbrowser\r\nwebbrowser.open('https://epic.enterprise.slack.com/archives/C078L5N1ML1')")
   "On Clicked (WidgetSelect_Debug)"() {is=title K2Node_ComponentBoundEvent}
   n10 = Get MainSwitcher() -> (MainSwitcher:Widget Switcher Object Reference)
   > SetActiveWidgetIndex(self=@n10, Index=0)
   "Reset Widget (EDM_DebugMenuOptions)"() {is=title K2Node_ComponentBoundEvent}
   > Reset()
   "On Clicked (WidgetSelect_ToDo)"() {is=title K2Node_ComponentBoundEvent}
   > SetActiveWidgetIndex(self=@n10, Index=2)
   Rebuild Widget() {is=title K2Node_CustomEvent}
   > DelayUntilNextTick!(execute, WorldContextObject:Object Reference) -> (then)
   > Refresh Widget()
   > Update Customisation Options()
   > Set Ready To Save(Ready To Save:bool=true) -> (Output_Get:bool)
   > Call Favourites Changed!(execute:Exec, self:Editor Debug Menu Object Reference) -> (then:Exec) {is=title K2Node_CallDelegate}
   > RetriggerableDelay!(execute, WorldContextObject:Object Reference, Duration=0.1) -> (then)
   > Check Load is Valid() {is=title K2Node_CallFunction}
   Update Customisation Options() {is=title K2Node_CustomEvent}
   > n11 = Get!(execute:Exec) -> (then:Exec, else:Exec, EDM Options Widget:EDM Debug Menu Options Object Reference)
   > Call Debug Option Changed!(
    execute:Exec
    self:Editor Debug Menu Object Reference
    Text Size:int=Get Text Size!(self:EDM Debug Menu Options Object Reference=@n11) -> (Text Size:int=0)
    Text Colour:Slate Color Structure=Get Text Colour!(self:EDM Debug Menu Options Object Reference=@n11) -> (Text Colour:Slate Color Structure)
    Button Colour:Slate Color Structure=Get Button Colour!(self:EDM Debug Menu Options Object Reference=@n11) -> (Button Colour:Slate Color Structure)
   ) -> (then:Exec) {is=title K2Node_CallDelegate}
   > Set User Directories(User Directories:Array of Names=Get User Directories!(self:EDM Debug Menu Options Object Reference=@n11) -> (User Directories:Array of Names)) -> (Output_Get:Array of Names)
   > Save Debug Menu Options()
   n12 = Event Construct()
   n13 = Event Run EDM Console Command!() -> (OutputDelegate, then, Console Command:String, Target World:Enum_EDM_WorldType Enum=NewEnumerator0) {is=title K2Node_Event}
   > Run Console Command!(execute, self, Console Command:String=@n13.Console Command, World:Enum_EDM_WorldType Enum=@n13.Target World) -> (then)
   n14 = Event Toggle EDM Favourite!() -> (OutputDelegate, then, Add:bool, Input Data Obj:"Obj EDM Input Data Object Reference (by ref)") {is=title K2Node_Event}
   > n15 = Branch(Condition=@n14.Add)
   n16 = Get Favourite Map()
   > n17 = Add!(
    execute
    TargetSet:Set of Obj EDM Input Data Object References=@n16
    NewItem:"Obj EDM Input Data Object Reference (by ref)"=@n14.Input Data Obj
   ) -> (then) {is=title K2Node_CallFunction}
   Call Favourites Changed(execute=[@n18.then, @n17.then])
   n18 = Remove!(
    execute=@n15.else
    TargetSet:Set of Obj EDM Input Data Object References=@n16
    Item:"Obj EDM Input Data Object Reference (by ref)"=@n14.Input Data Obj
   ) -> (then, ReturnValue:bool)
   n19 = Event Run EDM Task!() -> (OutputDelegate, then, Task Class:EUT EDM Base Soft Class Reference) {is=title K2Node_Event}
   > n20 = Get!(execute) -> (then, else, Active Task:Editor Utility Task Object Reference)
   > PrintString(InString=A task is already active - Please wait till that task is complete)
   n21 = Async Load Class Asset!(execute:Exec=@n20.else, AssetClass:Object Soft Class Reference=@n19.Task Class) -> (then:Exec, Completed:Exec, Class:Object Class Reference) {is=title K2Node_LoadAssetClass}
   n22 = Construct Object from Class(
    execute=@n21.Completed
    Class=Cast To EUT_EDM_Base Class!(Class:Object Class Reference=@n21) -> (AsEUT EDM Base:EUT EDM Base Class Reference, bSuccess:bool=~) {is=title K2Node_ClassDynamicCast}
   ) -> (ReturnValue:EUT EDM Base Object Reference)
   > n23 = RegisterAndExecuteTask!(execute:Exec, Task:Editor Utility Task Object Reference=@n22, OptionalParentTask:Editor Utility Task Object Reference) -> (then:Exec, OnFinished:Exec, Task:Editor Utility Task Object Reference) {is=title K2Node_AsyncAction}
   > Set Active Task(Active Task:Editor Utility Task Object Reference=@n23) -> (Output_Get:Editor Utility Task Object Reference)
   Set Active Task(execute=@n23.OnFinished)
   > CollectGarbage!(execute) -> (then)
   Event Trigger Save EDM() {is=title K2Node_Event}
   > Save Config()
   n24 = Event Update EDM Collections!() -> (
    OutputDelegate
    then
    Collections To Add:Set of Strings
    Collections To Remove:Set of Strings
    Data Object:"Obj EDM Input Data Object Reference (by ref)"
   ) {is=title K2Node_Event}
   > Update Collections!(
    execute
    self
    Add:Set of Strings=@n24.Collections To Add
    Remove:Set of Strings=@n24.Collections To Remove
    NewItem:"Obj EDM Input Data Object Reference (by ref)"=@n24.Data Object
   ) -> (then)
   n25 = Event Run Object!() -> (OutputDelegate, then, EUO:Editor Utility Object Soft Class Reference) {is=title K2Node_Event}
   > n26 = LoadClassAsset_Blocking(AssetClass=@n25.EUO)
   > n27 = Cast To EditorUtilityObject Class!(execute:Exec, Class:Object Class Reference=@n26) -> (then:Exec, CastFailed:Exec, AsEditor Utility Object:Editor Utility Object Class Reference) {is=title K2Node_ClassDynamicCast}
   > n28 = Construct Object from Class(Class=@n27) -> (ReturnValue:Editor Utility Object Object Reference)
   > Run(self:Editor Utility Object Object Reference=@n28)
   > CollectGarbage()
   TestList!() -> (OutputDelegate, then, Item:Object Reference) {is=title K2Node_CustomEvent}
   Event PreConstruct()
  ]}
 FavouritesChanged {nodes [FavouritesChanged() {is=title K2Node_FunctionEntry}]}
 Find or Add Project Data {nodes [
   Find or Add Project Data!() -> (then, Project:"String (by ref)", ItemsToAdd:Array of Struct EDM Input Data 01 Structures, Source Location:String) {is=title K2Node_FunctionEntry}
   > n1 = Sequence()
   n2 = Find(TargetMap=Get Project Data(), Key=Get Project() -> (Project:"String (by ref)"))
   > n3 = Branch(Condition=@n2.ReturnValue)
   > Set FoundProjectData(FoundProjectData:Obj Project Data Object Reference=@n2) -> (Output_Get:Obj Project Data Object Reference)
   n4 = Get Project()
   n5 = Construct Object from Class!(
    execute=@n3.else
    Class=&/Obj_ProjectData.Obj_ProjectData_C
    self
    Project Name:String=@n4
    Debug Menu:Editor Debug Menu Object Reference=Self-Reference()
   ) -> (then, ReturnValue:Obj Project Data Object Reference)
   > n6 = Set FoundProjectData(FoundProjectData=@n5)
   > Initalise(self:Obj Project Data Object Reference=@n6)
   > Add!(
    execute
    TargetMap:Map of Strings to Obj Project Data Object References=Get Project Data()
    Key:"String (by ref)"=@n4
    Value:"Obj Project Data Object Reference (by ref)"=@n6
   ) -> (then)
   Append Row Data!(
    execute=@n1.then_1
    self:Obj Project Data Object Reference=Get FoundProjectData() -> (FoundProjectData:Obj Project Data Object Reference)
    Data:Array of Struct EDM Input Data 01 Structures=Get ItemsToAdd() -> (ItemsToAdd:Array of Struct EDM Input Data 01 Structures)
    Source Location:String=Get Source Location() -> (Source Location:String)
   ) -> (then)
  ]}
 Gather DTs in Folder {nodes [
   Gather DTs in Folder() {is=title K2Node_FunctionEntry}
   > n1 = Set SearchDirectories(SearchDirectories:Array of Names=Make Array("[0]":Name=&/Data/) -> (Array:Array of Names)) -> (Output_Get:Array of Names)
   > Append Array!(execute, TargetArray:Array of Names=@n1, SourceArray:Array of Names=Get User Directories!() -> (User Directories)) -> (then)
   > n2 = For Each Loop(Array:Array of Names=Get SearchDirectories() -> (SearchDirectories:Array of Names)) -> (Array Element:Name)
   > n3 = GetAssetsByPath!(execute, self:Asset Registry Interface=GetAssetRegistry(), PackagePath:Name=@n2, bRecursive:bool=true, bIncludeOnlyOnDiskAssets:bool=true) -> (then, OutAssetData:Array of Asset Data Structures, ReturnValue:bool)
   > Branch(Condition=@n3.ReturnValue)
   > n4 = For Each Loop(Array:Array of Asset Data Structures=@n3) -> (Array Element:Asset Data Structure)
   > n5 = Cast To DataTable!(execute, Object=GetAsset(InAssetData=@n4)) -> (then, CastFailed, AsData Table:Data Table Object Reference) {is=title K2Node_DynamicCast}
   > n6 = GetPathNameForLoadedAsset(LoadedAsset=@n5)
   > n7 = GetRowsForDataTable!(execute, self, Table:Data Table Object Reference=@n5) -> (then, Rows:Array of Struct EDM Input Data 01 Structures)
   n8 = ToSoftObjectPath!(InAssetData:"Asset Data Structure (by ref)"=@n4) -> (ReturnValue_PathString:String)
   > Find or Add Project Data!(
    execute
    self
    Project:"String (by ref)"=Split(SourceString=Split!(InPath:String=@n6) -> (PathPart:String, FilenamePart:String, ExtensionPart:String), InStr=/, SearchDir=FromEnd).RightS
    ItemsToAdd:Array of Struct EDM Input Data 01 Structures=@n7
    Source Location:String=@n8
   ) -> (then) {is=title K2Node_CallFunction}
   Split!(SourceString:String=@n8, InStr:String=., SearchCase:ESearchCase Enum, SearchDir:ESearchDir Enum=FromEnd) -> (LeftS:String, RightS:String, ReturnValue:bool)
  ]}
 Gather Data Registry Items {nodes [
   Gather Data Registry Items() {is=title K2Node_FunctionEntry}
   > n1 = GetDebugMenuRegistryIdMap!(execute) -> (then, OutIdMap:Map of Strings to Debug Menu Registry Id List Structures)
   > n2 = Keys(TargetMap:Map of Strings to Debug Menu Registry Id List Structures=@n1)
   > n3 = For Each Loop(Array:Array of Strings=@n2) -> (Array Element:String)
   > n4 = Get Registry Rows For Source!(execute, self, SourceToRowsMap:Map of Strings to Debug Menu Registry Id List Structures=@n1, SourceToFind:"String (by ref)"=@n3) -> (then, RegistryRows:Array of Struct EDM Input Data 01 Structures) {is=title K2Node_CallFunction}
   > Find or Add Project Data(Project=@n3, ItemsToAdd=@n4, Source Location=Append(A="Data Registry - ", B=@n3))
  ]}
 Get Config {nodes [
   n1 = Get Config!() -> (then, Category:String) {is=title K2Node_FunctionEntry}
   > Return Node!(
    execute
    State:bool=Get State!(self:EDM Save Game Config Object Reference=Get Save game config(), Category:String=@n1) -> (State:bool) {is=title K2Node_CallFunction}
   )
  ]}
 Get Registry Rows For Source {nodes [
   n1 = Get Registry Rows For Source!() -> (then, SourceToRowsMap:Map of Strings to Debug Menu Registry Id List Structures, SourceToFind:"String (by ref)") {is=title K2Node_FunctionEntry}
   n2 = Find!(TargetMap:Map of Strings to Debug Menu Registry Id List Structures=@n1, Key=@n1.SourceToFind) -> (Value_RegistryIdList:Array of Data Registry Id Structures, ReturnValue)
   > Branch(Condition=@n2.ReturnValue)
   > n3 = For Each Loop(Array:Array of Data Registry Id Structures=@n2) -> (Array Element:Data Registry Id Structure)
   Return Node!(
    execute=@n3.Completed
    RegistryRows:Array of Struct EDM Input Data 01 Structures=Get Rows() -> (Rows:Array of Struct EDM Input Data 01 Structures)
   )
   > n4 = Get Data Registry Item!(
    execute
    ItemId:Data Registry Id Structure=@n3
    OutItem:Struct EDM Input Data 01 Structure=Get Current Registry Item() -> (Current Registry Item:Struct EDM Input Data 01 Structure)
   ) -> (then, ReturnValue:bool) {is=title K2Node_CallFunction}
   > Branch(Condition=@n4)
   > Add(
    TargetArray:Array of Struct EDM Input Data 01 Structures=Get Rows()
    NewItem:"Struct EDM Input Data 01 Structure (by ref)"=Get Current Registry Item()
   ) {is=title K2Node_CallArrayFunction}
  ]}
 Get Widget From Debug Type {nodes [
   n1 = Get Widget From Debug Type!() -> (then, Debug Option:"Enum_EDM_InputType Enum (by ref)"=NewEnumerator0) {is=title K2Node_FunctionEntry}
   > Return Node!(
    execute
    Widget:EDM Input Base Class Reference=Find!(
     TargetMap:Map of Enum_EDM_InputType Enums to EDM Input Base Class References=Get Debug Widget Map() -> (Debug Widget Map:Map of Enum_EDM_InputType Enums to EDM Input Base Class References=NewEnumerator0)
     Key:"Enum_EDM_InputType Enum (by ref)"=@n1
    ) -> (Value:EDM Input Base Class Reference, ReturnValue)
   )
  ]}
 GetRowsForDataTable {nodes [
   n1 = GetRowsForDataTable!() -> (then, Table:Data Table Object Reference) {is=title K2Node_FunctionEntry}
   > n2 = GetDataTableRowNames(Table=@n1)
   > n3 = For Each Loop(Array:Array of Names=@n2) -> (Array Element:Name)
   Return Node!(
    execute=@n3.Completed
    Rows:Array of Struct EDM Input Data 01 Structures=Get OutRows() -> (OutRows:Array of Struct EDM Input Data 01 Structures)
   )
   > n4 = Get Data Table Row!(execute:Exec, DataTable:Data Table Object Reference=@n1, RowName:Name=@n3) -> (then:Exec, RowNotFound:Exec, ReturnValue:Struct EDM Input Data 01 Structure)
   > Add(TargetArray=Get OutRows(), NewItem=@n4)
  ]}
 Load Debug Menu Options {nodes [
   Load Debug Menu Options() {is=title K2Node_FunctionEntry}
   > n1 = Sequence()
   n2 = Get Search!(self:EDM Save Game Config Object Reference=Get Save game config()) -> (Search:String) {is=title K2Node_CallFunction}
   > Set Search(Search=@n2)
   > "SetText (Text Box)"(self=Get SearchBox(), InText="To Text (String)"(InString=@n2))
   n3 = Get Save game config()
   Set Text Size(execute=@n1.then_1, Text Size:int=Get Text Size(self:EDM Save Game Config Object Reference=@n3) -> (Text Size=0)) -> (Output_Get:int=0)
   > Set Text Colour(Text Colour:Slate Color Structure=Get Text Colour(self:EDM Save Game Config Object Reference=@n3)) -> (Output_Get:Slate Color Structure)
   > Set Button Colour(Button Colour:Slate Color Structure=Get Button Color!(self:EDM Save Game Config Object Reference=@n3) -> (Button Color:Slate Color Structure)) -> (Output_Get:Slate Color Structure)
   > n4 = Set User Directories(User Directories=Get User Directories!(self:EDM Save Game Config Object Reference=@n3) -> (User Directories))
   > n5 = Sequence()
   > n6 = Get!(execute) -> (then, else, EDM Options Widget:EDM Debug Menu Options Object Reference)
   > Set User Directories!(execute, User Directories=@n4, self:EDM Debug Menu Options Object Reference=@n6) -> (then, Output_Get)
   PrintString(execute=@n6.else, InString=------no options widget)
   n7 = Get!(execute=@n5.then_1) -> (then, else, EDM_DebugMenuOptions:EDM Debug Menu Options Object Reference)
   > Set User Directories(User Directories=@n4, self=@n7)
   Set Text Size!(execute, Text Size=0, self:EDM Debug Menu Options Object Reference=@n6) -> (then, Output_Get=0)
   > Set Text Colour!(execute, Text Colour, self:EDM Debug Menu Options Object Reference=@n6) -> (then, Output_Get)
   > Set Button Colour!(execute, Button Colour, self:EDM Debug Menu Options Object Reference=@n6) -> (then, Output_Get)
   Set Text Size(Text Size=0, self=@n7) -> (Output_Get=0)
   > Set Text Colour(self=@n7)
   > Set Button Colour(self=@n7)
  ]}
 Load Slider State {nodes [
   n1 = Load Slider State!() -> (then, Name:"String (by ref)") {is=title K2Node_FunctionEntry}
   > n2 = Get!(execute) -> (then, else, Save game config:EDM Save Game Config Object Reference)
   n3 = Find(
    TargetMap:"Map of Strings to Float (double-precision)s"=Get SliderStates!(self:EDM Save Game Config Object Reference=@n2) -> (SliderStates:"Map of Strings to Float (double-precision)s")
    Key:"String (by ref)"=@n1
   ) -> (Value:double)
   > Return Node!(execute, Value:double=@n3, Found:bool=@n3.ReturnValue)
  ]}
 Refresh Widget {nodes [
   Refresh Widget() {is=title K2Node_FunctionEntry}
   > Clear Data() {is=title K2Node_CallFunction}
   > Check for saved config() {is=title K2Node_CallFunction}
   > Gather Data Registry Items() {is=title K2Node_CallFunction}
   > Gather DTs in Folder() {is=title K2Node_CallFunction}
   > SortAndPopulateData()
   > Create Widgets() {is=title K2Node_CallFunction}
  ]}
 Reset {nodes [
   Reset() {is=title K2Node_FunctionEntry}
   > Toggle Loading Screen(Toggle=true)
   > n1 = CreateSaveGameObject(SaveGameClass=&EDM_SaveGameConfig_C)
   > n2 = Set Save game config(Save game config=@n1)
   > SaveGameToSlot!(execute, SaveGameObject:Save Game Object Reference=@n2, SlotName:String=Get Save Game Slot Name(), UserIndex:int=0) -> (then, ReturnValue:bool)
   > Rebuild Widget() {is=title K2Node_CallFunction}
  ]}
 Run Console Command {nodes [
   n1 = Run Console Command!() -> (then, Console Command:String, World:Enum_EDM_WorldType Enum=NewEnumerator0) {is=title K2Node_FunctionEntry}
   > Set Command(Command:String=@n1) -> (Output_Get:String)
   > K2Node_Composite_0 = Replace List Items!(execute) -> (then:Exec) {is=title K2Node_Composite}
   > n2 = Sequence()
   > n3 = Switch on Enum_EDM_WorldType!(execute, Selection:Enum_EDM_WorldType Enum=@n1.World) -> (NewEnumerator0:Exec, NewEnumerator1:Exec, NewEnumerator2:Exec, NewEnumerator3:Exec) {is=title K2Node_SwitchEnum}
   n4 = Sequence(execute=@n3.NewEnumerator2)
   n5 = Get UnrealEditorSubsystem()
   n6 = GetGameWorld!(execute=[@n4.then_0, @n3.NewEnumerator0], self:Unreal Editor Subsystem Object Reference=@n5) -> (then, ReturnValue:World Object Reference)
   n7 = GetPlayerController!(WorldContextObject:Object Reference, PlayerIndex:int=0) -> (ReturnValue:Player Controller Object Reference)
   n8 = Get Command() -> (Command:String)
   > ExecuteConsoleCommand(WorldContextObject=@n6, Command=@n8, SpecificPlayer=@n7)
   n9 = GetEditorWorld(execute=[@n4.then_1, @n3.NewEnumerator1], self=@n5)
   > ExecuteConsoleCommand(WorldContextObject=@n9, Command=@n8, SpecificPlayer=@n7)
   n10 = GetPIEWorlds!(execute=@n3.NewEnumerator3, bIncludeDedicatedServer:bool=false) -> (then, ReturnValue:Array of World Object References)
   > n11 = For Each Loop(Array:Array of World Object References=@n10) -> (Array Element:World Object Reference)
   > ExecuteConsoleCommand(WorldContextObject=@n11, Command=@n8, SpecificPlayer=GetPlayerController(WorldContextObject=@n11, PlayerIndex=0))
   "SetText (Editable Text)"!(
    execute=@n2.then_1
    self:Editable Text Object Reference=Get Output_Text() -> (Output_Text:Editor Utility Editable Text Object Reference)
    InText:Text=Format Text!(Format="LOCTEXT(\"Running Command: {Command}\")", Command:String=Get Console Command() -> (Console Command:String)) -> (Result)
   ) -> (then)
  ]}
 Replace List Items {
  path=Run Console Command.K2Node_Composite_0.Replace List Items
  nodes [
   Inputs()
   > n1 = Branch(Condition=Contains(SearchIn=Get Command(), Substring="<ListItem>"))
   > n2 = Get Selected List Item!(execute, self:EDM List Generator Object Reference=Get EDM_ListGenerator() -> (EDM_ListGenerator:EDM List Generator Object Reference)) -> (then, List Item:String) {is=title K2Node_CallFunction}
   > n3 = ReplaceInline!(
    execute
    SourceString:"String (by ref)"=Get Command()
    SearchText:String="<ListItem>"
    ReplacementText:String=@n2
    SearchCase:ESearchCase Enum
   ) -> (then, ReturnValue:int=0)
   Outputs!(then:Exec=[@n3.then, @n1.else])
  ]
 }
 Save Config {nodes [
   Save Config() {is=title K2Node_FunctionEntry}
   > SaveGameToSlot(SaveGameObject=Get Save game config(), SlotName=Get Save Game Slot Name(), UserIndex=0)
  ]}
 Save Debug Menu Options {nodes [
   Save Debug Menu Options() {is=title K2Node_FunctionEntry}
   > n1 = Get()
   > Set Text Size(Text Size=Get Text Size!() -> (Text Size=0), self:EDM Save Game Config Object Reference=@n1) -> (Output_Get=0)
   > Set Text Colour(Text Colour=Get Text Colour!() -> (Text Colour), self:EDM Save Game Config Object Reference=@n1)
   > Set Button Color!(execute, Button Color:Slate Color Structure=Get Button Colour!() -> (Button Colour), self:EDM Save Game Config Object Reference=@n1) -> (then, Output_Get:Slate Color Structure)
   > Set User Directories(User Directories=Get User Directories!() -> (User Directories), self:EDM Save Game Config Object Reference=@n1)
   > Save Config() {is=title K2Node_CallFunction}
  ]}
 Save Slider State {nodes [
   n1 = Save Slider State!() -> (then, Name:String, Value:double=~) {is=title K2Node_FunctionEntry}
   > n2 = Get()
   > Add(
    TargetMap:"Map of Strings to Float (double-precision)s"=Get SliderStates(self=@n2)
    Key=@n1
    Value:"Float (double-precision) (by ref)"=@n1.Value
   ) {is=title K2Node_CallFunction}
   > Save Config()
  ]}
 SearchUpdated {nodes [SearchUpdated!() -> (then, Search Term:String) {is=title K2Node_FunctionEntry}]}
 SortAndPopulateData {nodes [
   SortAndPopulateData() {is=title K2Node_FunctionEntry}
   > n1 = Values!(execute, TargetMap:Map of Strings to Obj Project Data Object References=Get Project Data()) -> (then, Values:Array of Obj Project Data Object References)
   > n2 = For Each Loop(Array:Array of Obj Project Data Object References=@n1) -> (Array Element:Obj Project Data Object Reference)
   > n3 = Populate Data!(execute, self:Obj Project Data Object Reference=@n2) -> (then, Unsorted Keys:Array of Strings, Favourites:Map of Names to Struct EDM Input Data 01 Structures)
   > n4 = SortStringArray!(execute, InStrings:Array of Strings=@n3) -> (then, OutSortedStrings:Array of Strings)
   > Sort Data!(execute, self:Obj Project Data Object Reference=@n2, Sorted Keys:Array of Strings=@n4) -> (then)
   > Keys(TargetMap:Map of Names to Struct EDM Input Data 01 Structures=@n3.Favourites) -> (Keys:Array of Names)
  ]}
 TestDispatch {nodes [TestDispatch!() -> (then, FieldId:Field Notification Id Structure) {is=title K2Node_FunctionEntry}]}
 Toggle Loading Screen {nodes [
   n1 = Toggle Loading Screen!() -> (then, Toggle:bool=~) {is=title K2Node_FunctionEntry}
   > SetActiveWidgetIndex(self=Get MainSwitcher(), Index="To Integer (Boolean)"!(InBool:bool=@n1) -> (ReturnValue:int=0))
  ]}
 Update Collections {nodes [
   n1 = Update Collections!() -> (then, Add:Set of Strings, Remove:Set of Strings, NewItem:"Obj EDM Input Data Object Reference (by ref)") {is=title K2Node_FunctionEntry}
   > n2 = Sequence()
   > Branch(Condition=Is Not Empty!(TargetSet:Set of Strings=@n1) -> (ReturnValue:bool) {is=title K2Node_CallFunction})
   > n3 = To Array!(execute, A:Set of Strings=@n1) -> (then, Result:Array of Strings)
   > n4 = For Each Loop(Array:Array of Strings=@n3) -> (Array Element:String)
   > n5 = Set Current Collection(Current Collection:String=@n4) -> (Output_Get:String)
   n6 = Find(TargetMap:Map of Strings to EDM Container Collection Object References=Get Collections Map(), Key=@n5) -> (Value:EDM Container Collection Object Reference)
   > n7 = Branch(Condition=@n6.ReturnValue)
   > Add To Collection!(
    execute
    self:EDM Container Collection Object Reference=@n6
    NewItem:"Obj EDM Input Data Object Reference (by ref)"=Get NewItem() -> (NewItem:"Obj EDM Input Data Object Reference (by ref)")
   ) -> (then)
   n8 = Create Widget!(
    execute=@n7.else
    WorldContextObject
    Class=&/EDM_Container_Collection.EDM_Container_Collection_C
    OwningPlayer
    Debug Menu=Self-Reference()
    Collection Name:String=Get Current Collection() -> (Current Collection:String)
   ) -> (then, ReturnValue:EDM Container Collection Object Reference)
   > Add To Collection(self=@n8, NewItem=Get NewItem())
   > AddChildToStackBox!(execute, self:Stack Box Object Reference=Get CollectionsContainer(), Content:Widget Object Reference=@n8) -> (then, ReturnValue:Stack Box Slot Object Reference)
   > Add(
    TargetMap:Map of Strings to EDM Container Collection Object References=Get Collections Map()
    Key=Get Current Collection()
    Value:"EDM Container Collection Object Reference (by ref)"=@n8
   )
   Branch(execute=@n2.then_1, Condition=Is Not Empty(TargetSet=@n1.Remove))
   > n9 = To Array(A=@n1.Remove)
   > n10 = For Each Loop(Array=@n9)
   > n11 = Set Current Collection(Current Collection=@n10)
   n12 = Find(TargetMap=Get Collections Map(), Key=@n11)
   > Branch(Condition=@n12.ReturnValue)
   > n13 = Remove from collection!(execute, self:EDM Container Collection Object Reference=@n12, Item:"Obj EDM Input Data Object Reference (by ref)"=Get NewItem()) -> (then, Empty:bool)
   > Branch(Condition=@n13)
   > RemoveFromParent(self=@n12)
   > Remove!(
    execute
    TargetMap:Map of Strings to EDM Container Collection Object References=Get Collections Map()
    Key:"String (by ref)"=Get Current Collection()
   ) -> (then, ReturnValue)
  ]}
 Update Config {nodes [
   n1 = Update Config!() -> (then, Category Name:String, Saved State:bool=~) {is=title K2Node_FunctionEntry}
   > n2 = Get()
   > Save State!(execute, self:EDM Save Game Config Object Reference=@n2, Category:String=@n1, State:bool=@n1.Saved State) -> (then)
   > SaveGameToSlot(SaveGameObject=@n2, SlotName=Get Save Game Slot Name(), UserIndex=0)
   PrintString(execute=@n2.else, InString=Save game not valid)
  ]}
 Update Favourite {nodes [
   n1 = Update Favourite!() -> (then, Favourite:Name, New State:bool=~) {is=title K2Node_FunctionEntry}
   > Modify Favourite!(
    execute
    self:EDM Save Game Config Object Reference=Get Save game config()
    NewItem:"String (by ref)"="To String (Name)"(InName=@n1)
    New State:bool=@n1.New State
   ) -> (then)
   > Call Favourites Changed()
   > Save Config()
   Get Favourite Map()
  ]}
 Update Project Filter {nodes [
   n1 = Update Project Filter!() -> (then, Project Filter:Object Reference) {is=title K2Node_FunctionEntry}
   > n2 = Cast To Obj_ProjectData!(execute, Object=@n1) -> (then, CastFailed, AsObj Project Data:Obj Project Data Object Reference) {is=title K2Node_DynamicCast}
   n3 = Get Project Name!(self:Obj Project Data Object Reference=@n2) -> (Project Name:String)
   n4 = Find(TargetMap:Map of Strings to Obj Project Data Object References=Get Project Data(), Key=@n3) -> (Value:Obj Project Data Object Reference)
   > n5 = Set Visible!(
    execute
    Visible:bool=NOT Boolean(A=Get Visible!(self:Obj Project Data Object Reference=@n4) -> (Visible:bool))
    self:Obj Project Data Object Reference=@n4
   ) -> (then, Output_Get:bool)
   > Set Project Filter Visibility!(execute, self:EDM Save Game Config Object Reference=Get Save game config(), Project:String=@n3, Vis State:bool=@n5) -> (then) {is=title K2Node_CallFunction}
   > Save Config()
   > Create Widgets()
  ]}
 ViewUpdate {nodes [ViewUpdate!() -> (then, Command:String) {is=title K2Node_FunctionEntry}]}
]
variables [
 Active Task {type=EditorUtilityTask, default=null}
 "Auto Sync (Expensive)" {type=boolean}
 Button Colour {type=SlateColor, default {colorUseRule=UseColor_Specified, specifiedColor=<rgba 0.0395 0.0395 0.0395 1>}}
 Collections Map {type="map<EDM_Container_Collection_C>", default {}}
 Container Widgets {type="array<EDM_Container_Project_C>", default []}
 Current Version {type=number, default=1.02}
 Debug Option Changed {default=()}
 Debug Widget Map {default {
   Boolean=&/EDM_Input_Toggle.EDM_Input_Toggle_C
   Button=&/EDM_Input_Button.EDM_Input_Button_C
   Button And Text=&/EDM_Input_Button_Text.EDM_Input_Button_Text_C
   Button Series=&EDM_Input_Button_C
   Editor Task=&/EDM_Input_Task.EDM_Input_Task_C
   Editor Utility Object=&/EDM_Input_EUO.EDM_Input_EUO_C
   Slider=&/EDM_Input_Slider.EDM_Input_Slider_C
  }}
 Edit Collections {type=boolean}
 Edit Collections Toggled {default=()}
 EDM Options Widget {type=EDM_DebugMenuOptions_C, default=null}
 Favourite Map {type="array<Obj_EDM_InputData_C>", default []}
 FavouritesChanged {default=()}
 Options {type="map<Obj_EDM_CategoryData_C>", default {}}
 Project Data {type="map<Obj_ProjectData_C>", default {}}
 Ready To Save {type=boolean}
 Save game config {type=EDM_SaveGameConfig_C, default=null}
 Save Game Slot Name {type=string, default=EditorDebugMenuConfig}
 Search {type=string, default=""}
 SearchUpdated {default=()}
 TestDispatch {default=()}
 Text Colour {type=SlateColor, default {colorUseRule=UseColor_Specified, specifiedColor=<rgba 1 1 1 1>}}
 Text Size {type=integer, default=8}
 Timer - Auto Sync {type=TimerHandle}
 User Directories {type="array<string>", default []}
 ViewUpdate {default=()}
]
widgets {
 info {namedSlotCount=2, rootWidgetClass=&HorizontalBox, widgetCount=54}
 widgets [
  HorizontalBox_0 {widgetClassPath=&HorizontalBox, widget=HorizontalBox_0}
  VerticalBox_187 {widgetClassPath=&VerticalBox, parent=HorizontalBox_0, widget=VerticalBox_187, slot {class=HorizontalBoxSlot}}
  Overlay_0 {
   widgetClassPath=&Overlay
   parent=VerticalBox_187
   widget=Overlay_0
   slot {class=VerticalBoxSlot, properties {padding=<ltrb 2 2 2 2>}}
  }
  WidgetSelect_Debug {
   widgetClassPath=&EditorUtilityButton
   parent=Overlay_0
   widget=WidgetSelect_Debug
   properties {
    bOverride_Cursor=true
    cursor=Hand
    toolTipText=Debug Menu
    widgetStyle {
     clickedSlateSound {resourceObject=null}
     disabled {
      bIsDynamicallyLoaded=false
      drawAs=RoundedBox
      imageType=NoImage
      mirroring=NoMirror
      resourceName=None
      resourceObject=null
      tiling=NoTile
      imageSize=<0 0>
      margin=<ltrb 0 0 0 0>
      outlineSettings {
       bUseBrushTransparency=false
       roundingType=FixedRadius
       width=1
       color {colorUseRule=UseColor_ColorTable, specifiedColor=<rgba 0 0 0 0>}
       cornerRadii=<4 4 4 4>
      }
      tintColor {colorUseRule=UseColor_Specified, specifiedColor=<rgba 0.0395 0.0395 0.0395 1>}
      uVRegion {bIsValid=false, max=<0 0>, min=<0 0>}
     }
     disabledForeground {colorUseRule=UseColor_ColorTable, specifiedColor=<rgba 0 0 0 0>}
     hovered {
      bIsDynamicallyLoaded=false
      drawAs=RoundedBox
      imageType=NoImage
      mirroring=NoMirror
      resourceName=None
      resourceObject=null
      tiling=NoTile
      imageSize=<18 18>
      margin=<ltrb 0 0 0 0>
      outlineSettings {
       bUseBrushTransparency=false
       roundingType=FixedRadius
       width=1
       color {colorUseRule=UseColor_ColorTable, specifiedColor=<rgba 0 0 0 0>}
       cornerRadii=<6 6 6 6>
      }
      tintColor {colorUseRule=UseColor_Specified, specifiedColor=<rgba 0.0953 0.0953 0.0953 1>}
      uVRegion {bIsValid=false, max=<0 0>, min=<0 0>}
     }
     hoveredForeground {colorUseRule=UseColor_ColorTable, specifiedColor=<rgba 0 0 0 0>}
     hoveredSlateSound {resourceObject=null}
     normal {
      bIsDynamicallyLoaded=false
      drawAs=RoundedBox
      imageType=FullColor
      mirroring=NoMirror
      resourceName=None
      resourceObject=null
      tiling=NoTile
      imageSize=<18 18>
      margin=<ltrb 0 0 0 0>
      outlineSettings {
       bUseBrushTransparency=false
       roundingType=FixedRadius
       width=0
       color {colorUseRule=UseColor_ColorTable, specifiedColor=<rgba 0 0 0 0>}
       cornerRadii=<6 6 6 6>
      }
      tintColor {colorUseRule=UseColor_Specified, specifiedColor=<rgba 0.0521 0.0521 0.0521 1>}
      uVRegion {bIsValid=false, max=<0 0>, min=<0 0>}
     }
     normalForeground {colorUseRule=UseColor_ColorTable, specifiedColor=<rgba 0 0 0 0>}
     normalPadding=<ltrb 12 1.5 12 1.5>
     pressed {
      bIsDynamicallyLoaded=false
      drawAs=RoundedBox
      imageType=NoImage
      mirroring=NoMirror
      resourceName=None
      resourceObject=null
      tiling=NoTile
      imageSize=<18 18>
      margin=<ltrb 0 0 0 0>
      outlineSettings {
       bUseBrushTransparency=false
       roundingType=FixedRadius
       width=1
       color {colorUseRule=UseColor_ColorTable, specifiedColor=<rgba 0 0 0 0>}
       cornerRadii=<6 6 6 6>
      }
      tintColor {colorUseRule=UseColor_Specified, specifiedColor=<rgba 0.0284 0.0284 0.0284 1>}
      uVRegion {bIsValid=false, max=<0 0>, min=<0 0>}
     }
     pressedForeground {colorUseRule=UseColor_ColorTable, specifiedColor=<rgba 0 0 0 0>}
     pressedPadding=<ltrb 12 2.5 12 0.5>
     pressedSlateSound {resourceObject=null}
    }
   }
   slot {class=OverlaySlot}
  }
  TextBlock_74 {
   widgetClassPath=&TextBlock
   parent=Overlay_0
   widget=TextBlock_74
   properties {
    justification=Center
    text=D
    visibility=SelfHitTestInvisible
    font {
     bForceMonospaced=false
     bMaterialIsStencil=false
     fontMaterial=null
     fontObject=&Roboto
     letterSpacing=0
     monospacedWidth=1
     size=7.5
     skewAmount=0
     typefaceFontName=Bold
     outlineSettings {
      bApplyOutlineToDropShadows=false
      bMiteredCorners=false
      bSeparateFillAlpha=false
      outlineMaterial=null
      outlineSize=0
      outlineColor=<rgba 0 0 0 1>
     }
    }
   }
   slot {class=OverlaySlot, properties {horizontalAlignment=HAlign_Center, verticalAlignment=VAlign_Center}}
  }
  Overlay_1 {
   widgetClassPath=&Overlay
   parent=VerticalBox_187
   widget=Overlay_1
   slot {class=VerticalBoxSlot, properties {padding=<ltrb 2 2 2 2>}}
  }
  WidgetSelect_ToDo {
   widgetClassPath=&EditorUtilityButton
   parent=Overlay_1
   widget=WidgetSelect_ToDo
   properties {
    bOverride_Cursor=true
    cursor=Hand
    toolTipText=Todo List
    widgetStyle {
     clickedSlateSound {resourceObject=null}
     disabled {
      bIsDynamicallyLoaded=false
      drawAs=RoundedBox
      imageType=NoImage
      mirroring=NoMirror
      resourceName=None
      resourceObject=null
      tiling=NoTile
      imageSize=<0 0>
      margin=<ltrb 0 0 0 0>
      outlineSettings {
       bUseBrushTransparency=false
       roundingType=FixedRadius
       width=1
       color {colorUseRule=UseColor_ColorTable, specifiedColor=<rgba 0 0 0 0>}
       cornerRadii=<4 4 4 4>
      }
      tintColor {colorUseRule=UseColor_Specified, specifiedColor=<rgba 0.0395 0.0395 0.0395 1>}
      uVRegion {bIsValid=false, max=<0 0>, min=<0 0>}
     }
     disabledForeground {colorUseRule=UseColor_ColorTable, specifiedColor=<rgba 0 0 0 0>}
     hovered {
      bIsDynamicallyLoaded=false
      drawAs=RoundedBox
      imageType=NoImage
      mirroring=NoMirror
      resourceName=None
      resourceObject=null
      tiling=NoTile
      imageSize=<18 18>
      margin=<ltrb 0 0 0 0>
      outlineSettings {
       bUseBrushTransparency=false
       roundingType=FixedRadius
       width=1
       color {colorUseRule=UseColor_ColorTable, specifiedColor=<rgba 0 0 0 0>}
       cornerRadii=<6 6 6 6>
      }
      tintColor {colorUseRule=UseColor_Specified, specifiedColor=<rgba 0.0953 0.0953 0.0953 1>}
      uVRegion {bIsValid=false, max=<0 0>, min=<0 0>}
     }
     hoveredForeground {colorUseRule=UseColor_ColorTable, specifiedColor=<rgba 0 0 0 0>}
     hoveredSlateSound {resourceObject=null}
     normal {
      bIsDynamicallyLoaded=false
      drawAs=RoundedBox
      imageType=FullColor
      mirroring=NoMirror
      resourceName=None
      resourceObject=null
      tiling=NoTile
      imageSize=<18 18>
      margin=<ltrb 0 0 0 0>
      outlineSettings {
       bUseBrushTransparency=false
       roundingType=FixedRadius
       width=0
       color {colorUseRule=UseColor_ColorTable, specifiedColor=<rgba 0 0 0 0>}
       cornerRadii=<6 6 6 6>
      }
      tintColor {colorUseRule=UseColor_Specified, specifiedColor=<rgba 0.0521 0.0521 0.0521 1>}
      uVRegion {bIsValid=false, max=<0 0>, min=<0 0>}
     }
     normalForeground {colorUseRule=UseColor_ColorTable, specifiedColor=<rgba 0 0 0 0>}
     normalPadding=<ltrb 12 1.5 12 1.5>
     pressed {
      bIsDynamicallyLoaded=false
      drawAs=RoundedBox
      imageType=NoImage
      mirroring=NoMirror
      resourceName=None
      resourceObject=null
      tiling=NoTile
      imageSize=<18 18>
      margin=<ltrb 0 0 0 0>
      outlineSettings {
       bUseBrushTransparency=false
       roundingType=FixedRadius
       width=1
       color {colorUseRule=UseColor_ColorTable, specifiedColor=<rgba 0 0 0 0>}
       cornerRadii=<6 6 6 6>
      }
      tintColor {colorUseRule=UseColor_Specified, specifiedColor=<rgba 0.0284 0.0284 0.0284 1>}
      uVRegion {bIsValid=false, max=<0 0>, min=<0 0>}
     }
     pressedForeground {colorUseRule=UseColor_ColorTable, specifiedColor=<rgba 0 0 0 0>}
     pressedPadding=<ltrb 12 2.5 12 0.5>
     pressedSlateSound {resourceObject=null}
    }
   }
   slot {class=OverlaySlot}
  }
  TextBlock {
   widgetClassPath=&TextBlock
   parent=Overlay_1
   widget=TextBlock
   properties {
    justification=Center
    text=T
    visibility=SelfHitTestInvisible
    font {
     bForceMonospaced=false
     bMaterialIsStencil=false
     fontMaterial=null
     fontObject=&Roboto
     letterSpacing=0
     monospacedWidth=1
     size=7.5
     skewAmount=0
     typefaceFontName=Bold
     outlineSettings {
      bApplyOutlineToDropShadows=false
      bMiteredCorners=false
      bSeparateFillAlpha=false
      outlineMaterial=null
      outlineSize=0
      outlineColor=<rgba 0 0 0 1>
     }
    }
   }
   slot {class=OverlaySlot, properties {horizontalAlignment=HAlign_Center, verticalAlignment=VAlign_Center}}
  }
  MainSwitcher {
   widgetClassPath=&WidgetSwitcher
   bIsVariable=true
   parent=HorizontalBox_0
   widget=MainSwitcher
   slot {class=HorizontalBoxSlot, properties {size {sizeRule=Fill, value=1}}}
  }
  DebugScreen {widgetClassPath=&VerticalBox, parent=MainSwitcher, widget=DebugScreen, slot {class=WidgetSwitcherSlot}}
  ProjectFilter {
   widgetClassPath=/Script/Blutility.EditorUtilityExpandableArea
   bIsVariable=true
   parent=DebugScreen
   widget=ProjectFilter
   properties {
    clipping=ClipToBounds
    borderBrush {
     bIsDynamicallyLoaded=false
     drawAs=NoDrawType
     imageType=NoImage
     mirroring=NoMirror
     resourceName=None
     resourceObject=null
     tiling=NoTile
     imageSize=<0 0>
     margin=<ltrb 0 0 0 0>
     outlineSettings {
      bUseBrushTransparency=false
      roundingType=FixedRadius
      width=0
      color {colorUseRule=UseColor_Specified, specifiedColor=<rgba 0 0 0 0>}
      cornerRadii=<4 4 4 4>
     }
     tintColor {colorUseRule=UseColor_Specified, specifiedColor=<rgba 0.0176 0.0176 0.0176 1>}
     uVRegion {bIsValid=false, max=<0 0>, min=<0 0>}
    }
    style {
     rolloutAnimationSeconds=0.3
     collapsedImage {
      bIsDynamicallyLoaded=false
      drawAs=Image
      imageType=Vector
      mirroring=NoMirror
      resourceName=../../../Engine/Content/Slate/Starship/Common/caret-right.svg
      resourceObject=null
      tiling=NoTile
      imageSize=<16 16>
      margin=<ltrb 0 0 0 0>
      outlineSettings {
       bUseBrushTransparency=false
       roundingType=HalfHeightRadius
       width=0
       color {colorUseRule=UseColor_Specified, specifiedColor=<rgba 0 0 0 0>}
       cornerRadii=<0 0 0 0>
      }
      tintColor {colorUseRule=UseColor_Specified, specifiedColor=<rgba 0.5271 0.5271 0.5271 1>}
      uVRegion {bIsValid=false, max=<0 0>, min=<0 0>}
     }
     expandedImage {
      bIsDynamicallyLoaded=false
      drawAs=Image
      imageType=Vector
      mirroring=NoMirror
      resourceName=../../../Engine/Content/Slate/Starship/Common/caret-down.svg
      resourceObject=null
      tiling=NoTile
      imageSize=<16 16>
      margin=<ltrb 0 0 0 0>
      outlineSettings {
       bUseBrushTransparency=false
       roundingType=HalfHeightRadius
       width=0
       color {colorUseRule=UseColor_Specified, specifiedColor=<rgba 0 0 0 0>}
       cornerRadii=<0 0 0 0>
      }
      tintColor {colorUseRule=UseColor_Specified, specifiedColor=<rgba 0.5271 0.5271 0.5271 1>}
      uVRegion {bIsValid=false, max=<0 0>, min=<0 0>}
     }
    }
   }
   slot {class=VerticalBoxSlot, properties {verticalAlignment=VAlign_Top}}
  }
  HorizontalBox_3 {widgetClassPath=&HorizontalBox, namedSlotHost=ProjectFilter, widget=HorizontalBox_3}
  Txt_ProjectFilter {
   widgetClassPath=&UEFN_TextBlock
   parent=HorizontalBox_3
   widget=Txt_ProjectFilter
   properties {
    text=Project Filter
    font {
     bForceMonospaced=false
     bMaterialIsStencil=false
     fontMaterial=null
     fontObject=&NotoSans
     letterSpacing=0
     monospacedWidth=1
     size=7.5
     skewAmount=0
     typefaceFontName=NotoSans-Regular
     outlineSettings {
      bApplyOutlineToDropShadows=false
      bMiteredCorners=false
      bSeparateFillAlpha=false
      outlineMaterial=null
      outlineSize=0
      outlineColor=<rgba 0 0 0 1>
     }
    }
   }
   slot {class=HorizontalBoxSlot, properties {verticalAlignment=VAlign_Center, size {sizeRule=Fill, value=1}}}
  }
  ProjectFilter_ListView {
   widgetClassPath=/Script/Blutility.EditorUtilityListView
   bIsVariable=true
   namedSlotHost=ProjectFilter
   widget=ProjectFilter_ListView
   properties {
    bAllowDragging=false
    bClearSelectionOnClick=true
    bEnableScrollAnimation=true
    clipping=Inherit
    entryWidgetClass=&/EDM_ListEntry_ProjectFilter.EDM_ListEntry_ProjectFilter_C
    selectionMode=None
   }
  }
  MainHeader {
   widgetClassPath=&VerticalBox
   parent=DebugScreen
   widget=MainHeader
   slot {class=VerticalBoxSlot, properties {padding=<ltrb 0 0 0 3>}}
  }
  Buttons {
   widgetClassPath=&StackBox
   parent=MainHeader
   widget=Buttons
   slot {class=VerticalBoxSlot, properties {padding=<ltrb 0 2 0 2>, size {sizeRule=Fill, value=1}}}
  }
  RefreshButton {
   widgetClassPath=&EditorUtilityButton
   parent=Buttons
   widget=RefreshButton
   properties {toolTipText=Sync Cheats}
   slot {class=StackBoxSlot, properties {size {sizeRule=Automatic, value=0.35}}}
  }
  HorizontalBox_1 {
   widgetClassPath=&HorizontalBox
   parent=RefreshButton
   widget=HorizontalBox_1
   slot {class=ButtonSlot, properties {padding=<ltrb 2 2 2 2>}}
  }
  Image_104 {
   widgetClassPath=&Image
   bIsVariable=true
   parent=HorizontalBox_1
   widget=Image_104
   properties {
    brush {
     bIsDynamicallyLoaded=false
     drawAs=Image
     imageType=NoImage
     mirroring=NoMirror
     resourceName=None
     resourceObject=/EditorDebugMenu/Textures/T_Refresh
     tiling=NoTile
     imageSize=<15 15>
     margin=<ltrb 0 0 0 0>
     outlineSettings {
      bUseBrushTransparency=false
      roundingType=HalfHeightRadius
      width=0
      color {colorUseRule=UseColor_Specified, specifiedColor=<rgba 0 0 0 0>}
      cornerRadii=<0 0 0 0>
     }
     tintColor {colorUseRule=UseColor_Specified, specifiedColor=<rgba 1 1 1 1>}
     uVRegion {bIsValid=false, max=<0 0>, min=<0 0>}
    }
    colorAndOpacity=<rgba 1 1 1 0.5>
   }
   slot {class=HorizontalBoxSlot, properties {horizontalAlignment=HAlign_Center, verticalAlignment=VAlign_Center}}
  }
  ButtonText {
   widgetClassPath=&UEFN_TextBlock
   parent=HorizontalBox_1
   widget=ButtonText
   properties {
    text=Sync
    font {
     bForceMonospaced=false
     bMaterialIsStencil=false
     fontMaterial=null
     fontObject=&NotoSans
     letterSpacing=0
     monospacedWidth=1
     size=7.5
     skewAmount=0
     typefaceFontName=NotoSans-Regular
     outlineSettings {
      bApplyOutlineToDropShadows=false
      bMiteredCorners=false
      bSeparateFillAlpha=false
      outlineMaterial=null
      outlineSize=0
      outlineColor=<rgba 0 0 0 1>
     }
    }
   }
   slot {class=HorizontalBoxSlot, properties {verticalAlignment=VAlign_Center, padding=<ltrb 8 0 0 0>}}
  }
  ExpandButton {
   widgetClassPath=&EditorUtilityButton
   parent=Buttons
   widget=ExpandButton
   properties {toolTipText=Expand All Cheats}
   slot {class=StackBoxSlot, properties {size {sizeRule=Automatic, value=0.35}}}
  }
  StackBox_3 {widgetClassPath=&StackBox, parent=ExpandButton, widget=StackBox_3, slot {class=ButtonSlot, properties {padding=<ltrb 2 2 2 2>}}}
  Image_224 {
   widgetClassPath=&Image
   bIsVariable=true
   parent=StackBox_3
   widget=Image_224
   properties {brush {
     bIsDynamicallyLoaded=false
     drawAs=Image
     imageType=NoImage
     mirroring=NoMirror
     resourceName=None
     resourceObject=/EditorDebugMenu/Textures/T_Expand
     tiling=NoTile
     imageSize=<16 16>
     margin=<ltrb 0 0 0 0>
     outlineSettings {
      bUseBrushTransparency=false
      roundingType=HalfHeightRadius
      width=0
      color {colorUseRule=UseColor_Specified, specifiedColor=<rgba 0 0 0 0>}
      cornerRadii=<0 0 0 0>
     }
     tintColor {colorUseRule=UseColor_Specified, specifiedColor=<rgba 1 1 1 1>}
     uVRegion {bIsValid=false, max=<0 0>, min=<0 0>}
    }}
   slot {class=StackBoxSlot}
  }
  Txt_ExpandAll {
   widgetClassPath=&UEFN_TextBlock
   parent=StackBox_3
   widget=Txt_ExpandAll
   properties {
    text=Expand
    font {
     bForceMonospaced=false
     bMaterialIsStencil=false
     fontMaterial=null
     fontObject=&NotoSans
     letterSpacing=0
     monospacedWidth=1
     size=7.5
     skewAmount=0
     typefaceFontName=NotoSans-Regular
     outlineSettings {
      bApplyOutlineToDropShadows=false
      bMiteredCorners=false
      bSeparateFillAlpha=false
      outlineMaterial=null
      outlineSize=0
      outlineColor=<rgba 0 0 0 1>
     }
    }
   }
   slot {class=StackBoxSlot, properties {padding=<ltrb 5 0 0 0>}}
  }
  CollapseAllButton {
   widgetClassPath=&EditorUtilityButton
   parent=Buttons
   widget=CollapseAllButton
   properties {toolTipText=Collapse All Cheats}
   slot {class=StackBoxSlot, properties {size {sizeRule=Automatic, value=0.35}}}
  }
  StackBox {
   widgetClassPath=&StackBox
   parent=CollapseAllButton
   widget=StackBox
   slot {class=ButtonSlot, properties {padding=<ltrb 2 2 2 2>}}
  }
  Image_1 {
   widgetClassPath=&Image
   bIsVariable=true
   parent=StackBox
   widget=Image_1
   properties {brush {
     bIsDynamicallyLoaded=false
     drawAs=Image
     imageType=NoImage
     mirroring=NoMirror
     resourceName=None
     resourceObject=/EditorDebugMenu/Textures/T_Collapse
     tiling=NoTile
     imageSize=<16 16>
     margin=<ltrb 0 0 0 0>
     outlineSettings {
      bUseBrushTransparency=false
      roundingType=HalfHeightRadius
      width=0
      color {colorUseRule=UseColor_Specified, specifiedColor=<rgba 0 0 0 0>}
      cornerRadii=<0 0 0 0>
     }
     tintColor {colorUseRule=UseColor_Specified, specifiedColor=<rgba 1 1 1 1>}
     uVRegion {bIsValid=false, max=<0 0>, min=<0 0>}
    }}
   slot {class=StackBoxSlot}
  }
  Txt_ExpandAll_1 {
   widgetClassPath=&UEFN_TextBlock
   parent=StackBox
   widget=Txt_ExpandAll_1
   properties {
    text=Collapse
    font {
     bForceMonospaced=false
     bMaterialIsStencil=false
     fontMaterial=null
     fontObject=&NotoSans
     letterSpacing=0
     monospacedWidth=1
     size=7.5
     skewAmount=0
     typefaceFontName=NotoSans-Regular
     outlineSettings {
      bApplyOutlineToDropShadows=false
      bMiteredCorners=false
      bSeparateFillAlpha=false
      outlineMaterial=null
      outlineSize=0
      outlineColor=<rgba 0 0 0 1>
     }
    }
   }
   slot {class=StackBoxSlot, properties {padding=<ltrb 5 0 0 0>}}
  }
  SlackButton {
   widgetClassPath=&EditorUtilityButton
   parent=Buttons
   widget=SlackButton
   properties {toolTipText=Open up Slack to the Debug Menu Channel}
   slot {class=StackBoxSlot, properties {horizontalAlignment=HAlign_Right, size {sizeRule=Fill, value=0.35}}}
  }
  StackBox_4 {widgetClassPath=&StackBox, parent=SlackButton, widget=StackBox_4, slot {class=ButtonSlot, properties {padding=<ltrb 2 2 2 2>}}}
  Image {
   widgetClassPath=&Image
   bIsVariable=true
   parent=StackBox_4
   widget=Image
   properties {brush {
     bIsDynamicallyLoaded=false
     drawAs=Image
     imageType=FullColor
     mirroring=NoMirror
     resourceName=None
     resourceObject=/EditorDebugMenu/Textures/Slack_icon_2019_svg
     tiling=NoTile
     imageSize=<14 14>
     margin=<ltrb 0 0 0 0>
     outlineSettings {
      bUseBrushTransparency=false
      roundingType=HalfHeightRadius
      width=0
      color {colorUseRule=UseColor_Specified, specifiedColor=<rgba 0 0 0 0>}
      cornerRadii=<0 0 0 0>
     }
     tintColor {colorUseRule=UseColor_Specified, specifiedColor=<rgba 1 1 1 1>}
     uVRegion {bIsValid=false, max=<0 0>, min=<0 0>}
    }}
   slot {class=StackBoxSlot}
  }
  Txt_ExpandAll_3 {
   widgetClassPath=&UEFN_TextBlock
   parent=StackBox_4
   widget=Txt_ExpandAll_3
   properties {
    text="HELP!"
    font {
     bForceMonospaced=false
     bMaterialIsStencil=false
     fontMaterial=null
     fontObject=&NotoSans
     letterSpacing=0
     monospacedWidth=1
     size=7.5
     skewAmount=0
     typefaceFontName=NotoSans-Regular
     outlineSettings {
      bApplyOutlineToDropShadows=false
      bMiteredCorners=false
      bSeparateFillAlpha=false
      outlineMaterial=null
      outlineSize=0
      outlineColor=<rgba 0 0 0 1>
     }
    }
   }
   slot {class=StackBoxSlot, properties {padding=<ltrb 5 0 0 0>}}
  }
  Border_0 {
   widgetClassPath=&Border
   parent=MainHeader
   widget=Border_0
   properties {brushColor=<rgba 0.0104 0.0104 0.0104 1>}
   slot {class=VerticalBoxSlot}
  }
  StackBox_0 {widgetClassPath=&StackBox, parent=Border_0, widget=StackBox_0, properties {orientation=Orient_Vertical}, slot {class=BorderSlot}}
  Txt_ExpandAll_4 {
   widgetClassPath=&UEFN_TextBlock
   parent=StackBox_0
   widget=Txt_ExpandAll_4
   properties {
    text=Search:
    colorAndOpacity {colorUseRule=UseColor_Specified, specifiedColor=<rgba 0.2969 0.2969 0.2969 1>}
    font {
     bForceMonospaced=false
     bMaterialIsStencil=false
     fontMaterial=null
     fontObject=&NotoSans
     letterSpacing=0
     monospacedWidth=1
     size=8.25
     skewAmount=0
     typefaceFontName=NotoSans-Regular
     outlineSettings {
      bApplyOutlineToDropShadows=false
      bMiteredCorners=false
      bSeparateFillAlpha=false
      outlineMaterial=null
      outlineSize=0
      outlineColor=<rgba 0 0 0 1>
     }
    }
   }
   slot {class=StackBoxSlot, properties {padding=<ltrb 5 2 0 0>}}
  }
  SearchBar {
   widgetClassPath=&HorizontalBox
   parent=StackBox_0
   widget=SearchBar
   slot {class=StackBoxSlot, properties {padding=<ltrb 10 0 0 0>}}
  }
  SearchBox {
   widgetClassPath=&EditorUtilityEditableTextBox
   bIsVariable=true
   parent=SearchBar
   widget=SearchBox
   properties {widgetStyle {
     backgroundColor {colorUseRule=UseColor_Specified, specifiedColor=<rgba 1 1 1 1>}
     backgroundImageFocused {
      bIsDynamicallyLoaded=false
      drawAs=RoundedBox
      imageType=NoImage
      mirroring=NoMirror
      resourceName=None
      resourceObject=null
      tiling=NoTile
      imageSize=<0 0>
      margin=<ltrb 0 0 0 0>
      outlineSettings {
       bUseBrushTransparency=false
       roundingType=HalfHeightRadius
       width=1
       color {colorUseRule=UseColor_ColorTable, specifiedColor=<rgba 0 0 0 0>}
       cornerRadii=<4 4 4 4>
      }
      tintColor {colorUseRule=UseColor_Specified, specifiedColor=<rgba 0.0048 0.0048 0.0048 1>}
      uVRegion {bIsValid=false, max=<0 0>, min=<0 0>}
     }
     backgroundImageHovered {
      bIsDynamicallyLoaded=false
      drawAs=RoundedBox
      imageType=NoImage
      mirroring=NoMirror
      resourceName=None
      resourceObject=null
      tiling=NoTile
      imageSize=<0 0>
      margin=<ltrb 0 0 0 0>
      outlineSettings {
       bUseBrushTransparency=false
       roundingType=HalfHeightRadius
       width=1
       color {colorUseRule=UseColor_ColorTable, specifiedColor=<rgba 0 0 0 0>}
       cornerRadii=<4 4 4 4>
      }
      tintColor {colorUseRule=UseColor_Specified, specifiedColor=<rgba 0.0048 0.0048 0.0048 1>}
      uVRegion {bIsValid=false, max=<0 0>, min=<0 0>}
     }
     backgroundImageNormal {
      bIsDynamicallyLoaded=false
      drawAs=RoundedBox
      imageType=NoImage
      mirroring=NoMirror
      resourceName=None
      resourceObject=null
      tiling=NoTile
      imageSize=<0 0>
      margin=<ltrb 0 0 0 0>
      outlineSettings {
       bUseBrushTransparency=false
       roundingType=HalfHeightRadius
       width=1
       color {colorUseRule=UseColor_ColorTable, specifiedColor=<rgba 0 0 0 0>}
       cornerRadii=<4 4 4 4>
      }
      tintColor {colorUseRule=UseColor_Specified, specifiedColor=<rgba 0.0048 0.0048 0.0048 1>}
      uVRegion {bIsValid=false, max=<0 0>, min=<0 0>}
     }
     backgroundImageReadOnly {
      bIsDynamicallyLoaded=false
      drawAs=RoundedBox
      imageType=NoImage
      mirroring=NoMirror
      resourceName=None
      resourceObject=null
      tiling=NoTile
      imageSize=<0 0>
      margin=<ltrb 0 0 0 0>
      outlineSettings {
       bUseBrushTransparency=false
       roundingType=HalfHeightRadius
       width=1
       color {colorUseRule=UseColor_ColorTable, specifiedColor=<rgba 0 0 0 0>}
       cornerRadii=<4 4 4 4>
      }
      tintColor {colorUseRule=UseColor_Specified, specifiedColor=<rgba 0.0284 0.0284 0.0284 1>}
      uVRegion {bIsValid=false, max=<0 0>, min=<0 0>}
     }
     focusedForegroundColor {colorUseRule=UseColor_Specified, specifiedColor=<rgba 1 1 1 1>}
     foregroundColor {colorUseRule=UseColor_Specified, specifiedColor=<rgba 0.5271 0.5271 0.5271 1>}
     hScrollBarPadding=<ltrb 0 0 0 0>
     padding=<ltrb 8 3 8 4>
     readOnlyForegroundColor {colorUseRule=UseColor_Foreground, specifiedColor=<rgba 1 0 1 1>}
     scrollBarStyle {
      thickness=8
      draggedThumbImage {
       bIsDynamicallyLoaded=false
       drawAs=RoundedBox
       imageType=NoImage
       mirroring=NoMirror
       resourceName=None
       resourceObject=null
       tiling=NoTile
       imageSize=<0 0>
       margin=<ltrb 0 0 0 0>
       outlineSettings {
        bUseBrushTransparency=false
        roundingType=FixedRadius
        width=0
        color {colorUseRule=UseColor_Specified, specifiedColor=<rgba 0 0 0 0>}
        cornerRadii=<4 4 4 4>
       }
       tintColor {colorUseRule=UseColor_Specified, specifiedColor=<rgba 0.2159 0.2159 0.2159 1>}
       uVRegion {bIsValid=false, max=<0 0>, min=<0 0>}
      }
      horizontalBackgroundImage {
       bIsDynamicallyLoaded=false
       drawAs=NoDrawType
       imageType=NoImage
       mirroring=NoMirror
       resourceName=None
       resourceObject=null
       tiling=NoTile
       imageSize=<0 0>
       margin=<ltrb 0 0 0 0>
       outlineSettings {
        bUseBrushTransparency=false
        roundingType=HalfHeightRadius
        width=0
        color {colorUseRule=UseColor_Specified, specifiedColor=<rgba 0 0 0 0>}
        cornerRadii=<0 0 0 0>
       }
       tintColor {colorUseRule=UseColor_Specified, specifiedColor=<rgba 1 1 1 1>}
       uVRegion {bIsValid=false, max=<0 0>, min=<0 0>}
      }
      horizontalBottomSlotImage {
       bIsDynamicallyLoaded=false
       drawAs=NoDrawType
       imageType=NoImage
       mirroring=NoMirror
       resourceName=None
       resourceObject=null
       tiling=NoTile
       imageSize=<0 0>
       margin=<ltrb 0 0 0 0>
       outlineSettings {
        bUseBrushTransparency=false
        roundingType=HalfHeightRadius
        width=0
        color {colorUseRule=UseColor_Specified, specifiedColor=<rgba 0 0 0 0>}
        cornerRadii=<0 0 0 0>
       }
       tintColor {colorUseRule=UseColor_Specified, specifiedColor=<rgba 1 1 1 1>}
       uVRegion {bIsValid=false, max=<0 0>, min=<0 0>}
      }
      horizontalTopSlotImage {
       bIsDynamicallyLoaded=false
       drawAs=NoDrawType
       imageType=NoImage
       mirroring=NoMirror
       resourceName=None
       resourceObject=null
       tiling=NoTile
       imageSize=<0 0>
       margin=<ltrb 0 0 0 0>
       outlineSettings {
        bUseBrushTransparency=false
        roundingType=HalfHeightRadius
        width=0
        color {colorUseRule=UseColor_Specified, specifiedColor=<rgba 0 0 0 0>}
        cornerRadii=<0 0 0 0>
       }
       tintColor {colorUseRule=UseColor_Specified, specifiedColor=<rgba 1 1 1 1>}
       uVRegion {bIsValid=false, max=<0 0>, min=<0 0>}
      }
      hoveredThumbImage {
       bIsDynamicallyLoaded=false
       drawAs=RoundedBox
       imageType=NoImage
       mirroring=NoMirror
       resourceName=None
       resourceObject=null
       tiling=NoTile
       imageSize=<0 0>
       margin=<ltrb 0 0 0 0>
       outlineSettings {
        bUseBrushTransparency=false
        roundingType=FixedRadius
        width=0
        color {colorUseRule=UseColor_Specified, specifiedColor=<rgba 0 0 0 0>}
        cornerRadii=<4 4 4 4>
       }
       tintColor {colorUseRule=UseColor_Specified, specifiedColor=<rgba 0.2159 0.2159 0.2159 1>}
       uVRegion {bIsValid=false, max=<0 0>, min=<0 0>}
      }
      normalThumbImage {
       bIsDynamicallyLoaded=false
       drawAs=RoundedBox
       imageType=NoImage
       mirroring=NoMirror
       resourceName=None
       resourceObject=null
       tiling=NoTile
       imageSize=<0 0>
       margin=<ltrb 0 0 0 0>
       outlineSettings {
        bUseBrushTransparency=false
        roundingType=FixedRadius
        width=0
        color {colorUseRule=UseColor_Specified, specifiedColor=<rgba 0 0 0 0>}
        cornerRadii=<4 4 4 4>
       }
       tintColor {colorUseRule=UseColor_Specified, specifiedColor=<rgba 0.0953 0.0953 0.0953 1>}
       uVRegion {bIsValid=false, max=<0 0>, min=<0 0>}
      }
      verticalBackgroundImage {
       bIsDynamicallyLoaded=false
       drawAs=NoDrawType
       imageType=NoImage
       mirroring=NoMirror
       resourceName=None
       resourceObject=null
       tiling=NoTile
       imageSize=<0 0>
       margin=<ltrb 0 0 0 0>
       outlineSettings {
        bUseBrushTransparency=false
        roundingType=HalfHeightRadius
        width=0
        color {colorUseRule=UseColor_Specified, specifiedColor=<rgba 0 0 0 0>}
        cornerRadii=<0 0 0 0>
       }
       tintColor {colorUseRule=UseColor_Specified, specifiedColor=<rgba 1 1 1 1>}
       uVRegion {bIsValid=false, max=<0 0>, min=<0 0>}
      }
      verticalBottomSlotImage {
       bIsDynamicallyLoaded=false
       drawAs=NoDrawType
       imageType=NoImage
       mirroring=NoMirror
       resourceName=None
       resourceObject=null
       tiling=NoTile
       imageSize=<0 0>
       margin=<ltrb 0 0 0 0>
       outlineSettings {
        bUseBrushTransparency=false
        roundingType=HalfHeightRadius
        width=0
        color {colorUseRule=UseColor_Specified, specifiedColor=<rgba 0 0 0 0>}
        cornerRadii=<0 0 0 0>
       }
       tintColor {colorUseRule=UseColor_Specified, specifiedColor=<rgba 1 1 1 1>}
       uVRegion {bIsValid=false, max=<0 0>, min=<0 0>}
      }
      verticalTopSlotImage {
       bIsDynamicallyLoaded=false
       drawAs=NoDrawType
       imageType=NoImage
       mirroring=NoMirror
       resourceName=None
       resourceObject=null
       tiling=NoTile
       imageSize=<0 0>
       margin=<ltrb 0 0 0 0>
       outlineSettings {
        bUseBrushTransparency=false
        roundingType=HalfHeightRadius
        width=0
        color {colorUseRule=UseColor_Specified, specifiedColor=<rgba 0 0 0 0>}
        cornerRadii=<0 0 0 0>
       }
       tintColor {colorUseRule=UseColor_Specified, specifiedColor=<rgba 1 1 1 1>}
       uVRegion {bIsValid=false, max=<0 0>, min=<0 0>}
      }
     }
     textStyle {
      overflowPolicy=Clip
      transformPolicy=None
      colorAndOpacity {colorUseRule=UseColor_Foreground, specifiedColor=<rgba 1 0 1 1>}
      font {
       bForceMonospaced=false
       bMaterialIsStencil=false
       fontMaterial=null
       fontObject=&NotoSans
       letterSpacing=0
       monospacedWidth=1
       size=8.25
       skewAmount=0
       typefaceFontName=NotoSans-Regular
       outlineSettings {
        bApplyOutlineToDropShadows=false
        bMiteredCorners=false
        bSeparateFillAlpha=false
        outlineMaterial=null
        outlineSize=0
        outlineColor=<rgba 0 0 0 1>
       }
      }
      highlightColor {colorUseRule=UseColor_ColorTable, specifiedColor=<rgba 0 0 0 0>}
      highlightShape {
       bIsDynamicallyLoaded=false
       drawAs=Image
       imageType=NoImage
       mirroring=NoMirror
       resourceName=None
       resourceObject=null
       tiling=NoTile
       imageSize=<0 0>
       margin=<ltrb 0 0 0 0>
       outlineSettings {
        bUseBrushTransparency=false
        roundingType=HalfHeightRadius
        width=0
        color {colorUseRule=UseColor_Specified, specifiedColor=<rgba 0 0 0 0>}
        cornerRadii=<0 0 0 0>
       }
       tintColor {colorUseRule=UseColor_ColorTable, specifiedColor=<rgba 0 0 0 0>}
       uVRegion {bIsValid=false, max=<0 0>, min=<0 0>}
      }
      selectedBackgroundColor {colorUseRule=UseColor_ColorTable, specifiedColor=<rgba 0 0 0 0>}
      shadowColorAndOpacity=<rgba 0 0 0 1>
      shadowOffset=<0 0>
      strikeBrush {
       bIsDynamicallyLoaded=false
       drawAs=Image
       imageType=NoImage
       mirroring=NoMirror
       resourceName=None
       resourceObject=null
       tiling=NoTile
       imageSize=<32 32>
       margin=<ltrb 0 0 0 0>
       outlineSettings {
        bUseBrushTransparency=false
        roundingType=HalfHeightRadius
        width=0
        color {colorUseRule=UseColor_Specified, specifiedColor=<rgba 0 0 0 0>}
        cornerRadii=<0 0 0 0>
       }
       tintColor {colorUseRule=UseColor_Specified, specifiedColor=<rgba 1 1 1 1>}
       uVRegion {bIsValid=false, max=<0 0>, min=<0 0>}
      }
      underlineBrush {
       bIsDynamicallyLoaded=false
       drawAs=Image
       imageType=NoImage
       mirroring=NoMirror
       resourceName=None
       resourceObject=null
       tiling=NoTile
       imageSize=<32 32>
       margin=<ltrb 0 0 0 0>
       outlineSettings {
        bUseBrushTransparency=false
        roundingType=HalfHeightRadius
        width=0
        color {colorUseRule=UseColor_Specified, specifiedColor=<rgba 0 0 0 0>}
        cornerRadii=<0 0 0 0>
       }
       tintColor {colorUseRule=UseColor_Specified, specifiedColor=<rgba 1 1 1 1>}
       uVRegion {bIsValid=false, max=<0 0>, min=<0 0>}
      }
     }
     vScrollBarPadding=<ltrb 0 0 0 0>
    }}
   slot {class=HorizontalBoxSlot, properties {size {sizeRule=Fill, value=1}}}
  }
  ClearSearch {
   widgetClassPath=&EditorUtilityButton
   parent=SearchBar
   widget=ClearSearch
   properties {widgetStyle {
     clickedSlateSound {resourceObject=null}
     disabled {
      bIsDynamicallyLoaded=false
      drawAs=RoundedBox
      imageType=NoImage
      mirroring=NoMirror
      resourceName=None
      resourceObject=null
      tiling=NoTile
      imageSize=<0 0>
      margin=<ltrb 0 0 0 0>
      outlineSettings {
       bUseBrushTransparency=false
       roundingType=FixedRadius
       width=1
       color {colorUseRule=UseColor_ColorTable, specifiedColor=<rgba 0 0 0 0>}
       cornerRadii=<4 4 4 4>
      }
      tintColor {colorUseRule=UseColor_Specified, specifiedColor=<rgba 0.0395 0.0395 0.0395 1>}
      uVRegion {bIsValid=false, max=<0 0>, min=<0 0>}
     }
     disabledForeground {colorUseRule=UseColor_ColorTable, specifiedColor=<rgba 0 0 0 0>}
     hovered {
      bIsDynamicallyLoaded=false
      drawAs=RoundedBox
      imageType=NoImage
      mirroring=NoMirror
      resourceName=None
      resourceObject=null
      tiling=NoTile
      imageSize=<32 32>
      margin=<ltrb 0 0 0 0>
      outlineSettings {
       bUseBrushTransparency=false
       roundingType=FixedRadius
       width=1
       color {colorUseRule=UseColor_ColorTable, specifiedColor=<rgba 0 0 0 0>}
       cornerRadii=<4 4 4 4>
      }
      tintColor {colorUseRule=UseColor_Specified, specifiedColor=<rgba 0.0953 0.0953 0.0953 1>}
      uVRegion {bIsValid=false, max=<0 0>, min=<0 0>}
     }
     hoveredForeground {colorUseRule=UseColor_ColorTable, specifiedColor=<rgba 0 0 0 0>}
     hoveredSlateSound {resourceObject=null}
     normal {
      bIsDynamicallyLoaded=false
      drawAs=NoDrawType
      imageType=FullColor
      mirroring=NoMirror
      resourceName=None
      resourceObject=/EditorDebugMenu/Textures/T_Clear
      tiling=NoTile
      imageSize=<32 32>
      margin=<ltrb 0 0 0 0>
      outlineSettings {
       bUseBrushTransparency=false
       roundingType=FixedRadius
       width=1
       color {colorUseRule=UseColor_ColorTable, specifiedColor=<rgba 0 0 0 0>}
       cornerRadii=<4 4 4 4>
      }
      tintColor {colorUseRule=UseColor_Specified, specifiedColor=<rgba 0.9844 0.9844 0.9844 1>}
      uVRegion {bIsValid=false, max=<0 0>, min=<0 0>}
     }
     normalForeground {colorUseRule=UseColor_ColorTable, specifiedColor=<rgba 0 0 0 0>}
     normalPadding=<ltrb 0 0 0 0>
     pressed {
      bIsDynamicallyLoaded=false
      drawAs=RoundedBox
      imageType=NoImage
      mirroring=NoMirror
      resourceName=None
      resourceObject=null
      tiling=NoTile
      imageSize=<32 32>
      margin=<ltrb 0 0 0 0>
      outlineSettings {
       bUseBrushTransparency=false
       roundingType=FixedRadius
       width=1
       color {colorUseRule=UseColor_ColorTable, specifiedColor=<rgba 0 0 0 0>}
       cornerRadii=<4 4 4 4>
      }
      tintColor {colorUseRule=UseColor_Specified, specifiedColor=<rgba 0.0284 0.0284 0.0284 1>}
      uVRegion {bIsValid=false, max=<0 0>, min=<0 0>}
     }
     pressedForeground {colorUseRule=UseColor_ColorTable, specifiedColor=<rgba 0 0 0 0>}
     pressedPadding=<ltrb 12 2.5 12 0.5>
     pressedSlateSound {resourceObject=null}
    }}
   slot {class=HorizontalBoxSlot, properties {size {sizeRule=Automatic, value=0.35}}}
  }
  Image_195 {
   widgetClassPath=&Image
   bIsVariable=true
   parent=ClearSearch
   widget=Image_195
   properties {brush {
     bIsDynamicallyLoaded=false
     drawAs=Image
     imageType=NoImage
     mirroring=NoMirror
     resourceName=None
     resourceObject=&T_Clear
     tiling=NoTile
     imageSize=<18 18>
     margin=<ltrb 0 0 0 0>
     outlineSettings {
      bUseBrushTransparency=false
      roundingType=HalfHeightRadius
      width=0
      color {colorUseRule=UseColor_Specified, specifiedColor=<rgba 0 0 0 0>}
      cornerRadii=<0 0 0 0>
     }
     tintColor {colorUseRule=UseColor_Specified, specifiedColor=<rgba 1 1 1 1>}
     uVRegion {bIsValid=false, max=<0 0>, min=<0 0>}
    }}
   slot {class=ButtonSlot, properties {padding=<ltrb 0 0 0 0>}}
  }
  EDM_ListGenerator {
   widgetClassPath=&/EDM_ListGenerator
   bIsVariable=true
   parent=StackBox_0
   widget=EDM_ListGenerator
   slot {class=StackBoxSlot, properties {padding=<ltrb 0 0 0 5>}}
  }
  FavouritesMenu {
   widgetClassPath=&/EDM_Container_Favourites
   bIsVariable=true
   parent=DebugScreen
   widget=FavouritesMenu
   slot {class=VerticalBoxSlot}
  }
  T_Line {
   widgetClassPath=&Image
   bIsVariable=true
   parent=DebugScreen
   widget=T_Line
   properties {
    brush {
     bIsDynamicallyLoaded=false
     drawAs=Image
     imageType=NoImage
     mirroring=NoMirror
     resourceName=None
     resourceObject=null
     tiling=NoTile
     imageSize=<32 1>
     margin=<ltrb 0 0 0 0>
     outlineSettings {
      bUseBrushTransparency=false
      roundingType=HalfHeightRadius
      width=0
      color {colorUseRule=UseColor_Specified, specifiedColor=<rgba 0 0 0 0>}
      cornerRadii=<0 0 0 0>
     }
     tintColor {colorUseRule=UseColor_Specified, specifiedColor=<rgba 1 1 1 1>}
     uVRegion {bIsValid=false, max=<0 0>, min=<0 0>}
    }
    colorAndOpacity=<rgba 0.0607 0.0677 0.0625 1>
   }
   slot {class=VerticalBoxSlot, properties {verticalAlignment=VAlign_Center, padding=<ltrb 20 0 20 3>, size {sizeRule=Automatic, value=0.5}}}
  }
  CollectionsScroll {
   widgetClassPath=&EditorUtilityScrollBox
   parent=DebugScreen
   widget=CollectionsScroll
   properties {bAnimateWheelScrolling=true, consumeMouseWheel=Always}
   slot {class=VerticalBoxSlot}
  }
  CollectionsContainer {
   widgetClassPath=&StackBox
   bIsVariable=true
   parent=CollectionsScroll
   widget=CollectionsContainer
   properties {orientation=Orient_Vertical}
   slot {class=ScrollBoxSlot, properties {size {sizeRule=Fill, value=1}}}
  }
  ProjectsScroll {
   widgetClassPath=&EditorUtilityScrollBox
   parent=DebugScreen
   widget=ProjectsScroll
   properties {alwaysShowScrollbar=true, alwaysShowScrollbarTrack=true, bAnimateWheelScrolling=true, consumeMouseWheel=Always}
   slot {class=VerticalBoxSlot, properties {size {sizeRule=Fill, value=1}}}
  }
  ProjectsContainer {
   widgetClassPath=&StackBox
   bIsVariable=true
   parent=ProjectsScroll
   widget=ProjectsContainer
   properties {orientation=Orient_Vertical}
   slot {class=ScrollBoxSlot, properties {size {sizeRule=Fill, value=1}}}
  }
  EDM_DebugMenuOptions {
   widgetClassPath=&/EDM_DebugMenuOptions
   bIsVariable=true
   parent=DebugScreen
   widget=EDM_DebugMenuOptions
   slot {class=VerticalBoxSlot}
  }
  Output_Border {
   widgetClassPath=&Border
   parent=DebugScreen
   widget=Output_Border
   properties {
    verticalAlignment=VAlign_Center
    background {
     bIsDynamicallyLoaded=false
     drawAs=Image
     imageType=NoImage
     mirroring=NoMirror
     resourceName=None
     resourceObject=null
     tiling=NoTile
     imageSize=<32 32>
     margin=<ltrb 0 0 0 0>
     outlineSettings {
      bUseBrushTransparency=false
      roundingType=HalfHeightRadius
      width=0
      color {colorUseRule=UseColor_Specified, specifiedColor=<rgba 0 0 0 0>}
      cornerRadii=<0 0 0 0>
     }
     tintColor {colorUseRule=UseColor_Specified, specifiedColor=<rgba 0.03 0.0365 0.0357 1>}
     uVRegion {bIsValid=false, max=<0 0>, min=<0 0>}
    }
    padding=<ltrb 5 5 5 5>
   }
   slot {class=VerticalBoxSlot}
  }
  Output_Text {
   widgetClassPath=/Script/Blutility.EditorUtilityEditableText
   bIsVariable=true
   parent=Output_Border
   widget=Output_Text
   properties {
    hintText=Output Text
    isReadOnly=true
    widgetStyle {
     backgroundImageComposing {
      bIsDynamicallyLoaded=false
      drawAs=Border
      imageType=FullColor
      mirroring=NoMirror
      resourceName=../../../Engine/Content/Slate/Old/HyperlinkDotted.png
      resourceObject=null
      tiling=Both
      imageSize=<0 0>
      margin=<ltrb 0 0 0 0.1875>
      outlineSettings {
       bUseBrushTransparency=false
       roundingType=HalfHeightRadius
       width=0
       color {colorUseRule=UseColor_Specified, specifiedColor=<rgba 0 0 0 0>}
       cornerRadii=<0 0 0 0>
      }
      tintColor {colorUseRule=UseColor_Specified, specifiedColor=<rgba 1 1 1 1>}
      uVRegion {bIsValid=false, max=<0 0>, min=<0 0>}
     }
     backgroundImageSelected {
      bIsDynamicallyLoaded=false
      drawAs=Image
      imageType=NoImage
      mirroring=NoMirror
      resourceName=None
      resourceObject=null
      tiling=NoTile
      imageSize=<0 0>
      margin=<ltrb 0 0 0 0>
      outlineSettings {
       bUseBrushTransparency=false
       roundingType=HalfHeightRadius
       width=0
       color {colorUseRule=UseColor_Specified, specifiedColor=<rgba 0 0 0 0>}
       cornerRadii=<0 0 0 0>
      }
      tintColor {colorUseRule=UseColor_Specified, specifiedColor=<rgba 0 0.162 0.7454 1>}
      uVRegion {bIsValid=false, max=<0 0>, min=<0 0>}
     }
     caretImage {
      bIsDynamicallyLoaded=false
      drawAs=Image
      imageType=NoImage
      mirroring=NoMirror
      resourceName=None
      resourceObject=null
      tiling=NoTile
      imageSize=<0 0>
      margin=<ltrb 0 0 0 0>
      outlineSettings {
       bUseBrushTransparency=false
       roundingType=HalfHeightRadius
       width=0
       color {colorUseRule=UseColor_Specified, specifiedColor=<rgba 0 0 0 0>}
       cornerRadii=<0 0 0 0>
      }
      tintColor {colorUseRule=UseColor_Specified, specifiedColor=<rgba 1 1 1 1>}
      uVRegion {bIsValid=false, max=<0 0>, min=<0 0>}
     }
     colorAndOpacity {colorUseRule=UseColor_Specified, specifiedColor=<rgba 1 1 1 1>}
     font {
      bForceMonospaced=false
      bMaterialIsStencil=false
      fontMaterial=null
      fontObject=&NotoSans
      letterSpacing=0
      monospacedWidth=1
      size=7.5
      skewAmount=0
      typefaceFontName=NotoSans-Regular
      outlineSettings {
       bApplyOutlineToDropShadows=false
       bMiteredCorners=false
       bSeparateFillAlpha=false
       outlineMaterial=null
       outlineSize=0
       outlineColor=<rgba 0 0 0 1>
      }
     }
     hintColorAndOpacity {colorUseRule=UseColor_Foreground, specifiedColor=<rgba 1 0 1 1>}
     selectedBackgroundColor {colorUseRule=UseColor_Foreground, specifiedColor=<rgba 1 0 1 1>}
    }
   }
   slot {class=BorderSlot, properties {verticalAlignment=VAlign_Center, padding=<ltrb 5 5 5 5>}}
  }
  LoadingScreen {
   widgetClassPath=&StackBox
   parent=MainSwitcher
   widget=LoadingScreen
   properties {orientation=Orient_Vertical}
   slot {class=WidgetSwitcherSlot}
  }
  StackBox_1 {
   widgetClassPath=&StackBox
   parent=LoadingScreen
   widget=StackBox_1
   properties {orientation=Orient_Vertical}
   slot {class=StackBoxSlot, properties {horizontalAlignment=HAlign_Center, verticalAlignment=VAlign_Center, size {sizeRule=Fill, value=1}}}
  }
  Txt_ExpandAll_2 {
   widgetClassPath=&UEFN_TextBlock
   parent=StackBox_1
   widget=Txt_ExpandAll_2
   properties {
    text=SYNCING
    font {
     bForceMonospaced=false
     bMaterialIsStencil=false
     fontMaterial=null
     fontObject=&NotoSans
     letterSpacing=0
     monospacedWidth=1
     size=7.5
     skewAmount=0
     typefaceFontName=NotoSans-Bold
     outlineSettings {
      bApplyOutlineToDropShadows=false
      bMiteredCorners=false
      bSeparateFillAlpha=false
      outlineMaterial=null
      outlineSize=0
      outlineColor=<rgba 0 0 0 1>
     }
    }
   }
   slot {class=StackBoxSlot, properties {horizontalAlignment=HAlign_Center, verticalAlignment=VAlign_Center, padding=<ltrb 5 5 5 5>}}
  }
  EditorUtilityThrobber_182 {
   widgetClassPath=/Script/Blutility.EditorUtilityThrobber
   bIsVariable=true
   parent=StackBox_1
   widget=EditorUtilityThrobber_182
   properties {image {
     bIsDynamicallyLoaded=false
     drawAs=Image
     imageType=FullColor
     mirroring=NoMirror
     resourceName=../../../Engine/Content/Slate/Common/Throbber_Piece.png
     resourceObject=null
     tiling=NoTile
     imageSize=<8 8>
     margin=<ltrb 0 0 0 0>
     outlineSettings {
      bUseBrushTransparency=false
      roundingType=HalfHeightRadius
      width=0
      color {colorUseRule=UseColor_Specified, specifiedColor=<rgba 0 0 0 0>}
      cornerRadii=<0 0 0 0>
     }
     tintColor {colorUseRule=UseColor_Foreground, specifiedColor=<rgba 1 1 1 1>}
     uVRegion {bIsValid=false, max=<0 0>, min=<0 0>}
    }}
   slot {class=StackBoxSlot, properties {horizontalAlignment=HAlign_Center, verticalAlignment=VAlign_Center}}
  }
  EUW_TodoList {widgetClassPath=/Editor_TodoList/EUW_TodoList, bIsVariable=true, parent=MainSwitcher, widget=EUW_TodoList, slot {class=WidgetSwitcherSlot}}
 ]
}
