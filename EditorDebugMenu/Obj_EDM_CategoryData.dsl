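; Obj_EDM_CategoryData is a data object for the editor debug menu that represents one named category of input entries.
; It stores a group name and an array of Obj EDM Input Data object references, with AddInput appending to that array.
; Getters return the name and the array, and CompareGroupName checks a string against the name with a case-insensitive match.
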(fn AddInput
  (params (NewItem :type Obj%20EDM%20Input%20Data%20Object%20Reference%20%28by%20ref%29))
  (Utilities|Array|Add (Variables|Default|GetGroup%20Data) NewItem))

(fn CompareGroupName
  (params (Group%20Name :type String))
  (return (Utilities|String|Equal,CaseInsensitive(String) Group%20Name (Variables|Default|GetGroup%20Name))))

(fn GetData
  (return (Variables|Default|GetGroup%20Data)))

(fn GetGroupName
  (return (Variables|Default|GetGroup%20Name)))
