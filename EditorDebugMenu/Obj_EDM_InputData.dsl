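; Data object for one Editor Debug Menu entry, acting as a button, text button, boolean, slider or task.
; Input actions build console commands from its data struct and send them, or tasks and objects, to the menu widget through interface messages.
; Sliders store normalised values, remapped and snapped for commands.
; Slider state, favourites and collections persist by unique ID in the menu's save config, with dispatchers announcing changes.
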(fn ButtonPressed
  (bind _input_data (Variables|Default|GetInput%20Data))
  (bind (_consolecommand_41_b61cefe149a87bbbbcaac99912b4a295 _additionalconsolecommands_42_b752ac1d481fb87b5c81958346d172de _targetworld_45_30d3cf1941627a9972ea57807055ca92) (Utilities|Struct|BreakStructEDMInputData01 _input_data))
  (EditorDebugMenu|RunEDMConsoleCommand(Message) (Variables|Setup|GetDebug%20Menu) _consolecommand_41_b61cefe149a87bbbbcaac99912b4a295 _targetworld_45_30d3cf1941627a9972ea57807055ca92)
  (for _array_element _additionalconsolecommands_42_b752ac1d481fb87b5c81958346d172de
    (EditorDebugMenu|RunEDMConsoleCommand(Message) (Variables|Setup|GetDebug%20Menu) _array_element _targetworld_45_30d3cf1941627a9972ea57807055ca92)))

(fn ButtonPressedWithText
  (params (Input%20Text :type String))
  (bind (_consolecommand_41_b61cefe149a87bbbbcaac99912b4a295 _targetworld_45_30d3cf1941627a9972ea57807055ca92) (Utilities|Struct|BreakStructEDMInputData01 (Variables|Default|GetInput%20Data)))
  (bind _output_get (Variables|ButtonText|SetButton%20Input%20Text Input%20Text))
  (EditorDebugMenu|RunEDMConsoleCommand(Message) (Variables|Setup|GetDebug%20Menu) (Utilities|String|Append _consolecommand_41_b61cefe149a87bbbbcaac99912b4a295 " " _output_get) _targetworld_45_30d3cf1941627a9972ea57807055ca92))

(fn GetConsoleCommand
  (bind _input_data (Variables|Default|GetInput%20Data))
  (bind (_debugtype_38_78b69ab643c98b0d8473699c01c66c19 _consolecommand_41_b61cefe149a87bbbbcaac99912b4a295) (Utilities|Struct|BreakStructEDMInputData01 _input_data))
  (switch Utilities|FlowControl|Switch|SwitchonEnum_EDM_InputType _debugtype_38_78b69ab643c98b0d8473699c01c66c19
    (:NewEnumerator0
      (return _consolecommand_41_b61cefe149a87bbbbcaac99912b4a295))
    (:NewEnumerator2
      (return (Utilities|String|Append _consolecommand_41_b61cefe149a87bbbbcaac99912b4a295 " " (Utilities|String|SelectString "1" "0" (Variables|Boolean|GetBooolean%20Value)))))
    (:NewEnumerator3
      (return (Utilities|String|Append _consolecommand_41_b61cefe149a87bbbbcaac99912b4a295 " " (Utilities|String|ToString(Float) (Variables|Slider|GetSlider%20Value)))))
    (:NewEnumerator6
      (return _consolecommand_41_b61cefe149a87bbbbcaac99912b4a295))))

(fn GetInputType
  (return (Utilities|Struct|BreakStructEDMInputData01 (Variables|Default|GetInput%20Data))))

(fn GetData
  (return (Variables|Default|GetInput%20Data)))

(fn Initalise
  (params (New%20Data :type Struct%20EDM%20Input%20Data%2001%20Structure))
  (bind _input_data (Variables|Default|GetInput%20Data))
  (bind _sliderdata_37_42f30dfb4afc229be3862183fa5634fb (Utilities|Struct|BreakStructEDMInputData01 _input_data))
  (bind (_start_2_fe102dff4138494986147fbb1bf39365 _end_4_0f44a3b24bbd87de078d188d91bb1e3d _default_9_0b843e17418e22ea1013e0a1f6dbb55c) (Utilities|Struct|BreakStructEDMOptionSliderData _sliderdata_37_42f30dfb4afc229be3862183fa5634fb))
  (bind (_value _returnvalue) (Utilities|Map|Find (Variables|/EditorDebugMenu/EDM_SaveGameConfig.EDM_SaveGameConfig_C|GetSliderStates (Variables|/EditorDebugMenu/EUW_EditorDebugMenu.EUW_EditorDebugMenu_C|GetSave%20game%20config (Variables|Setup|GetDebug%20Menu))) (Variables|Setup|GetUnique%20ID)))
  (Utilities|FlowControl|Sequence
    (:then_0
      (bind _output_get (Variables|Slider|SetSlider%20Value (Math|Float|SelectFloat _value _default_9_0b843e17418e22ea1013e0a1f6dbb55c _returnvalue)))
      (Variables|Slider|SetSlider%20Mapped%20Value (Math|Float|MapRangeClamped _output_get 0.0 1.0 _start_2_fe102dff4138494986147fbb1bf39365 _end_4_0f44a3b24bbd87de078d188d91bb1e3d)))
    (:then_1
      (switch Utilities|FlowControl|Switch|SwitchonEnum_EDM_InputType (Utilities|Struct|BreakStructEDMInputData01 (Variables|Default|GetInput%20Data))))
    (:then_2
      (bind _output_get_2 (Variables|Default|SetFavourite (Class|EDMSaveGameConfig|CheckIsFavourite (Variables|/EditorDebugMenu/EUW_EditorDebugMenu.EUW_EditorDebugMenu_C|GetSave%20game%20config (Variables|Setup|GetDebug%20Menu)) (Variables|Setup|GetUnique%20ID))))
      (if _output_get_2
        (Default|CallFavouriteUpdated true)
        (EditorDebugMenu|ToggleEDMFavourite(Message) (Variables|Setup|GetDebug%20Menu) true self)))
    (:then_3
      (bind _collections (Class|EDMSaveGameConfig|LoadInputCollection (Variables|/EditorDebugMenu/EUW_EditorDebugMenu.EUW_EditorDebugMenu_C|GetSave%20game%20config (Variables|Setup|GetDebug%20Menu)) (Variables|Setup|GetUnique%20ID)))
      (bind _output_get_3 (Variables|Default|SetIn%20collections _collections))
      (EditorDebugMenu|UpdateEDMCollections(Message) (Variables|Setup|GetDebug%20Menu) _output_get_3 (Utilities|Set|MakeSet "") self))))

(fn ResetSlider
  (bind (_start_2_fe102dff4138494986147fbb1bf39365 _end_4_0f44a3b24bbd87de078d188d91bb1e3d _default_9_0b843e17418e22ea1013e0a1f6dbb55c) (Utilities|Struct|BreakStructEDMOptionSliderData (Utilities|Struct|BreakStructEDMInputData01 (Variables|Default|GetInput%20Data))))
  (Data|Slider|UpdateSlider (Math|Float|MapRangeClamped _default_9_0b843e17418e22ea1013e0a1f6dbb55c _start_2_fe102dff4138494986147fbb1bf39365 _end_4_0f44a3b24bbd87de078d188d91bb1e3d 0.0 1.0)))

(fn RunObject
  (EditorDebugMenu|RunObject(Message) (Variables|Setup|GetDebug%20Menu) (Utilities|Struct|BreakStructEDMInputData01 (Variables|Default|GetInput%20Data))))

(fn RunTask
  (EditorDebugMenu|RunEDMTask(Message) (Variables|Setup|GetDebug%20Menu) (Utilities|Struct|BreakStructEDMInputData01 (Variables|Default|GetInput%20Data))))

(fn SaveData
  (bind _debug_menu (Variables|Setup|GetDebug%20Menu))
  (Class|EDMSaveGameConfig|SaveInput (Variables|/EditorDebugMenu/EUW_EditorDebugMenu.EUW_EditorDebugMenu_C|GetSave%20game%20config _debug_menu) (Variables|Setup|GetUnique%20ID) self)
  (EditorDebugMenu|TriggerSaveEDM(Message) _debug_menu))

(fn SetupFromSave
  (Development|PrintString 0 (Utilities|String|Append (Variables|Setup|GetUnique%20ID) " - Loaded"))
  (Variables|Default|SetFavourite (Class|EDMSaveGameConfig|CheckIsFavourite (Variables|/EditorDebugMenu/EUW_EditorDebugMenu.EUW_EditorDebugMenu_C|GetSave%20game%20config (Variables|Setup|GetDebug%20Menu)) (Variables|Setup|GetUnique%20ID))))

(fn ToggleFavourite
  (bind _debug_menu (Variables|Setup|GetDebug%20Menu))
  (bind _output_get (Variables|Default|SetFavourite (not (Variables|Default|GetFavourite))))
  (Default|CallFavouriteUpdated _output_get)
  (EditorDebugMenu|ToggleEDMFavourite(Message) _debug_menu _output_get self)
  (Class|EDMSaveGameConfig|ModifyFavourite (Variables|/EditorDebugMenu/EUW_EditorDebugMenu.EUW_EditorDebugMenu_C|GetSave%20game%20config _debug_menu) (Variables|Setup|GetUnique%20ID) _output_get)
  (EditorDebugMenu|TriggerSaveEDM _debug_menu))

(fn TriggerFromData
  (switch Utilities|FlowControl|Switch|SwitchonEnum_EDM_InputType (Utilities|Struct|BreakStructEDMInputData01 (Variables|Default|GetInput%20Data))
    (:NewEnumerator0
      (Data|Button|ButtonPressed))
    (:NewEnumerator2
      (Data|Boolean|UpdateBoolean (Variables|Boolean|GetBooolean%20Value)))
    (:NewEnumerator3
      (Data|Slider|UpdateSlider (Variables|Slider|GetSlider%20Value)))
    (:NewEnumerator5
      (Data|Task|RunTask))
    (:NewEnumerator6
      (Data|Button|ButtonPressedwithText (Variables|ButtonText|GetButton%20Input%20Text)))))

(fn UpdateBoolean
  (params (Booolean%20Value :type Boolean))
  (bind (_consolecommand_41_b61cefe149a87bbbbcaac99912b4a295 _targetworld_45_30d3cf1941627a9972ea57807055ca92) (Utilities|Struct|BreakStructEDMInputData01 (Variables|Default|GetInput%20Data)))
  (bind _output_get (Variables|Boolean|SetBooolean%20Value Booolean%20Value))
  (Default|CallBooleanUpdated _output_get)
  (bind _debug_menu (Variables|Setup|GetDebug%20Menu)
    (:then
      (EditorDebugMenu|RunEDMConsoleCommand(Message) _debug_menu (Utilities|String|Append _consolecommand_41_b61cefe149a87bbbbcaac99912b4a295 " " (Utilities|String|SelectString "1" "0" _output_get)) _targetworld_45_30d3cf1941627a9972ea57807055ca92))
    (:else)))

(fn UpdateCollectionsfromInput
  (params (Collections :type String))
  (bind _debug_menu (Variables|Setup|GetDebug%20Menu))
  (Utilities|FlowControl|Sequence
    (:then_0
      (for _array_element (Utilities|String|ParseIntoArray Collections "," false)
        (if (not (Utilities|String|IsEmpty (Utilities|String|TrimTrailing (Utilities|String|Trim _array_element))))
          (Utilities|Set|Add (LocalVariables|GetNewCollection) (Utilities|String|TrimTrailing (Utilities|String|Trim _array_element)))))
      (bind _result (Utilities|Set|Difference (Variables|Default|GetIn%20collections) (LocalVariables|GetNewCollection)))
      (EditorDebugMenu|UpdateEDMCollections(Message) (Variables|Setup|GetDebug%20Menu) (LocalVariables|GetNewCollection) _result self)
      (Variables|Default|SetIn%20collections (LocalVariables|GetNewCollection)))
    (:then_1
      ;[1 1 1] Save and dispatch
      (Class|EDMSaveGameConfig|SaveInputCollection (Variables|/EditorDebugMenu/EUW_EditorDebugMenu.EUW_EditorDebugMenu_C|GetSave%20game%20config _debug_menu) (Variables|Setup|GetUnique%20ID) (Variables|Default|GetIn%20collections))
      (EditorDebugMenu|TriggerSaveEDM(Message) _debug_menu)
      (Default|CallCollectionsUpdated))))

(fn UpdateSlider
  (params (New%20Value :type Float%20%28double-precision%29))
  (bind (_consolecommand_41_b61cefe149a87bbbbcaac99912b4a295 _targetworld_45_30d3cf1941627a9972ea57807055ca92) (Utilities|Struct|BreakStructEDMInputData01 (Variables|Default|GetInput%20Data)))
  (bind (_start_2_fe102dff4138494986147fbb1bf39365 _end_4_0f44a3b24bbd87de078d188d91bb1e3d _intervals_6_bafe7e614490efea6c43e48c3ce04a47) (Utilities|Struct|BreakStructEDMOptionSliderData (Utilities|Struct|BreakStructEDMInputData01 (Variables|Default|GetInput%20Data))))
  (if (!= New%20Value (Variables|Slider|GetSlider%20Value))
    (bind _output_get (Variables|Slider|SetSlider%20Value New%20Value))
    (Variables|Slider|SetSlider%20Mapped%20Value (Math|Float|SnapToGrid(Float) (Math|Float|MapRangeClamped _output_get 0.0 1.0 _start_2_fe102dff4138494986147fbb1bf39365 _end_4_0f44a3b24bbd87de078d188d91bb1e3d) _intervals_6_bafe7e614490efea6c43e48c3ce04a47))
    (bind _debug_menu (Variables|Setup|GetDebug%20Menu)
      (:then
        (EditorDebugMenu|RunEDMConsoleCommand(Message) _debug_menu (Utilities|String|Append _consolecommand_41_b61cefe149a87bbbbcaac99912b4a295 " " (Utilities|String|ToString(Float) (Variables|Slider|GetSlider%20Mapped%20Value))) _targetworld_45_30d3cf1941627a9972ea57807055ca92)
        (Default|CallSliderUpdated (Variables|Slider|GetSlider%20Value) (Variables|Slider|GetSlider%20Mapped%20Value))
        (Utilities|Map|Add (Variables|/EditorDebugMenu/EDM_SaveGameConfig.EDM_SaveGameConfig_C|GetSliderStates (Variables|/EditorDebugMenu/EUW_EditorDebugMenu.EUW_EditorDebugMenu_C|GetSave%20game%20config _debug_menu)) (Variables|Setup|GetUnique%20ID) (Variables|Slider|GetSlider%20Value))
        (EditorDebugMenu|TriggerSaveEDM(Message) _debug_menu))
      (:else))))
