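; Per-project data object for the Editor Debug Menu, holding one project's debug inputs and its filter visibility.
; Appended rows record their source location, then PopulateData wraps each in an InputData object, groups them into CategoryData objects and returns category keys plus a favourite map.
; SortData rebuilds the category map in a supplied key order.
; Visibility loads from the menu's save config; toggling it fires an update event and saves.
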(fn AppendRowData
  (params (Data :type Array%20of%20Struct%20EDM%20Input%20Data%2001%20Structures) (Source%20Location :type String))
  (Utilities|Array|AppendArray (Variables|Default|GetRows) Data)
  (for _array_element Data
    (Utilities|Map|Add (Variables|Default|GetSource%20Map) (Utilities|Struct|BreakStructEDMInputData01 _array_element) Source%20Location)))

(fn CreateInputData
  (bind _values (Utilities|Map|Values (Variables|Default|GetData)))
  (for _array_element _values
    _))

(fn CreateObjInputData
  (params (InputPin :type Struct%20EDM%20Input%20Data%2001%20Structure))
  (bind (_category_8_05d9292f449018413d12d7968e7be430 _name_10_7664e7304de1f78177707a877518e4f8) (Utilities|Struct|BreakStructEDMInputData01 InputPin))
  (bind _returnvalue (Game|ConstructObjectfromClass "/EditorDebugMenu/Obj_EDM_InputData.Obj_EDM_InputData_C" 0 InputPin (Variables|Setup|GetDebug%20Menu) (Utilities|String|Append (Variables|Setup|GetProject%20Name) "_" _category_8_05d9292f449018413d12d7968e7be430 "_" (Utilities|String|ToString(Name) _name_10_7664e7304de1f78177707a877518e4f8)) (Utilities|Map|Find (Variables|Default|GetSource%20Map) _name_10_7664e7304de1f78177707a877518e4f8)))
  (Save|Initalise :self _returnvalue)
  (Utilities|Map|Add (Variables|Default|GetInputs) (Utilities|Struct|BreakStructEDMInputData01 InputPin) _returnvalue)
  (return _returnvalue))

(fn Initalise
  (Variables|Setup|SetVisible (Class|EDMSaveGameConfig|GetProjectFilterVisibility (Variables|/EditorDebugMenu/EUW_EditorDebugMenu.EUW_EditorDebugMenu_C|GetSave%20game%20config (Variables|Setup|GetDebug%20Menu)) (Variables|Setup|GetProject%20Name))))

(fn PopulateData
  (bind _data (Variables|Default|GetData))
  (bind (_found_enabled_40_08449da947554cd8b6b3d99886bba2d8 _found_debugtype_38_78b69ab643c98b0d8473699c01c66c19 _found_category_8_05d9292f449018413d12d7968e7be430) (Utilities|Struct|BreakStructEDMInputData01 (Variables|GetFound_Enabled_40_08449DA947554CD8B6B3D99886BBA2D8)))
  (for _array_element (Variables|Default|GetRows)
    (bind _output_get (LocalVariables|SetFound _array_element))
    (bind _returnvalue (Game|ConstructObjectfromClass "/EditorDebugMenu/Obj_EDM_InputData.Obj_EDM_InputData_C" 0 (LocalVariables|GetFound) (Variables|Setup|GetDebug%20Menu) (Utilities|String|Append (Variables|Setup|GetProject%20Name) "_" (Utilities|Struct|BreakStructEDMInputData01 _output_get) "_" (Utilities|String|ToString(Name) (.Name_10_7664E7304DE1F78177707A877518E4F8 (Utilities|Struct|BreakStructEDMInputData01 _output_get)))) (Utilities|Map|Find (Variables|Default|GetSource%20Map) (.Name_10_7664E7304DE1F78177707A877518E4F8 (Utilities|Struct|BreakStructEDMInputData01 _output_get)))))
    (Save|Initalise _returnvalue (LocalVariables|GetFound))
    (Utilities|Map|Add (Variables|Default|GetInputs) (Utilities|Struct|BreakStructEDMInputData01 _output_get) _returnvalue)
    (Utilities|FlowControl|Sequence
      (:then_0
        (if (.ReturnValue (Utilities|Map|Find _data (Utilities|Struct|BreakStructEDMInputData01 _output_get)))
          (Class|ObjEDMCategoryData|AddInput (Utilities|Map|Find _data (Utilities|Struct|BreakStructEDMInputData01 _output_get)) _returnvalue)
          (else
            (bind _returnvalue_2 (Game|ConstructObjectfromClass "/EditorDebugMenu/Obj_EDM_CategoryData.Obj_EDM_CategoryData_C" 0 (Utilities|Struct|BreakStructEDMInputData01 _output_get)))
            (Class|ObjEDMCategoryData|AddInput _returnvalue_2 _returnvalue)
            (Utilities|Map|Add (Variables|Default|GetData) _found_category_8_05d9292f449018413d12d7968e7be430 _returnvalue_2))))
      (:then_1
        ;[1 1 1] Create favourite map
        (Utilities|Map|Add (LocalVariables|GetFavourite%20Map) (.Name_10_7664E7304DE1F78177707A877518E4F8 (Utilities|Struct|BreakStructEDMInputData01 _output_get)) (LocalVariables|GetFound)))))
  (bind _keys (Utilities|Map|Keys (Variables|Default|GetData)))
  (return _keys (LocalVariables|GetFavourite%20Map)))

(fn SortData
  (params (Sorted%20Keys :type Array%20of%20Strings))
  (LocalVariables|SetTemp (Variables|Default|GetData))
  (Utilities|Map|Clear (Variables|Default|GetData))
  (for _array_element Sorted%20Keys
    (Utilities|Map|Add (Variables|Default|GetData) _array_element (Utilities|Map|Find (LocalVariables|GetTemp) _array_element)))
  (CallFunction|CreateInputData))

(fn ToggleProjectVisibility
  (bind _debug_menu (Variables|Setup|GetDebug%20Menu))
  (bind _output_get (Variables|Setup|SetVisible (not (Variables|Setup|GetVisible))))
  (Default|CallProjectVisibilityUpdated _output_get)
  (Class|EDMSaveGameConfig|SetProjectFilterVisibility (Variables|/EditorDebugMenu/EUW_EditorDebugMenu.EUW_EditorDebugMenu_C|GetSave%20game%20config _debug_menu) (Variables|Setup|GetProject%20Name) _output_get)
  (EditorDebugMenu|TriggerSaveEDM(Message) _debug_menu))
